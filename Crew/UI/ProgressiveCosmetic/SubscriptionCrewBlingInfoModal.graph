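# Fortnite Crew info modal for Crew Styles: a blurred lightbox with a scrollable bullet list, subscriber-since logo, fine print and Back button.
# ShowUp plays on activation; backing out plays Dismiss, unless already playing, then exits the modal when it finishes.
# Generic or platform-specific entries rebuild the bullets as formatted GenericModalEntry widgets, skipping empty text.
# Once platform-specific entries arrive, a flag makes later generic entries ignored.
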WidgetBlueprint /Crew/UI/ProgressiveCosmetic/SubscriptionCrewBlingInfoModal
animations [
 ShowUp {
  binding_count=7
  length_seconds=0.75
  object=&self:ShowUp
  sequence {
   bindings [
    Text_FinePrint {
     id=#1
     possessed_class=CommonTextBlock
     tracks [
      {
       type=MovieSceneFloatTrack
       display_name=Render Opacity
       sections [{type=MovieSceneFloatSection, channels [None {type=MovieSceneScriptingFloatChannel, keys [[45,0],[57,1]]}]}]
      }
      {
       type=MovieScene2DTransformTrack
       sections [{
         type=MovieScene2DTransformSection
         channels [
          Translation.X {type=MovieSceneScriptingFloatChannel, keys [[45,128],[57,0]]}
          Translation.Y {type=MovieSceneScriptingFloatChannel}
          Angle {type=MovieSceneScriptingFloatChannel}
          Scale.X {type=MovieSceneScriptingFloatChannel}
          Scale.Y {type=MovieSceneScriptingFloatChannel}
          Shear.X {type=MovieSceneScriptingFloatChannel}
          Shear.Y {type=MovieSceneScriptingFloatChannel}
         ]
        }]
      }
     ]
    }
    Image_BackgroundBottomBar {
     id=#2
     possessed_class=Image
     tracks [
      {
       type=MovieSceneWidgetMaterialTrack
       sections [{type=MovieSceneParameterSection, channels [Shape_Offset_Top {type=MovieSceneScriptingFloatChannel, keys [[0,-1],[48,-0.25]]}]}]
      }
      {
       type=MovieSceneFloatTrack
       display_name=Render Opacity
       sections [{type=MovieSceneFloatSection, channels [None {type=MovieSceneScriptingFloatChannel, keys [[0,0],[20.4,1]]}]}]
      }
     ]
    }
    LogoSB {
     id=#3
     possessed_class=ScaleBox
     tracks [{
       type=MovieSceneFloatTrack
       display_name=Render Opacity
       sections [{type=MovieSceneFloatSection, channels [None {type=MovieSceneScriptingFloatChannel, keys [[28.8,0],[32,1]]}]}]
      }]
    }
    AnimBoundScalingBorder {
     id=#4
     possessed_class=Border
     tracks [
      {
       type=MovieSceneFloatVectorTrack
       display_name=Desired Size Scale
       sections [{
         type=MovieSceneFloatVectorSection
         channels [
          Vector.X {type=MovieSceneScriptingFloatChannel}
          Vector.Y {type=MovieSceneScriptingFloatChannel, keys [[12,0],[28.8,1]]}
         ]
        }]
      }
      {
       type=MovieSceneFloatTrack
       display_name=Render Opacity
       sections [{type=MovieSceneFloatSection, channels [None {type=MovieSceneScriptingFloatChannel, keys [[12,0],[28.8,1]]}]}]
      }
     ]
    }
    PrimaryScrollBox {
     id=#5
     possessed_class=ScrollBox
     tracks [
      {
       type=MovieSceneFloatTrack
       display_name=Render Opacity
       sections [{type=MovieSceneFloatSection, channels [None {type=MovieSceneScriptingFloatChannel, keys [[45,0],[57,1]]}]}]
      }
      {
       type=MovieScene2DTransformTrack
       sections [{
         type=MovieScene2DTransformSection
         channels [
          Translation.X {type=MovieSceneScriptingFloatChannel, keys [[45,128],[57,0]]}
          Translation.Y {type=MovieSceneScriptingFloatChannel}
          Angle {type=MovieSceneScriptingFloatChannel}
          Scale.X {type=MovieSceneScriptingFloatChannel}
          Scale.Y {type=MovieSceneScriptingFloatChannel}
          Shear.X {type=MovieSceneScriptingFloatChannel}
          Shear.Y {type=MovieSceneScriptingFloatChannel}
         ]
        }]
      }
     ]
    }
    BackgroundBlurLightbox {
     id=#6
     possessed_class=BackgroundBlur
     tracks [{
       type=MovieSceneFloatTrack
       display_name=Render Opacity
       sections [{type=MovieSceneFloatSection, channels [None {type=MovieSceneScriptingFloatChannel, keys [[0,0],[48,1]]}]}]
      }]
    }
    Button_Back {
     id=#7
     possessed_class=Spatial_Specialbutton_C
     tracks [{
       type=MovieScene2DTransformTrack
       sections [{
         type=MovieScene2DTransformSection
         channels [
          Translation.X {type=MovieSceneScriptingFloatChannel}
          Translation.Y {type=MovieSceneScriptingFloatChannel}
          Angle {type=MovieSceneScriptingFloatChannel}
          Scale.X {type=MovieSceneScriptingFloatChannel, keys [[60,0],[72,1],[90,1]]}
          Scale.Y {type=MovieSceneScriptingFloatChannel, keys [[60,0],[72,1],[90,1]]}
          Shear.X {type=MovieSceneScriptingFloatChannel}
          Shear.Y {type=MovieSceneScriptingFloatChannel}
         ]
        }]
      }]
    }
   ]
   display_rate {numerator=120}
  }
 }
 Dismiss {
  binding_count=7
  length_seconds=0.35
  object=&self:Dismiss
  sequence {bindings [
    BackgroundBlurLightbox {
     id=#8
     possessed_class=BackgroundBlur
     tracks [{
       type=MovieSceneFloatTrack
       display_name=Render Opacity
       sections [{type=MovieSceneFloatSection, channels [None {type=MovieSceneScriptingFloatChannel, keys [[4,1],[7,0]]}]}]
      }]
    }
    Text_FinePrint {
     id=#9
     possessed_class=CommonTextBlock
     tracks [
      {
       type=MovieSceneFloatTrack
       display_name=Render Opacity
       sections [{type=MovieSceneFloatSection, channels [None {type=MovieSceneScriptingFloatChannel, keys [[1.2,1],[4,0]]}]}]
      }
      {
       type=MovieScene2DTransformTrack
       sections [{
         type=MovieScene2DTransformSection
         channels [
          Translation.X {type=MovieSceneScriptingFloatChannel, keys [[1.2,0],[4,128]]}
          Translation.Y {type=MovieSceneScriptingFloatChannel}
          Angle {type=MovieSceneScriptingFloatChannel}
          Scale.X {type=MovieSceneScriptingFloatChannel}
          Scale.Y {type=MovieSceneScriptingFloatChannel}
          Shear.X {type=MovieSceneScriptingFloatChannel}
          Shear.Y {type=MovieSceneScriptingFloatChannel}
         ]
        }]
      }
     ]
    }
    Button_Back {
     id=#10
     possessed_class=Spatial_Specialbutton_C
     tracks [{
       type=MovieScene2DTransformTrack
       sections [{
         type=MovieScene2DTransformSection
         channels [
          Translation.X {type=MovieSceneScriptingFloatChannel}
          Translation.Y {type=MovieSceneScriptingFloatChannel}
          Angle {type=MovieSceneScriptingFloatChannel}
          Scale.X {type=MovieSceneScriptingFloatChannel, keys [[0,1],[2.8,1],[4,0]]}
          Scale.Y {type=MovieSceneScriptingFloatChannel, keys [[-0.0003,1],[2.8,1],[4,0]]}
          Shear.X {type=MovieSceneScriptingFloatChannel}
          Shear.Y {type=MovieSceneScriptingFloatChannel}
         ]
        }]
      }]
    }
    Image_BackgroundBottomBar {
     id=#11
     possessed_class=Image
     tracks [
      {
       type=MovieSceneWidgetMaterialTrack
       sections [{type=MovieSceneParameterSection, channels [Shape_Offset_Top {type=MovieSceneScriptingFloatChannel, keys [[3.4,-0.25],[4.6,-1]]}]}]
      }
      {
       type=MovieSceneFloatTrack
       display_name=Render Opacity
       sections [{type=MovieSceneFloatSection, channels [None {type=MovieSceneScriptingFloatChannel, keys [[3.4,1],[4.6,0]]}]}]
      }
     ]
    }
    PrimaryScrollBox {
     id=#12
     possessed_class=ScrollBox
     tracks [
      {
       type=MovieScene2DTransformTrack
       sections [{
         type=MovieScene2DTransformSection
         channels [
          Translation.X {type=MovieSceneScriptingFloatChannel, keys [[1.2,0],[4,128]]}
          Translation.Y {type=MovieSceneScriptingFloatChannel}
          Angle {type=MovieSceneScriptingFloatChannel}
          Scale.X {type=MovieSceneScriptingFloatChannel}
          Scale.Y {type=MovieSceneScriptingFloatChannel}
          Shear.X {type=MovieSceneScriptingFloatChannel}
          Shear.Y {type=MovieSceneScriptingFloatChannel}
         ]
        }]
      }
      {
       type=MovieSceneFloatTrack
       display_name=Render Opacity
       sections [{type=MovieSceneFloatSection, channels [None {type=MovieSceneScriptingFloatChannel, keys [[1.2,1],[4,0]]}]}]
      }
     ]
    }
    AnimBoundScalingBorder {
     id=#13
     possessed_class=Border
     tracks [
      {
       type=MovieSceneFloatVectorTrack
       display_name=Desired Size Scale
       sections [{
         type=MovieSceneFloatVectorSection
         channels [
          Vector.X {type=MovieSceneScriptingFloatChannel}
          Vector.Y {type=MovieSceneScriptingFloatChannel, keys [[3.8,1],[6,0]]}
         ]
        }]
      }
      {
       type=MovieSceneFloatTrack
       display_name=Render Opacity
       sections [{type=MovieSceneFloatSection, channels [None {type=MovieSceneScriptingFloatChannel, keys [[3.8,1],[6,0]]}]}]
      }
     ]
    }
    LogoSB {
     id=#14
     possessed_class=ScaleBox
     tracks [{
       type=MovieSceneFloatTrack
       display_name=Render Opacity
       sections [{type=MovieSceneFloatSection, channels [None {type=MovieSceneScriptingFloatChannel, keys [[1.8,1],[4,0]]}]}]
      }]
    }
   ]}
 }
]
graphs [EventGraph {nodes [
   n1 = TriggerExitModal() {is=title K2Node_CustomEvent}
   > ExitModal(self:Fort Subscription Platform Modal Object Reference)
   Event OnModalBackout() {is=title K2Node_Event}
   > n2 = Branch(Condition=IsAnimationPlaying(InAnimation=Get Dismiss() -> (Dismiss:Widget Animation Object Reference)))
   n3 = Get Dismiss()
   PlayAnimationForward(execute=@n2.else, InAnimation=@n3)
   > BindToAnimationFinished(Animation=@n3, Delegate=@n1)
   Event On Activated()
   > PlayAnimationForward(InAnimation=Get ShowUp() -> (ShowUp:Widget Animation Object Reference))
   > SetScrollWidget!(execute, ToScroll:Widget Object Reference=Get PrimaryScrollBox() -> (PrimaryScrollBox:Scroll Box Object Reference)) -> (then)
   n4 = Event SetPlatformGenericEntries!() -> (OutputDelegate, then, Entries:Array of Texts) {is=title K2Node_Event}
   > n5 = Branch(Condition=Get bPlatformSpecificEntriesReceived() -> (bPlatformSpecificEntriesReceived:bool))
   ClearChildren(execute=@n5.else, self=Get VerticalBox_BulletPoints() -> (VerticalBox_BulletPoints:Vertical Box Object Reference))
   > n6 = For Each Loop(Array:Array of Texts=@n4.Entries) -> (Array Element:Text)
   > n7 = FormatBulletPoint!(execute, self:Fort Subscription Crew Bling Info Modal Object Reference, InBulletPoint:"Text (by ref)"=@n6) -> (then, ReturnValue:Text)
   > n8 = Branch(Condition=TextIsEmpty(InText=@n7))
   n9 = Create Widget!(
    execute:Exec=@n8.else
    Class:User Widget Class Reference=/Game/Athena/UI/GenericModalEntry.GenericModalEntry_C
    OwningPlayer:Player Controller Object Reference
    text:Text=@n7
   ) -> (then:Exec, ReturnValue:Generic Modal Entry Object Reference)
   > AddChild(self=Get VerticalBox_BulletPoints(), Content=@n9)
   > ConvertToOldStyle(self:Generic Modal Entry Object Reference=@n9)
   n10 = Event SetPlatformSpecificEntries!() -> (OutputDelegate, then, Entries:Array of Texts) {is=title K2Node_Event}
   > Set bPlatformSpecificEntriesReceived(bPlatformSpecificEntriesReceived:bool=true) -> (Output_Get:bool)
   > ClearChildren(self=Get VerticalBox_BulletPoints())
   > n11 = For Each Loop(Array=@n10.Entries)
   > n12 = FormatBulletPoint(InBulletPoint=@n11)
   > n13 = Branch(Condition=TextIsEmpty(InText=@n12))
   n14 = Create Widget(execute=@n13.else, Class=&GenericModalEntry_C, text=@n12)
   > AddChild(self=Get VerticalBox_BulletPoints(), Content=@n14)
   > ConvertToOldStyle(self=@n14)
   Event PreConstruct()
   > ConvertToOldStyle(self=[
     Get SubscriptionModalEntry_5() -> (SubscriptionModalEntry_5:Generic Modal Entry Object Reference)
     Get SubscriptionModalEntry_3() -> (SubscriptionModalEntry_3:Generic Modal Entry Object Reference)
     Get SubscriptionModalEntry_4() -> (SubscriptionModalEntry_4:Generic Modal Entry Object Reference)
    ])
  ]}]
variables [bPlatformSpecificEntriesReceived {type=boolean}]
widgets {
 info {
  inheritedWidgetCount=2
  parentClass=/Script/FortniteUI.FortSubscriptionCrewBlingInfoModal
  rootWidgetClass=&Overlay
  widgetCount=27
 }
 widgets [
  Overlay_31 {widgetClassPath=&Overlay, widget=Overlay_31}
  BackgroundBlurLightbox {
   widgetClassPath=&BackgroundBlur
   parent=Overlay_31
   widget=BackgroundBlurLightbox
   properties {
    blurStrength=8
    lowQualityFallbackBrush {
     bIsDynamicallyLoaded=false
     drawAs=Image
     imageType=NoImage
     mirroring=NoMirror
     resourceName=None
     resourceObject=null
     tiling=NoTile
     imageSize=<0 0>
     margin=<ltrb 0 0 0 0>
     outlineSettings {
      bUseBrushTransparency=false
      roundingType=HalfHeightRadius
      width=0
      color {colorUseRule=UseColor_Specified, specifiedColor=<rgba 0 0 0 0>}
      cornerRadii=<0 0 0 0>
     }
     tintColor {colorUseRule=UseColor_Specified, specifiedColor=<rgba 0 0 0 0.75>}
     uVRegion {bIsValid=false, max=<0 0>, min=<0 0>}
    }
    padding=<ltrb -5 -5 -5 -5>
    renderTransform {angle=0, scale=<1 1>, shear=<0 0>, translation=<0 -2.002>}
   }
   slot {class=OverlaySlot, properties {horizontalAlignment=HAlign_Fill, verticalAlignment=VAlign_Fill}}
  }
  Lightbox_HitTestBlocker {
   widgetClassPath=&Image
   bIsVariable=true
   parent=BackgroundBlurLightbox
   widget=Lightbox_HitTestBlocker
   properties {brush {
     bIsDynamicallyLoaded=false
     drawAs=Image
     imageType=NoImage
     mirroring=NoMirror
     resourceName=None
     resourceObject=/FortniteCoreUI/Frontend/Art/M_UI_RadialGradientMaster_Blackout
     tiling=NoTile
     imageSize=<32 32>
     margin=<ltrb 0 0 0 0>
     outlineSettings {
      bUseBrushTransparency=false
      roundingType=HalfHeightRadius
      width=0
      color {colorUseRule=UseColor_Specified, specifiedColor=<rgba 0 0 0 0>}
      cornerRadii=<0 0 0 0>
     }
     tintColor {colorUseRule=UseColor_Specified, specifiedColor=<rgba 1 1 1 0.875>}
     uVRegion {bIsValid=false, max=<0 0>, min=<0 0>}
    }}
   slot {class=BackgroundBlurSlot, properties {padding=<ltrb -5 -5 -5 -5>}}
  }
  AnimBoundScalingBorder {
   widgetClassPath=&Border
   parent=Overlay_31
   widget=AnimBoundScalingBorder
   properties {
    background {
     bIsDynamicallyLoaded=false
     drawAs=NoDrawType
     imageType=NoImage
     mirroring=NoMirror
     resourceName=None
     resourceObject=null
     tiling=NoTile
     imageSize=<32 32>
     margin=<ltrb 0 0 0 0>
     outlineSettings {
      bUseBrushTransparency=false
      roundingType=HalfHeightRadius
      width=0
      color {colorUseRule=UseColor_Specified, specifiedColor=<rgba 0 0 0 0>}
      cornerRadii=<0 0 0 0>
     }
     tintColor {colorUseRule=UseColor_Specified, specifiedColor=<rgba 1 1 1 1>}
     uVRegion {bIsValid=false, max=<0 0>, min=<0 0>}
    }
    padding=<ltrb 0 0 0 0>
   }
   slot {class=OverlaySlot, properties {horizontalAlignment=HAlign_Fill, verticalAlignment=VAlign_Center, padding=<ltrb 0 0 0 32>}}
  }
  PrimaryHeightSB {
   widgetClassPath=&SizeBox
   parent=AnimBoundScalingBorder
   widget=PrimaryHeightSB
   properties {bOverride_MaxDesiredHeight=true, bOverride_MinDesiredHeight=true, maxDesiredHeight=704, minDesiredHeight=576}
   slot {class=BorderSlot, properties {padding=<ltrb 0 0 0 0>}}
  }
  Overlay_Modal {widgetClassPath=&Overlay, parent=PrimaryHeightSB, widget=Overlay_Modal, slot {class=SizeBoxSlot}}
  Containment {
   widgetClassPath=&Image
   bIsVariable=true
   parent=Overlay_Modal
   widget=Containment
   properties {
    brush {
     bIsDynamicallyLoaded=false
     drawAs=Image
     imageType=NoImage
     mirroring=NoMirror
     resourceName=None
     resourceObject=/Game/Athena/UI/Subscriptions/Art/M_UI_Subs_Color_FIll_Modals
     tiling=NoTile
     imageSize=<32 32>
     margin=<ltrb 0 0 0 0>
     outlineSettings {
      bUseBrushTransparency=false
      roundingType=HalfHeightRadius
      width=0
      color {colorUseRule=UseColor_Specified, specifiedColor=<rgba 0 0 0 0>}
      cornerRadii=<0 0 0 0>
     }
     tintColor {colorUseRule=UseColor_Specified, specifiedColor=<rgba 1 1 1 1>}
     uVRegion {bIsValid=false, max=<0 0>, min=<0 0>}
    }
    colorAndOpacity=<rgba 1 1 1 0.5>
   }
   slot {class=OverlaySlot, properties {horizontalAlignment=HAlign_Fill, verticalAlignment=VAlign_Fill}}
  }
  Image_BackgroundBottomBar {
   widgetClassPath=&Image
   bIsVariable=true
   parent=Overlay_Modal
   widget=Image_BackgroundBottomBar
   properties {brush {
     bIsDynamicallyLoaded=false
     drawAs=Image
     imageType=NoImage
     mirroring=NoMirror
     resourceName=None
     resourceObject=/FortUILibrary/Foundation/Art_ToolKit/AngledBoxes/M_UI_MoreInfo_BottomDividerLIne_Blue
     tiling=NoTile
     imageSize=<32 26>
     margin=<ltrb 0 0 0 0>
     outlineSettings {
      bUseBrushTransparency=false
      roundingType=HalfHeightRadius
      width=0
      color {colorUseRule=UseColor_Specified, specifiedColor=<rgba 0 0 0 0>}
      cornerRadii=<0 0 0 0>
     }
     tintColor {colorUseRule=UseColor_Specified, specifiedColor=<rgba 1 1 1 1>}
     uVRegion {bIsValid=false, max=<0 0>, min=<0 0>}
    }}
   slot {class=OverlaySlot, properties {horizontalAlignment=HAlign_Fill, verticalAlignment=VAlign_Bottom}}
  }
  ContentSZ {
   widgetClassPath=&SafeZone
   bIsVariable=true
   parent=Overlay_Modal
   widget=ContentSZ
   properties {padBottom=false, padLeft=false, padRight=false, padTop=false}
   slot {class=OverlaySlot, properties {horizontalAlignment=HAlign_Fill, verticalAlignment=VAlign_Fill}}
  }
  Content {widgetClassPath=&HorizontalBox, parent=ContentSZ, widget=Content, slot {class=SafeZoneSlot}}
  Spacer_302 {
   widgetClassPath=&Spacer
   parent=Content
   widget=Spacer_302
   slot {class=HorizontalBoxSlot, properties {horizontalAlignment=HAlign_Left, verticalAlignment=VAlign_Top, size {sizeRule=Fill, value=0.15}}}
  }
  TextContent {
   widgetClassPath=&VerticalBox
   parent=Content
   widget=TextContent
   slot {class=HorizontalBoxSlot, properties {padding=<ltrb 0 0 140 0>, size {sizeRule=Fill, value=1}}}
  }
  PrimaryScrollBox {
   widgetClassPath=&ScrollBox
   bIsVariable=true
   parent=TextContent
   widget=PrimaryScrollBox
   properties {
    widgetBarStyle {
     thickness=16
     draggedThumbImage {
      bIsDynamicallyLoaded=false
      drawAs=Image
      imageType=FullColor
      mirroring=NoMirror
      resourceName=../../../Engine/Content/Slate/Common/Scrollbar_Thumb.png
      resourceObject=/Game/Athena/UI/ReplayBrowser/M_Ui_Skewed_Vfor9slices_BG_Separator_Inst
      tiling=NoTile
      imageSize=<0 0>
      margin=<ltrb 0.25 0.25 0.25 0.25>
      outlineSettings {
       bUseBrushTransparency=false
       roundingType=HalfHeightRadius
       width=0
       color {colorUseRule=UseColor_Specified, specifiedColor=<rgba 0 0 0 0>}
       cornerRadii=<0 0 0 0>
      }
      tintColor {colorUseRule=UseColor_Specified, specifiedColor=<rgba 1 1 1 1>}
      uVRegion {bIsValid=false, max=<0 0>, min=<0 0>}
     }
     horizontalBackgroundImage {
      bIsDynamicallyLoaded=false
      drawAs=Image
      imageType=NoImage
      mirroring=NoMirror
      resourceName=None
      resourceObject=&M_Ui_Skewed_Vfor9slices_BG_Separator_Inst
      tiling=NoTile
      imageSize=<0 0>
      margin=<ltrb 0 0 0 0>
      outlineSettings {
       bUseBrushTransparency=false
       roundingType=HalfHeightRadius
       width=0
       color {colorUseRule=UseColor_Specified, specifiedColor=<rgba 0 0 0 0>}
       cornerRadii=<0 0 0 0>
      }
      tintColor {colorUseRule=UseColor_Specified, specifiedColor=<rgba 0 0.2159 1 0.498>}
      uVRegion {bIsValid=false, max=<0 0>, min=<0 0>}
     }
     horizontalBottomSlotImage {
      bIsDynamicallyLoaded=false
      drawAs=NoDrawType
      imageType=FullColor
      mirroring=NoMirror
      resourceName=../../../Engine/Content/Slate/Common/Scrollbar_Background_Horizontal.png
      resourceObject=null
      tiling=NoTile
      imageSize=<8 8>
      margin=<ltrb 0 0 0 0>
      outlineSettings {
       bUseBrushTransparency=false
       roundingType=HalfHeightRadius
       width=0
       color {colorUseRule=UseColor_Specified, specifiedColor=<rgba 0 0 0 0>}
       cornerRadii=<0 0 0 0>
      }
      tintColor {colorUseRule=UseColor_Specified, specifiedColor=<rgba 1 1 1 1>}
      uVRegion {bIsValid=false, max=<0 0>, min=<0 0>}
     }
     horizontalTopSlotImage {
      bIsDynamicallyLoaded=false
      drawAs=NoDrawType
      imageType=FullColor
      mirroring=NoMirror
      resourceName=../../../Engine/Content/Slate/Common/Scrollbar_Background_Horizontal.png
      resourceObject=null
      tiling=NoTile
      imageSize=<8 8>
      margin=<ltrb 0 0 0 0>
      outlineSettings {
       bUseBrushTransparency=false
       roundingType=HalfHeightRadius
       width=0
       color {colorUseRule=UseColor_Specified, specifiedColor=<rgba 0 0 0 0>}
       cornerRadii=<0 0 0 0>
      }
      tintColor {colorUseRule=UseColor_Specified, specifiedColor=<rgba 1 1 1 1>}
      uVRegion {bIsValid=false, max=<0 0>, min=<0 0>}
     }
     hoveredThumbImage {
      bIsDynamicallyLoaded=false
      drawAs=Image
      imageType=FullColor
      mirroring=NoMirror
      resourceName=../../../Engine/Content/Slate/Common/Scrollbar_Thumb.png
      resourceObject=&M_Ui_Skewed_Vfor9slices_BG_Separator_Inst
      tiling=NoTile
      imageSize=<0 0>
      margin=<ltrb 0.25 0.25 0.25 0.25>
      outlineSettings {
       bUseBrushTransparency=false
       roundingType=HalfHeightRadius
       width=0
       color {colorUseRule=UseColor_Specified, specifiedColor=<rgba 0 0 0 0>}
       cornerRadii=<0 0 0 0>
      }
      tintColor {colorUseRule=UseColor_Specified, specifiedColor=<rgba 0.5573 0.808 1 1>}
      uVRegion {bIsValid=false, max=<0 0>, min=<0 0>}
     }
     normalThumbImage {
      bIsDynamicallyLoaded=false
      drawAs=Image
      imageType=FullColor
      mirroring=NoMirror
      resourceName=../../../Engine/Content/Slate/Common/Scrollbar_Thumb.png
      resourceObject=&M_Ui_Skewed_Vfor9slices_BG_Separator_Inst
      tiling=NoTile
      imageSize=<12 12>
      margin=<ltrb 0.25 0.25 0.25 0.25>
      outlineSettings {
       bUseBrushTransparency=false
       roundingType=HalfHeightRadius
       width=0
       color {colorUseRule=UseColor_Specified, specifiedColor=<rgba 0 0 0 0>}
       cornerRadii=<0 0 0 0>
      }
      tintColor {colorUseRule=UseColor_Specified, specifiedColor=<rgba 0.2122 0.6584 1 1>}
      uVRegion {bIsValid=false, max=<0 0>, min=<0 0>}
     }
     verticalBackgroundImage {
      bIsDynamicallyLoaded=false
      drawAs=Image
      imageType=NoImage
      mirroring=NoMirror
      resourceName=None
      resourceObject=&M_Ui_Skewed_Vfor9slices_BG_Separator_Inst
      tiling=NoTile
      imageSize=<0 0>
      margin=<ltrb 0 0 0 0>
      outlineSettings {
       bUseBrushTransparency=false
       roundingType=HalfHeightRadius
       width=0
       color {colorUseRule=UseColor_Specified, specifiedColor=<rgba 0 0 0 0>}
       cornerRadii=<0 0 0 0>
      }
      tintColor {colorUseRule=UseColor_Specified, specifiedColor=<rgba 0 0.2159 1 0.498>}
      uVRegion {bIsValid=false, max=<0 0>, min=<0 0>}
     }
     verticalBottomSlotImage {
      bIsDynamicallyLoaded=false
      drawAs=NoDrawType
      imageType=FullColor
      mirroring=NoMirror
      resourceName=../../../Engine/Content/Slate/Common/Scrollbar_Background_Horizontal.png
      resourceObject=null
      tiling=NoTile
      imageSize=<8 8>
      margin=<ltrb 0 0 0 0>
      outlineSettings {
       bUseBrushTransparency=false
       roundingType=HalfHeightRadius
       width=0
       color {colorUseRule=UseColor_Specified, specifiedColor=<rgba 0 0 0 0>}
       cornerRadii=<0 0 0 0>
      }
      tintColor {colorUseRule=UseColor_Specified, specifiedColor=<rgba 1 1 1 1>}
      uVRegion {bIsValid=false, max=<0 0>, min=<0 0>}
     }
     verticalTopSlotImage {
      bIsDynamicallyLoaded=false
      drawAs=NoDrawType
      imageType=FullColor
      mirroring=NoMirror
      resourceName=../../../Engine/Content/Slate/Common/Scrollbar_Background_Horizontal.png
      resourceObject=null
      tiling=NoTile
      imageSize=<8 8>
      margin=<ltrb 0 0 0 0>
      outlineSettings {
       bUseBrushTransparency=false
       roundingType=HalfHeightRadius
       width=0
       color {colorUseRule=UseColor_Specified, specifiedColor=<rgba 0 0 0 0>}
       cornerRadii=<0 0 0 0>
      }
      tintColor {colorUseRule=UseColor_Specified, specifiedColor=<rgba 1 1 1 1>}
      uVRegion {bIsValid=false, max=<0 0>, min=<0 0>}
     }
    }
    widgetStyle {
     barThickness=9
     bottomShadowBrush {
      bIsDynamicallyLoaded=false
      drawAs=NoDrawType
      imageType=FullColor
      mirroring=NoMirror
      resourceName=../../../Engine/Content/Slate/Common/ScrollBoxShadowLeft.png
      resourceObject=null
      tiling=NoTile
      imageSize=<8 16>
      margin=<ltrb 1 0.5 0 0.5>
      outlineSettings {
       bUseBrushTransparency=false
       roundingType=HalfHeightRadius
       width=0
       color {colorUseRule=UseColor_Specified, specifiedColor=<rgba 0 0 0 0>}
       cornerRadii=<0 0 0 0>
      }
      tintColor {colorUseRule=UseColor_Specified, specifiedColor=<rgba 1 1 1 1>}
      uVRegion {bIsValid=false, max=<0 0>, min=<0 0>}
     }
     horizontalScrolledContentPadding=<ltrb 0 0 1 0>
     leftShadowBrush {
      bIsDynamicallyLoaded=false
      drawAs=NoDrawType
      imageType=FullColor
      mirroring=NoMirror
      resourceName=../../../Engine/Content/Slate/Common/ScrollBoxShadowLeft.png
      resourceObject=null
      tiling=NoTile
      imageSize=<8 16>
      margin=<ltrb 1 0.5 0 0.5>
      outlineSettings {
       bUseBrushTransparency=false
       roundingType=HalfHeightRadius
       width=0
       color {colorUseRule=UseColor_Specified, specifiedColor=<rgba 0 0 0 0>}
       cornerRadii=<0 0 0 0>
      }
      tintColor {colorUseRule=UseColor_Specified, specifiedColor=<rgba 1 1 1 1>}
      uVRegion {bIsValid=false, max=<0 0>, min=<0 0>}
     }
     rightShadowBrush {
      bIsDynamicallyLoaded=false
      drawAs=NoDrawType
      imageType=FullColor
      mirroring=NoMirror
      resourceName=../../../Engine/Content/Slate/Common/ScrollBoxShadowRight.png
      resourceObject=null
      tiling=NoTile
      imageSize=<8 16>
      margin=<ltrb 0 0.5 1 0.5>
      outlineSettings {
       bUseBrushTransparency=false
       roundingType=HalfHeightRadius
       width=0
       color {colorUseRule=UseColor_Specified, specifiedColor=<rgba 0 0 0 0>}
       cornerRadii=<0 0 0 0>
      }
      tintColor {colorUseRule=UseColor_Specified, specifiedColor=<rgba 1 1 1 1>}
      uVRegion {bIsValid=false, max=<0 0>, min=<0 0>}
     }
     topShadowBrush {
      bIsDynamicallyLoaded=false
      drawAs=NoDrawType
      imageType=FullColor
      mirroring=NoMirror
      resourceName=../../../Engine/Content/Slate/Common/ScrollBoxShadowLeft.png
      resourceObject=null
      tiling=NoTile
      imageSize=<8 16>
      margin=<ltrb 1 0.5 0 0.5>
      outlineSettings {
       bUseBrushTransparency=false
       roundingType=HalfHeightRadius
       width=0
       color {colorUseRule=UseColor_Specified, specifiedColor=<rgba 0 0 0 0>}
       cornerRadii=<0 0 0 0>
      }
      tintColor {colorUseRule=UseColor_Specified, specifiedColor=<rgba 1 1 1 1>}
      uVRegion {bIsValid=false, max=<0 0>, min=<0 0>}
     }
     verticalScrolledContentPadding=<ltrb 0 0 0 1>
    }
   }
   slot {class=VerticalBoxSlot, properties {padding=<ltrb 0 32 0 80>, size {sizeRule=Fill, value=1}}}
  }
  VBox_PlatformSpecific {
   widgetClassPath=&VerticalBox
   bIsVariable=true
   parent=PrimaryScrollBox
   widget=VBox_PlatformSpecific
   slot {class=ScrollBoxSlot, properties {padding=<ltrb 0 32 24 0>}}
  }
  Title_2 {
   widgetClassPath=&CommonTextBlock
   parent=VBox_PlatformSpecific
   widget=Title_2
   properties {
    text=Crew Styles
    textTransformPolicy=ToUpper
    wrapTextAt=1152
    font {
     bForceMonospaced=false
     bMaterialIsStencil=false
     fontMaterial=null
     fontObject=&Burbank
     letterSpacing=0
     monospacedWidth=1
     size=30
     skewAmount=0
     typefaceFontName=BurbankRegular-Black
     outlineSettings {
      bApplyOutlineToDropShadows=false
      bMiteredCorners=false
      bSeparateFillAlpha=false
      outlineMaterial=null
      outlineSize=0
      outlineColor=<rgba 1 1 1 0>
     }
    }
    renderTransform {angle=0, scale=<1 1>, shear=<-12 0>, translation=<0 0>}
    shadowOffset=<0 0>
   }
   slot {class=VerticalBoxSlot, properties {padding=<ltrb 0 0 0 16>}}
  }
  VerticalBox_BulletPoints {
   widgetClassPath=&VerticalBox
   bIsVariable=true
   parent=VBox_PlatformSpecific
   widget=VerticalBox_BulletPoints
   slot {class=VerticalBoxSlot}
  }
  SubscriptionModalEntry_5 {
   widgetClassPath=/Game/Athena/UI/GenericModalEntry
   bIsVariable=true
   parent=VerticalBox_BulletPoints
   widget=SubscriptionModalEntry_5
   properties {text=Bullet point 1}
   slot {class=VerticalBoxSlot}
  }
  SubscriptionModalEntry_3 {
   widgetClassPath=&GenericModalEntry
   bIsVariable=true
   parent=VerticalBox_BulletPoints
   widget=SubscriptionModalEntry_3
   properties {text=Bullet point 2}
   slot {class=VerticalBoxSlot}
  }
  SubscriptionModalEntry_4 {
   widgetClassPath=&GenericModalEntry
   bIsVariable=true
   parent=VerticalBox_BulletPoints
   widget=SubscriptionModalEntry_4
   properties {text=Bullet point 3}
   slot {class=VerticalBoxSlot}
  }
  ImageContent {widgetClassPath=&VerticalBox, parent=Content, widget=ImageContent, slot {class=HorizontalBoxSlot}}
  LogoSB {
   widgetClassPath=&ScaleBox
   parent=ImageContent
   widget=LogoSB
   slot {class=VerticalBoxSlot, properties {verticalAlignment=VAlign_Center, size {sizeRule=Fill, value=1}}}
  }
  WBP_SubscriberSince {
   widgetClassPath=/Crew/UI/Subscription/WBP_SubscriberSince
   bIsVariable=true
   parent=LogoSB
   widget=WBP_SubscriberSince
   slot {class=ScaleBoxSlot, properties {horizontalAlignment=HAlign_Fill}}
  }
  Spacer {
   widgetClassPath=&Spacer
   parent=Content
   widget=Spacer
   slot {class=HorizontalBoxSlot, properties {horizontalAlignment=HAlign_Left, verticalAlignment=VAlign_Top, size {sizeRule=Fill, value=0.15}}}
  }
  ButtonHeightSB {
   widgetClassPath=&SizeBox
   parent=Overlay_Modal
   widget=ButtonHeightSB
   properties {bOverride_MaxDesiredHeight=true, heightOverride=96, maxDesiredHeight=120}
   slot {class=OverlaySlot, properties {horizontalAlignment=HAlign_Center, verticalAlignment=VAlign_Bottom, padding=<ltrb 0 32 0 -44>}}
  }
  Button_Back {
   widgetClassPath=/Game/Athena/UI/GenericModalButton
   bInherited=true
   bIsVariable=true
   parent=ButtonHeightSB
   widget=Button_Back
   properties {
    auto wrap text=false
    button text=Back
    fontSize=24
    fontSizeOverride=true
    material Brush=/Game/Athena/UI/Frontend/Spatial/Art/M_Ui_Skewed_SpatialButton_MoreInfo_Blue
    minWidth=192
    onButtonBaseClicked=(&self:WidgetTree.Button_Back.HandleOnButtonClicked)
    onButtonBaseDoubleClicked=(&self:WidgetTree.Button_Back.HandleOnButtonDoubleClicked)
    onButtonBaseHovered=(&self:WidgetTree.Button_Back.HandleOnButtonHovered)
    onButtonBaseUnhovered=(&self:WidgetTree.Button_Back.HandleOnButtonUnhovered)
    onSelectedChangedBase=(&self:WidgetTree.Button_Back.HandleOnSelectedChanged)
    right Image=null
    shearText=true
    simpleHover=true
    hoveredSlateSoundOverride {resourceObject=/Game/Sounds/Fort_UI_Sounds/FrontEnd/DefaultSounds/PressedButtons/MSS_UI_Button_Hover_Default}
    pressedSlateSoundOverride {resourceObject=/Game/Sounds/Fort_UI_Sounds/FrontEnd/DefaultSounds/PressedButtons/MSS_UI_Button_Pressed_Default}
    triggeringInputAction {dataTable=&InputActionDataTable, rowName=GenericBack}
   }
   slot {class=SizeBoxSlot}
  }
  Text_FinePrint {
   widgetClassPath=&CommonTextBlock
   parent=Overlay_31
   widget=Text_FinePrint
   properties {
    mobileFontSizeMultiplier=1.2
    text=Crew Styles are a bonus offer for the Fortnite Crew subscription product that may be re-featured at later dates.
    wrapTextAt=1664
    colorAndOpacity {colorUseRule=UseColor_Specified, specifiedColor=<rgba 0.2122 0.6584 1 1>}
    font {
     bForceMonospaced=false
     bMaterialIsStencil=false
     fontMaterial=null
     fontObject=&BurbankSmall
     letterSpacing=0
     monospacedWidth=1
     size=15
     skewAmount=0
     typefaceFontName=Bold
     outlineSettings {
      bApplyOutlineToDropShadows=false
      bMiteredCorners=false
      bSeparateFillAlpha=false
      outlineMaterial=null
      outlineSize=0
      outlineColor=<rgba 0 0 0 1>
     }
    }
    renderTransform {angle=0, scale=<1 1>, shear=<-8 0>, translation=<0 0>}
    shadowOffset=<0 0>
   }
   slot {class=OverlaySlot, properties {verticalAlignment=VAlign_Bottom, padding=<ltrb 128 0 0 100>}}
  }
  Button_TouchBack {widgetClassPath=&CommonButtonLegacy, bInherited=true, widget=null}
 ]
}
