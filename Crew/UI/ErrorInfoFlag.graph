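# ErrorInfoFlag is a UMG message banner with a dark background, sheared accent strip, optional icon and right-justified rich text.
# PreConstruct applies AccentColor and IconMaterial, collapsing the icon when no material is set.
# Outside design time it also calls UpdateText with empty text, so the whole widget stays collapsed until UpdateText receives a message.
# Construct enlarges the content's ScaleBox slightly on mobile.
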WidgetBlueprint /Crew/UI/ErrorInfoFlag
graphs [
 EventGraph {nodes [
   n1 = Event PreConstruct()
   > UpdateAccentColor!(execute, self, Color:Linear Color Structure=Get AccentColor() -> (AccentColor:Linear Color Structure)) -> (then)
   > UpdateIconMaterial!(execute, self, Material:Material Interface Object Reference=Get IconMaterial() -> (IconMaterial:Material Interface Object Reference)) -> (then)
   > n2 = Branch(Condition=@n1.IsDesignTime)
   UpdateText!(execute=@n2.else, self, Text:Text) -> (then)
   Event Construct()
   > Branch(Condition=IsMobileGame())
   > SetUserSpecifiedScale!(
    execute
    self:Scale Box Object Reference=Get ScaleBox_Mobile() -> (ScaleBox_Mobile:Scale Box Object Reference)
    InUserSpecifiedScale:float=1.12
   ) -> (then)
   Event Tick()
  ]}
 UpdateAccentColor {nodes [
   n1 = UpdateAccentColor!() -> (then, Color:Linear Color Structure) {is=title K2Node_FunctionEntry}
   > SetColorAndOpacity(
    self:Image Object Reference=Get Image_ErrorBG_Accent() -> (Image_ErrorBG_Accent:Image Object Reference)
    InColorAndOpacity:Linear Color Structure=@n1
   )
  ]}
 UpdateIconMaterial {nodes [
   UpdateIconMaterial!() -> (then, Material:Material Interface Object Reference) {is=title K2Node_FunctionEntry}
   > n1 = Get!(execute:Exec) -> (then:Exec, else:Exec, Material:Material Interface Object Reference)
   n2 = Get OptionalIcon() -> (OptionalIcon:Image Object Reference)
   > SetBrushFromMaterial(self=@n2, Material=@n1)
   > SetVisibility(self=@n2, InVisibility=HitTestInvisible)
   SetVisibility(execute=@n1.else, self=@n2, InVisibility=Collapsed)
  ]}
 UpdateText {nodes [
   n1 = UpdateText!() -> (then, Text:Text) {is=title K2Node_FunctionEntry}
   > SetText(self=Get RichText_UserInformation() -> (RichText_UserInformation:Rich Text Block Object Reference), InText=@n1)
   > SetVisibility(InVisibility=Select(Option 0=HitTestInvisible, Option 1=Collapsed, Index=TextIsEmpty(InText=@n1)))
  ]}
]
variables [
 AccentColor {type=LinearColor, default=<rgba 0 0.4077 1 1>}
 IconMaterial {type=MaterialInterface, default=null}
]
widgets {
 info {inheritedWidgetCount=0, parentClass=&UserWidget, rootWidgetClass=&Overlay, widgetCount=8}
 widgets [
  Overlay_UserInformation {widgetClassPath=&Overlay, widget=Overlay_UserInformation}
  Image_ErrorBG {
   widgetClassPath=&Image
   bIsVariable=true
   parent=Overlay_UserInformation
   widget=Image_ErrorBG
   properties {
    brush {
     bIsDynamicallyLoaded=false
     drawAs=Box
     imageType=NoImage
     mirroring=NoMirror
     resourceName=None
     resourceObject=/Game/Athena/HUD/Materials/M_UI_MessageBG_Inst
     tiling=NoTile
     imageSize=<64 64>
     margin=<ltrb 0.4 0 0.4 0>
     outlineSettings {
      bUseBrushTransparency=false
      roundingType=HalfHeightRadius
      width=0
      color {colorUseRule=UseColor_Specified, specifiedColor=<rgba 0 0 0 0>}
      cornerRadii=<0 0 0 0>
     }
     tintColor {colorUseRule=UseColor_Specified, specifiedColor=<rgba 1 1 1 1>}
     uVRegion {bIsValid=false, max=<0 0>, min=<0 0>}
    }
    colorAndOpacity=<rgba 0 0 0 0.75>
   }
   slot {class=OverlaySlot, properties {horizontalAlignment=HAlign_Fill, verticalAlignment=VAlign_Fill}}
  }
  SizeBox_0 {
   widgetClassPath=&SizeBox
   parent=Overlay_UserInformation
   widget=SizeBox_0
   properties {bOverride_WidthOverride=true, widthOverride=18}
   slot {class=OverlaySlot, properties {horizontalAlignment=HAlign_Right, verticalAlignment=VAlign_Fill}}
  }
  Image_ErrorBG_Accent {
   widgetClassPath=&Image
   bIsVariable=true
   parent=SizeBox_0
   widget=Image_ErrorBG_Accent
   properties {
    brush {
     bIsDynamicallyLoaded=false
     drawAs=Box
     imageType=NoImage
     mirroring=NoMirror
     resourceName=None
     resourceObject=&M_UI_MessageBG_Inst
     tiling=NoTile
     imageSize=<64 64>
     margin=<ltrb 0.25 0 0.25 0>
     outlineSettings {
      bUseBrushTransparency=false
      roundingType=HalfHeightRadius
      width=0
      color {colorUseRule=UseColor_Specified, specifiedColor=<rgba 0 0 0 0>}
      cornerRadii=<0 0 0 0>
     }
     tintColor {colorUseRule=UseColor_Specified, specifiedColor=<rgba 1 1 1 1>}
     uVRegion {bIsValid=false, max=<0 0>, min=<0 0>}
    }
    colorAndOpacity=<rgba 0 0.4077 1 1>
    renderTransform {angle=0, scale=<1 1>, shear=<-4 0>, translation=<0 0>}
   }
   slot {class=SizeBoxSlot}
  }
  ScaleBox_Mobile {
   widgetClassPath=&ScaleBox
   bIsVariable=true
   parent=Overlay_UserInformation
   widget=ScaleBox_Mobile
   properties {stretch=UserSpecified}
   slot {class=OverlaySlot, properties {horizontalAlignment=HAlign_Right, padding=<ltrb 25 16 45 16>}}
  }
  HorizontalBox_0 {widgetClassPath=&HorizontalBox, parent=ScaleBox_Mobile, widget=HorizontalBox_0, slot {class=ScaleBoxSlot}}
  OptionalIcon {
   widgetClassPath=&Image
   bIsVariable=true
   parent=HorizontalBox_0
   widget=OptionalIcon
   properties {brush {
     bIsDynamicallyLoaded=false
     drawAs=Image
     imageType=NoImage
     mirroring=NoMirror
     resourceName=None
     resourceObject=null
     tiling=NoTile
     imageSize=<64 64>
     margin=<ltrb 0 0 0 0>
     outlineSettings {
      bUseBrushTransparency=false
      roundingType=HalfHeightRadius
      width=0
      color {colorUseRule=UseColor_Specified, specifiedColor=<rgba 0 0 0 0>}
      cornerRadii=<0 0 0 0>
     }
     tintColor {colorUseRule=UseColor_Specified, specifiedColor=<rgba 1 1 1 1>}
     uVRegion {bIsValid=false, max=<0 0>, min=<0 0>}
    }}
   slot {class=HorizontalBoxSlot, properties {horizontalAlignment=HAlign_Left, verticalAlignment=VAlign_Center, padding=<ltrb 0 0 10 0>}}
  }
  RichText_UserInformation {
   widgetClassPath=/Script/UMG.RichTextBlock
   bIsVariable=true
   parent=HorizontalBox_0
   widget=RichText_UserInformation
   properties {
    bOverrideDefaultStyle=true
    justification=Right
    text="[TBD] Message goes here. Collapses by default at runtime via UpdateText()"
    wrapTextAt=560
    defaultTextStyleOverride {
     overflowPolicy=Clip
     transformPolicy=None
     colorAndOpacity {colorUseRule=UseColor_Specified, specifiedColor=<rgba 1 1 1 1>}
     font {
      bForceMonospaced=false
      bMaterialIsStencil=false
      fontMaterial=null
      fontObject=&BurbankSmall
      letterSpacing=0
      monospacedWidth=1
      size=18
      skewAmount=0
      typefaceFontName=Bold
      outlineSettings {
       bApplyOutlineToDropShadows=false
       bMiteredCorners=false
       bSeparateFillAlpha=false
       outlineMaterial=null
       outlineSize=0
       outlineColor=<rgba 0 0 0 1>
      }
     }
     highlightColor {colorUseRule=UseColor_Specified, specifiedColor=<rgba 0 0 0 1>}
     highlightShape {
      bIsDynamicallyLoaded=false
      drawAs=Image
      imageType=NoImage
      mirroring=NoMirror
      resourceName=None
      resourceObject=null
      tiling=NoTile
      imageSize=<32 32>
      margin=<ltrb 0 0 0 0>
      outlineSettings {
       bUseBrushTransparency=false
       roundingType=HalfHeightRadius
       width=0
       color {colorUseRule=UseColor_Specified, specifiedColor=<rgba 0 0 0 0>}
       cornerRadii=<0 0 0 0>
      }
      tintColor {colorUseRule=UseColor_Specified, specifiedColor=<rgba 1 1 1 1>}
      uVRegion {bIsValid=false, max=<0 0>, min=<0 0>}
     }
     selectedBackgroundColor {colorUseRule=UseColor_Foreground, specifiedColor=<rgba 1 0 1 1>}
     shadowColorAndOpacity=<rgba 0 0 0 1>
     shadowOffset=<0 0>
     strikeBrush {
      bIsDynamicallyLoaded=false
      drawAs=Image
      imageType=NoImage
      mirroring=NoMirror
      resourceName=None
      resourceObject=null
      tiling=NoTile
      imageSize=<32 32>
      margin=<ltrb 0 0 0 0>
      outlineSettings {
       bUseBrushTransparency=false
       roundingType=HalfHeightRadius
       width=0
       color {colorUseRule=UseColor_Specified, specifiedColor=<rgba 0 0 0 0>}
       cornerRadii=<0 0 0 0>
      }
      tintColor {colorUseRule=UseColor_Specified, specifiedColor=<rgba 1 1 1 1>}
      uVRegion {bIsValid=false, max=<0 0>, min=<0 0>}
     }
     underlineBrush {
      bIsDynamicallyLoaded=false
      drawAs=Image
      imageType=NoImage
      mirroring=NoMirror
      resourceName=None
      resourceObject=null
      tiling=NoTile
      imageSize=<32 32>
      margin=<ltrb 0 0 0 0>
      outlineSettings {
       bUseBrushTransparency=false
       roundingType=HalfHeightRadius
       width=0
       color {colorUseRule=UseColor_Specified, specifiedColor=<rgba 0 0 0 0>}
       cornerRadii=<0 0 0 0>
      }
      tintColor {colorUseRule=UseColor_Specified, specifiedColor=<rgba 1 1 1 1>}
      uVRegion {bIsValid=false, max=<0 0>, min=<0 0>}
     }
    }
    renderTransform {angle=0, scale=<1 1>, shear=<-8 0>, translation=<0 0>}
   }
   slot {class=HorizontalBoxSlot}
  }
 ]
}
