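# Slim vertical pip button for progressive cosmetic UI, derived from FortTextButton with no input action widget.
# Selection, via button events or change-guarded Set Selected, plays an animation forward or reversed, growing the pip and driving material parameters.
# Update Subscribed passes gold or blue color sets to the pip's dynamic material through Set Pip Colors.
# On construct, a shadow image shows only on mobile with the Bacchus frontend enabled.
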WidgetBlueprint /Crew/UI/ProgressiveCosmetic/NoInputActionPipButton
animations [
 Selected {
  binding_count=3
  length_seconds=0.1
  object=&self:Selected
  sequence {bindings [
    AnimBoundSizeBox {
     id=#1
     possessed_class=SizeBox
     tracks [{
       type=MovieSceneFloatTrack
       display_name=Height Override
       sections [{type=MovieSceneFloatSection, channels [None {type=MovieSceneScriptingFloatChannel, keys [[0,80],[2,128]]}]}]
      }]
    }
    Text_Content {
     id=#2
     possessed_class=CommonTextBlock
     tracks [
      {
       type=MovieSceneWidgetMaterialTrack
       display_name=Font.FontMaterial
       sections [{type=MovieSceneParameterSection, channels [Hover {type=MovieSceneScriptingFloatChannel, keys [[0,0],[2,1]]}]}]
      }
      {
       type=MovieScene2DTransformTrack
       sections [{
         type=MovieScene2DTransformSection
         channels [
          Translation.X {type=MovieSceneScriptingFloatChannel, keys [[0,0],[2,8]]}
          Scale.X {type=MovieSceneScriptingFloatChannel, keys [[0,0.75],[1.9997,1]]}
          Scale.Y {type=MovieSceneScriptingFloatChannel, keys [[0,0.75],[1.9997,1]]}
         ]
        }]
      }
     ]
    }
    Pip {
     id=#3
     possessed_class=Image
     tracks [{
       type=MovieSceneWidgetMaterialTrack
       sections [{
         type=MovieSceneParameterSection
         channels [
          Selected {type=MovieSceneScriptingFloatChannel, keys [[0,0],[2,1]]}
          Hover {type=MovieSceneScriptingFloatChannel, const [0,1]}
         ]
        }]
      }]
    }
   ]}
 }
 Hover {
  binding_count=1
  length_seconds=0.1
  object=&self:Hover
  sequence {bindings [Pip {
     id=#4
     possessed_class=Image
     tracks [{
       type=MovieSceneWidgetMaterialTrack
       sections [{type=MovieSceneParameterSection, channels [Hover {type=MovieSceneScriptingFloatChannel, keys [[0,0],[2,1]]}]}]
      }]
    }]}
 }
]
graphs [
 EventGraph {nodes [
   n1 = Set Pip Colors!() -> (OutputDelegate, then, Base Color:Linear Color Structure, Hover Color:Linear Color Structure, Select Color:Linear Color Structure) {is=title K2Node_CustomEvent}
   > n2 = GetDynamicMaterial(self=Get Pip() -> (Pip:Image Object Reference))
   > SetVectorParameterValue(self=@n2, ParameterName=ColorA, Value=@n1.Base Color)
   > SetVectorParameterValue(self=@n2, ParameterName=ColorB, Value=@n1.Hover Color)
   > SetVectorParameterValue(self=@n2, ParameterName=ColorC, Value=@n1.Select Color)
   Event Construct()
   n3 = IsMobileGame()
   n4 = Get bBacchusFrontendEnabled!(self:Fort Runtime Options Object Reference=GetRuntimeOptions()) -> (bBacchusFrontendEnabled:bool)
   > SetVisibility(
    self=Get BacchusMobileShadow() -> (BacchusMobileShadow:Image Object Reference)
    InVisibility=Select(Option 0=Collapsed, Option 1=SelfHitTestInvisible, Index=AND Boolean(A=@n3, B=@n4))
   )
   Event On Selected()
   n5 = Get Selected() -> (Selected:Widget Animation Object Reference)
   > PlayAnimationForward(InAnimation=@n5)
   Event On Deselected()
   > PlayAnimationReverse(InAnimation=@n5)
  ]}
 Set Selected {nodes [
   n1 = Set Selected!() -> (then, Selected:bool=~) {is=title K2Node_FunctionEntry}
   > Branch(Condition="Not Equal (Boolean)"(A=@n1, B=Get Was Selected() -> (Was Selected:bool)) -> (ReturnValue=~))
   > n2 = Set Was Selected(Was Selected:bool=@n1) -> (Output_Get:bool)
   > StopAllAnimations(self:User Widget Object Reference)
   > n3 = Branch(Condition=@n2)
   n4 = Get Selected()
   > PlayAnimationForward(InAnimation=@n4)
   PlayAnimationReverse(execute=@n3.else, InAnimation=@n4)
  ]}
 Update Subscribed {nodes [
   n1 = Update Subscribed!() -> (then, Subscribed:bool=~) {is=title K2Node_FunctionEntry}
   > Set Pip Colors!(
    execute
    self
    Base Color:Linear Color Structure=SelectColor(A=(R=0.147,G=0.0284,B=0.0044,A=1), B=(R=0,G=0.2502,B=1,A=1), bPickA=@n1)
    Hover Color:Linear Color Structure=SelectColor(A=(R=0.974,G=0.365,B=0.1392,A=1), B=(R=0,G=0.5029,B=1,A=1), bPickA=@n1)
    Select Color:Linear Color Structure=SelectColor(A=(R=0.8963,G=0.4969,B=0.1329,A=1), B=(R=0.5029,G=1,B=1,A=1), bPickA=@n1)
   ) -> (then) {is=title K2Node_CallFunction}
  ]}
]
variables [Was Selected {type=boolean}]
widgets {
 info {inheritedWidgetCount=2, parentClass=/Script/FortniteUI.FortTextButton, rootWidgetClass=&SizeBox, widgetCount=8}
 widgets [
  SizeBox_1 {widgetClassPath=&SizeBox, widget=SizeBox_1, properties {bOverride_WidthOverride=true, widthOverride=88}}
  HorizontalBox_0 {widgetClassPath=&HorizontalBox, parent=SizeBox_1, widget=HorizontalBox_0, slot {class=SizeBoxSlot}}
  AnimBoundSizeBox {
   widgetClassPath=&SizeBox
   parent=HorizontalBox_0
   widget=AnimBoundSizeBox
   properties {bOverride_HeightOverride=true, bOverride_WidthOverride=true, heightOverride=80, widthOverride=8}
   slot {class=HorizontalBoxSlot, properties {horizontalAlignment=HAlign_Left}}
  }
  Overlay_4 {widgetClassPath=&Overlay, parent=AnimBoundSizeBox, widget=Overlay_4, slot {class=SizeBoxSlot}}
  BacchusMobileShadow {
   widgetClassPath=&Image
   bIsVariable=true
   parent=Overlay_4
   widget=BacchusMobileShadow
   properties {
    visibility=Collapsed
    brush {
     bIsDynamicallyLoaded=false
     drawAs=Image
     imageType=NoImage
     mirroring=NoMirror
     resourceName=None
     resourceObject=/Game/UI/Frontend/ItemShop/Art/M_UI_CatabaBacchusNipsShadow
     tiling=NoTile
     imageSize=<30 30>
     margin=<ltrb 0 0.5 0 0.5>
     outlineSettings {
      bUseBrushTransparency=false
      roundingType=HalfHeightRadius
      width=0
      color {colorUseRule=UseColor_Specified, specifiedColor=<rgba 0 0 0 0>}
      cornerRadii=<0 0 0 0>
     }
     tintColor {colorUseRule=UseColor_Specified, specifiedColor=<rgba 1 1 1 1>}
     uVRegion {bIsValid=false, max=<0 0>, min=<0 0>}
    }
   }
   slot {class=OverlaySlot, properties {verticalAlignment=VAlign_Fill, padding=<ltrb 0 0 -6 0>}}
  }
  Pip {
   widgetClassPath=&Image
   bIsVariable=true
   parent=Overlay_4
   widget=Pip
   properties {brush {
     bIsDynamicallyLoaded=false
     drawAs=Box
     imageType=NoImage
     mirroring=NoMirror
     resourceName=None
     resourceObject=/Game/UI/Frontend/ItemShop/Art/M_UI_ItemShop_SectionPip_Base
     tiling=NoTile
     imageSize=<8 88>
     margin=<ltrb 0 0.5 0 0.5>
     outlineSettings {
      bUseBrushTransparency=false
      roundingType=HalfHeightRadius
      width=0
      color {colorUseRule=UseColor_Specified, specifiedColor=<rgba 0 0 0 0>}
      cornerRadii=<0 0 0 0>
     }
     tintColor {colorUseRule=UseColor_Specified, specifiedColor=<rgba 1 1 1 1>}
     uVRegion {bIsValid=false, max=<0 0>, min=<0 0>}
    }}
   slot {class=OverlaySlot, properties {verticalAlignment=VAlign_Fill}}
  }
  Text_Content {
   widgetClassPath=&CommonTextBlock
   bInherited=true
   parent=HorizontalBox_0
   widget=Text_Content
   properties {
    mobileFontSizeMultiplier=1.34
    renderOpacity=0
    text="##"
    font {
     bForceMonospaced=false
     bMaterialIsStencil=false
     fontMaterial=/Game/UI/Frontend/ItemShop/Art/M_UI_ItemShop_SectionIndicator_Text
     fontObject=&Burbank
     letterSpacing=0
     monospacedWidth=1
     size=24
     skewAmount=0
     typefaceFontName=BurbankRegular-Black
     outlineSettings {
      bApplyOutlineToDropShadows=false
      bMiteredCorners=false
      bSeparateFillAlpha=true
      outlineMaterial=null
      outlineSize=3
      outlineColor=<rgba 0 0.0004 0.005 0.025>
     }
    }
    renderTransform {angle=0, scale=<0.75 0.75>, shear=<0 0>, translation=<0 0>}
    renderTransformPivot=<0 0.5>
    shadowOffset=<0 0>
   }
   slot {class=HorizontalBoxSlot, properties {verticalAlignment=VAlign_Center, padding=<ltrb 10 0 0 0>}}
  }
  InputActionWidget {widgetClassPath=&CommonActionWidget, bInherited=true, widget=null}
 ]
}
