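# Progressive cosmetic header banner widget for Crew UI: a skewed strip with a lock icon and rich text.
# Update Banner State tints the border by subscription, shows the unlocked icon when subscribed or sets are complete, and calls Update Text, which collapses the banner when text is empty.
# Get Banner Text returns an all-sets-completed message or CrewVM's new-stage unlock fine print.
# PreConstruct previews debug text at design time only.
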WidgetBlueprint /Crew/UI/ProgressiveCosmetic/ProgressiveHeaderBanner
graphs [
 EventGraph {nodes [
   n1 = Event PreConstruct()
   > Branch(Condition=@n1.IsDesignTime)
   > Update Banner State!(execute, self, Text:Text=Get Debug Text Preview() -> (Debug Text Preview:Text), Current Sets Complete:bool=false, Subscribed:bool=false) -> (then)
  ]}
 Get Banner Text {nodes [
   n1 = Get Banner Text!() -> (then, All Complete:bool=~) {is=title K2Node_FunctionEntry}
   n2 = Get CrewVM() -> (CrewVM:Crew VM Object Reference)
   > Return Node!(
    execute
    Result:Text=SelectText(
     A="LOCTEXT(\"All available sets are completed!\")"
     B=Get ProgressiveInfoNewStagesUnlockFinePrint!(self:Crew VM Object Reference=@n2) -> (ProgressiveInfoNewStagesUnlockFinePrint:Text)
     bPickA=@n1
    )
   )
  ]}
 Get Image {nodes [
   n1 = Get Image!() -> (then, Subscribed:bool=~, All Sets Complete:bool=~) {is=title K2Node_FunctionEntry}
   > Return Node!(
    execute
    ReturnValue:Texture 2D Object Reference=Select(
     Option 0:Texture 2D Object Reference=/Game/UI/Foundation/Textures/Icons/Locks/T-Icon-Lock-64
     Option 1:Texture 2D Object Reference=/Game/UI/Foundation/Textures/Icons/Locks/T-Icon-Unlocked-64
     Index=OR Boolean(A=@n1, B=@n1.All Sets Complete)
    ) -> (ReturnValue:Texture 2D Object Reference)
   )
  ]}
 Update Banner State {nodes [
   n1 = Update Banner State!() -> (then, Text:Text, Current Sets Complete:bool=~, Subscribed:bool=~) {is=title K2Node_FunctionEntry}
   > SetBrushColor(
    self=Get Banner_Border() -> (Banner_Border:Border Object Reference)
    InBrushColor=SelectColor(
     A=Get Subscribed() -> (Subscribed:Linear Color Structure)
     B=Get Unsubscribed() -> (Unsubscribed:Linear Color Structure)
     bPickA=@n1.Subscribed
    )
   )
   > SetBrushFromTexture!(
    execute
    self:Image Object Reference=Get Image() -> (Image:Image Object Reference) {is=title K2Node_VariableGet}
    Texture:Texture 2D Object Reference=Get Image!(
     self
     Subscribed:bool=Get Subscribed() -> (Subscribed:bool)
     All Sets Complete:bool=Get Current Sets Complete() -> (Current Sets Complete:bool)
    ) -> (ReturnValue:Texture 2D Object Reference) {is=title K2Node_CallFunction}
    bMatchSize:bool=false
   ) -> (then)
   > Update Text!(execute, self, InText:Text=Get Text!() -> (Text)) -> (then)
  ]}
 Update Text {nodes [
   n1 = Update Text!() -> (then, InText:Text) {is=title K2Node_FunctionEntry}
   > SetText(self=Get Banner_Text() -> (Banner_Text:Common Rich Text Block Object Reference), InText=@n1)
   > SetVisibility(InVisibility=Select(Option 0:ESlateVisibility Enum=HitTestInvisible, Option 1:ESlateVisibility Enum=Collapsed, Index=TextIsEmpty(InText=@n1)) -> (ReturnValue:ESlateVisibility Enum))
  ]}
]
variables [
 Debug Text Preview {type=string, default=Text goes here or the banner autohides}
 Subscribed {type=LinearColor, default=<rgba 0.147 0.0284 0.0044 1>}
 Unsubscribed {type=LinearColor, default=<rgba 0.6654 0 0.1301 1>}
]
widgets {
 info {inheritedWidgetCount=0, parentClass=&UserWidget, rootWidgetClass=&Border, widgetCount=4}
 widgets [
  Banner_Border {
   widgetClassPath=&Border
   bIsVariable=true
   widget=Banner_Border
   properties {
    background {
     bIsDynamicallyLoaded=false
     drawAs=Box
     imageType=NoImage
     mirroring=NoMirror
     resourceName=None
     resourceObject=/Game/Athena/UI/ReplayBrowser/M_Ui_Skewed_NoV1
     tiling=NoTile
     imageSize=<32 32>
     margin=<ltrb 0.25 0.25 0.25 0.25>
     outlineSettings {
      bUseBrushTransparency=false
      roundingType=HalfHeightRadius
      width=0
      color {colorUseRule=UseColor_Specified, specifiedColor=<rgba 0 0 0 0>}
      cornerRadii=<0 0 0 0>
     }
     tintColor {colorUseRule=UseColor_Specified, specifiedColor=<rgba 1 1 1 1>}
     uVRegion {bIsValid=false, max=<0 0>, min=<0 0>}
    }
    brushColor=<rgba 0.6654 0 0.1301 1>
    padding=<ltrb 3 4 8 4>
   }
  }
  HorizontalBox_1 {
   widgetClassPath=&HorizontalBox
   parent=Banner_Border
   widget=HorizontalBox_1
   slot {class=BorderSlot, properties {padding=<ltrb 3 4 8 4>}}
  }
  Image {
   widgetClassPath=&Image
   bIsVariable=true
   parent=HorizontalBox_1
   widget=Image
   properties {
    brush {
     bIsDynamicallyLoaded=false
     drawAs=Image
     imageType=NoImage
     mirroring=NoMirror
     resourceName=None
     resourceObject=&T-Icon-Lock-64
     tiling=NoTile
     imageSize=<36 36>
     margin=<ltrb 0 0 0 0>
     outlineSettings {
      bUseBrushTransparency=false
      roundingType=HalfHeightRadius
      width=0
      color {colorUseRule=UseColor_Specified, specifiedColor=<rgba 0 0 0 0>}
      cornerRadii=<0 0 0 0>
     }
     tintColor {colorUseRule=UseColor_Specified, specifiedColor=<rgba 1 1 1 1>}
     uVRegion {bIsValid=false, max=<0 0>, min=<0 0>}
    }
    renderTransform {angle=-7, scale=<1 1>, shear=<0 0>, translation=<0 0>}
   }
   slot {class=HorizontalBoxSlot, properties {verticalAlignment=VAlign_Bottom, padding=<ltrb 0 -16 4 0>}}
  }
  Banner_Text {
   widgetClassPath=&CommonRichTextBlock
   bIsVariable=true
   parent=HorizontalBox_1
   widget=Banner_Text
   properties {
    bOverrideDefaultStyle=true
    mobileTextBlockScale=1.2
    text=Text goes here or the banner autohides
    defaultTextStyleOverride {
     overflowPolicy=Clip
     transformPolicy=None
     colorAndOpacity {colorUseRule=UseColor_Specified, specifiedColor=<rgba 1 1 1 1>}
     font {
      bForceMonospaced=false
      bMaterialIsStencil=false
      fontMaterial=null
      fontObject=&BurbankSmall
      letterSpacing=0
      monospacedWidth=1
      size=14
      skewAmount=0
      typefaceFontName=Bold
      outlineSettings {
       bApplyOutlineToDropShadows=false
       bMiteredCorners=false
       bSeparateFillAlpha=false
       outlineMaterial=null
       outlineSize=0
       outlineColor=<rgba 0 0 0 1>
      }
     }
     highlightColor {colorUseRule=UseColor_Specified, specifiedColor=<rgba 0 0 0 1>}
     highlightShape {
      bIsDynamicallyLoaded=false
      drawAs=Image
      imageType=NoImage
      mirroring=NoMirror
      resourceName=None
      resourceObject=null
      tiling=NoTile
      imageSize=<32 32>
      margin=<ltrb 0 0 0 0>
      outlineSettings {
       bUseBrushTransparency=false
       roundingType=HalfHeightRadius
       width=0
       color {colorUseRule=UseColor_Specified, specifiedColor=<rgba 0 0 0 0>}
       cornerRadii=<0 0 0 0>
      }
      tintColor {colorUseRule=UseColor_Specified, specifiedColor=<rgba 1 1 1 1>}
      uVRegion {bIsValid=false, max=<0 0>, min=<0 0>}
     }
     selectedBackgroundColor {colorUseRule=UseColor_Foreground, specifiedColor=<rgba 1 0 1 1>}
     shadowColorAndOpacity=<rgba 0 0 0 1>
     shadowOffset=<0 0>
     strikeBrush {
      bIsDynamicallyLoaded=false
      drawAs=Image
      imageType=NoImage
      mirroring=NoMirror
      resourceName=None
      resourceObject=null
      tiling=NoTile
      imageSize=<32 32>
      margin=<ltrb 0 0 0 0>
      outlineSettings {
       bUseBrushTransparency=false
       roundingType=HalfHeightRadius
       width=0
       color {colorUseRule=UseColor_Specified, specifiedColor=<rgba 0 0 0 0>}
       cornerRadii=<0 0 0 0>
      }
      tintColor {colorUseRule=UseColor_Specified, specifiedColor=<rgba 1 1 1 1>}
      uVRegion {bIsValid=false, max=<0 0>, min=<0 0>}
     }
     underlineBrush {
      bIsDynamicallyLoaded=false
      drawAs=Image
      imageType=NoImage
      mirroring=NoMirror
      resourceName=None
      resourceObject=null
      tiling=NoTile
      imageSize=<32 32>
      margin=<ltrb 0 0 0 0>
      outlineSettings {
       bUseBrushTransparency=false
       roundingType=HalfHeightRadius
       width=0
       color {colorUseRule=UseColor_Specified, specifiedColor=<rgba 0 0 0 0>}
       cornerRadii=<0 0 0 0>
      }
      tintColor {colorUseRule=UseColor_Specified, specifiedColor=<rgba 1 1 1 1>}
      uVRegion {bIsValid=false, max=<0 0>, min=<0 0>}
     }
    }
    renderTransform {angle=0, scale=<1 1>, shear=<-8 0>, translation=<0 0>}
   }
   slot {class=HorizontalBoxSlot}
  }
 ]
}
