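# Battle pass reward tag widget showing an item's rarity pill beside its type label, built on the native AthenaRewardItemTypeRarityTag class.
# Init events, with or without item series, feed palette colors into the pill material's BackingColor and BorderColor parameters.
# The border color also tints the rarity text, while ItemTypeTextColor tints the type label.
# OverrideTypeTextColor sets the type text color directly and resets its parent border tint to white.
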WidgetBlueprint /Game/Athena/UI/Season/AthenaRewardItemTypeRarityTag
graphs [
 EventGraph {nodes [
   n1 = Event OnInitializeRarityWithSeries!() -> (OutputDelegate, then, SeriesData:Fort Item Series Definition Object Reference) {is=title K2Node_Event}
   > n2 = GetDynamicMaterial(self:Border Object Reference=Get Border_ItemRarity() -> (Border_ItemRarity:Border Object Reference))
   n3 = Break Fort Color Palette!(FortColorPalette:"Fort Color Palette Structure (by ref)"=Get Colors!(self:Fort Item Series Definition Object Reference=@n1.SeriesData) -> (Colors:Fort Color Palette Structure)) -> (
    Color1:Linear Color Structure
    Color2:Linear Color Structure
    Color3:Linear Color Structure
    Color4:Linear Color Structure
    Color5:Linear Color Structure
   ) {is=title K2Node_BreakStruct}
   > SetVectorParameterValue(self=@n2, ParameterName=BackingColor, Value=@n3.Color4)
   > SetVectorParameterValue(self=@n2, ParameterName=BorderColor, Value=@n3)
   > SetContentColorAndOpacity!(
    execute
    self:Border Object Reference=Get Border_RarityTint() -> (Border_RarityTint:Border Object Reference)
    InContentColorAndOpacity:Linear Color Structure=@n3
   ) -> (then)
   > SetContentColorAndOpacity(
    self=Get Border_ItemType() -> (Border_ItemType:Border Object Reference)
    InContentColorAndOpacity=Get ItemTypeTextColor() -> (ItemTypeTextColor:Linear Color Structure)
   )
   n4 = Event OnInitializeRarityWithoutSeries!() -> (OutputDelegate, then, RarityData:Fort Rarity Item Data Structure) {is=title K2Node_Event}
   > n5 = GetDynamicMaterial(self=Get Border_ItemRarity())
   n6 = Break Fort Rarity Item Data!(FortRarityItemData:"Fort Rarity Item Data Structure (by ref)"=@n4.RarityData) -> (
    Name:Text
    Color1:Linear Color Structure
    Color2:Linear Color Structure
    Color3:Linear Color Structure
    Color4:Linear Color Structure
    Color5:Linear Color Structure
    Radius:float=~
    Falloff:float=~
    Brightness:float=~
    Roughness:float=~
    Glow:float=~
   ) {is=title K2Node_BreakStruct}
   > SetVectorParameterValue(self=@n5, ParameterName=BackingColor, Value=@n6.Color4)
   > SetVectorParameterValue(self=@n5, ParameterName=BorderColor, Value=@n6.Color2)
   > SetContentColorAndOpacity(self=Get Border_RarityTint(), InContentColorAndOpacity=@n6.Color2)
   > SetContentColorAndOpacity(self=Get Border_ItemType(), InContentColorAndOpacity=Get ItemTypeTextColor())
  ]}
 OverrideTypeTextColor {nodes [
   n1 = OverrideTypeTextColor!() -> (then, InColorAndOpacity:Slate Color Structure) {is=title K2Node_FunctionEntry}
   > SetColorAndOpacity(self=Get Text_ItemType() -> (Text_ItemType:Common Text Object Reference), InColorAndOpacity=@n1)
   > SetContentColorAndOpacity(self=Get Border_ItemType(), InContentColorAndOpacity=(R=1,G=1,B=1,A=1))
  ]}
]
variables [ItemTypeTextColor {type=LinearColor, default=<rgba 1 1 1 1>}]
widgets {
 info {inheritedWidgetCount=4, parentClass=/Script/FortniteUI.AthenaRewardItemTypeRarityTag, rootWidgetClass=&SizeBox, widgetCount=7}
 widgets [
  HeightSB {widgetClassPath=&SizeBox, widget=HeightSB, properties {bOverride_HeightOverride=true, heightOverride=32}}
  Rarity+Type {widgetClassPath=&HorizontalBox, parent=HeightSB, widget=Rarity+Type, slot {class=SizeBoxSlot}}
  Border_ItemRarity {
   widgetClassPath=&Border
   bInherited=true
   bIsVariable=true
   parent=Rarity+Type
   widget=Border_ItemRarity
   properties {
    verticalAlignment=VAlign_Center
    background {
     bIsDynamicallyLoaded=false
     drawAs=Box
     imageType=NoImage
     mirroring=NoMirror
     resourceName=None
     resourceObject=/Game/Athena/UI/Season/Art/M_UI_BattlePass_SimpleTag
     tiling=NoTile
     imageSize=<32 32>
     margin=<ltrb 0.5 0 0.5 0>
     outlineSettings {
      bUseBrushTransparency=false
      roundingType=HalfHeightRadius
      width=0
      color {colorUseRule=UseColor_Specified, specifiedColor=<rgba 0 0 0 0>}
      cornerRadii=<0 0 0 0>
     }
     tintColor {colorUseRule=UseColor_Specified, specifiedColor=<rgba 1 1 1 1>}
     uVRegion {bIsValid=false, max=<0 0>, min=<0 0>}
    }
    padding=<ltrb 24 0 24 0>
   }
   slot {class=HorizontalBoxSlot, properties {horizontalAlignment=HAlign_Left, padding=<ltrb 0 0 14 0>}}
  }
  Border_RarityTint {
   widgetClassPath=&Border
   bInherited=true
   parent=Border_ItemRarity
   widget=Border_RarityTint
   properties {
    background {
     bIsDynamicallyLoaded=false
     drawAs=NoDrawType
     imageType=NoImage
     mirroring=NoMirror
     resourceName=None
     resourceObject=null
     tiling=NoTile
     imageSize=<32 32>
     margin=<ltrb 0 0 0 0>
     outlineSettings {
      bUseBrushTransparency=false
      roundingType=HalfHeightRadius
      width=0
      color {colorUseRule=UseColor_Specified, specifiedColor=<rgba 0 0 0 0>}
      cornerRadii=<0 0 0 0>
     }
     tintColor {colorUseRule=UseColor_Specified, specifiedColor=<rgba 1 1 1 1>}
     uVRegion {bIsValid=false, max=<0 0>, min=<0 0>}
    }
    padding=<ltrb 0 0 0 0>
   }
   slot {class=BorderSlot, properties {verticalAlignment=VAlign_Center, padding=<ltrb 24 0 24 0>}}
  }
  Text_ItemRarity {
   widgetClassPath=&CommonTextBlock
   bInherited=true
   bIsVariable=true
   parent=Border_RarityTint
   widget=Text_ItemRarity
   properties {
    mobileFontSizeMultiplier=1.2
    text=Item Rarity
    textTransformPolicy=ToUpper
    font {
     bForceMonospaced=false
     bMaterialIsStencil=false
     fontMaterial=null
     fontObject=&BurbankSmall
     letterSpacing=0
     monospacedWidth=1
     size=15
     skewAmount=0
     typefaceFontName=Black
     outlineSettings {
      bApplyOutlineToDropShadows=false
      bMiteredCorners=false
      bSeparateFillAlpha=false
      outlineMaterial=null
      outlineSize=0
      outlineColor=<rgba 0 0 0 1>
     }
    }
    margin=<ltrb 0 -4 0 -4>
    renderTransform {angle=0, scale=<1 1>, shear=<-12 0>, translation=<0 0>}
    shadowOffset=<0 0>
   }
   slot {class=BorderSlot, properties {padding=<ltrb 0 0 0 0>}}
  }
  Border_ItemType {
   widgetClassPath=&Border
   bIsVariable=true
   parent=Rarity+Type
   widget=Border_ItemType
   properties {
    background {
     bIsDynamicallyLoaded=false
     drawAs=NoDrawType
     imageType=NoImage
     mirroring=NoMirror
     resourceName=None
     resourceObject=null
     tiling=NoTile
     imageSize=<32 32>
     margin=<ltrb 0 0 0 0>
     outlineSettings {
      bUseBrushTransparency=false
      roundingType=HalfHeightRadius
      width=0
      color {colorUseRule=UseColor_Specified, specifiedColor=<rgba 0 0 0 0>}
      cornerRadii=<0 0 0 0>
     }
     tintColor {colorUseRule=UseColor_Specified, specifiedColor=<rgba 1 1 1 1>}
     uVRegion {bIsValid=false, max=<0 0>, min=<0 0>}
    }
    padding=<ltrb 0 0 0 0>
   }
   slot {class=HorizontalBoxSlot, properties {horizontalAlignment=HAlign_Right, verticalAlignment=VAlign_Center, size {sizeRule=Fill, value=1}}}
  }
  Text_ItemType {
   widgetClassPath=&CommonTextBlock
   bInherited=true
   bIsVariable=true
   parent=Border_ItemType
   widget=Text_ItemType
   properties {
    mobileFontSizeMultiplier=1.1
    scrollStyle=&TextScrollStyle-Base_C
    text=Item Type
    textTransformPolicy=ToUpper
    font {
     bForceMonospaced=false
     bMaterialIsStencil=false
     fontMaterial=null
     fontObject=&BurbankSmall
     letterSpacing=0
     monospacedWidth=1
     size=18
     skewAmount=0
     typefaceFontName=Black
     outlineSettings {
      bApplyOutlineToDropShadows=false
      bMiteredCorners=false
      bSeparateFillAlpha=false
      outlineMaterial=null
      outlineSize=0
      outlineColor=<rgba 0 0 0 1>
     }
    }
    shadowOffset=<0 0>
   }
   slot {class=BorderSlot, properties {padding=<ltrb 0 0 0 0>}}
  }
 ]
}
