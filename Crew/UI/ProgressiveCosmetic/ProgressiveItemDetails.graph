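# Fortnite Crew widget showing details of a progressive cosmetic reward: rarity tag, stage title, owned badge and lock-marked unlock criteria.
# UpdateSelectedReward initializes the rarity tag from the reward view model's first item, collapsing it when the item list is empty.
# UpdateUnlockCriteria sets the rich text and hides its row when empty.
# UpdateSubscribedVisuals recolors criteria and tag type text by subscription state; the graphs leave title and badge untouched.
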WidgetBlueprint /Crew/UI/ProgressiveCosmetic/ProgressiveItemDetails
graphs [
 EventGraph {nodes [UpdateSubscribedVisuals!(execute, self, IsSubscribed:bool=false) -> (then)]}
 UpdateSelectedReward {nodes [
   n1 = UpdateSelectedReward!() -> (then, Reward:Crew Progressive Reward VM Object Reference) {is=title K2Node_FunctionEntry}
   > Is Valid(InputObject=@n1)
   n2 = Get Items!(self:Crew Progressive Reward VM Object Reference=@n1) -> (Items:Array of Fort Item VM Object References)
   > n3 = Branch(Condition=Is Empty!(TargetArray:Array of Fort Item VM Object References=@n2) -> (ReturnValue:bool))
   n4 = Get Widget_ItemTypeRarityTag() -> (Widget_ItemTypeRarityTag:Athena Reward Item Type Rarity Tag Object Reference)
   > SetVisibility(self=@n4, InVisibility=Collapsed)
   SetVisibility(execute=@n3.else, self=@n4, InVisibility=SelfHitTestInvisible)
   > BP_InitializeTag!(
    execute
    self:Athena Reward Item Type Rarity Tag Object Reference=@n4
    RewardItem:Fort Item Object Reference=Get ItemInstance!(self:Fort Item VM Object Reference="Get (a copy)"(Array:Array of Fort Item VM Object References=@n2) -> (Output:Fort Item VM Object Reference)) -> (ItemInstance:Fort Item Object Reference)
   ) -> (then)
  ]}
 UpdateSubscribedVisuals {nodes [
   n1 = UpdateSubscribedVisuals!() -> (then, IsSubscribed:bool=~) {is=title K2Node_FunctionEntry}
   n2 = Select(
    Option 0:Slate Color Structure=Make Slate Color(SpecifiedColor=(R=0.0331,G=0.855,B=1,A=1))
    Option 1:Slate Color Structure=Make Slate Color(SpecifiedColor=(R=0.9734,G=0.9216,B=0.314,A=1))
    Index=@n1
   ) -> (ReturnValue:Slate Color Structure)
   > SetDefaultColorAndOpacity!(
    execute
    self:Rich Text Block Object Reference=Get Text_UnlockCriteria() -> (Text_UnlockCriteria:Common Rich Text Block Object Reference)
    InColorAndOpacity:Slate Color Structure=@n2
   ) -> (then)
   > OverrideTypeTextColor!(
    execute
    self:Athena Reward Item Type Rarity Tag Object Reference=Get Widget_ItemTypeRarityTag()
    InColorAndOpacity:Slate Color Structure=@n2
   ) -> (then)
  ]}
 UpdateUnlockCriteria {nodes [
   n1 = UpdateUnlockCriteria!() -> (then, Text:Text) {is=title K2Node_FunctionEntry}
   > SetText(self=Get Text_UnlockCriteria(), InText=@n1)
   > SetVisibility(
    self=Get Widget_UnlockCriteriaContainer() -> (Widget_UnlockCriteriaContainer:Horizontal Box Object Reference)
    InVisibility=Select(Option 0:ESlateVisibility Enum=SelfHitTestInvisible, Option 1:ESlateVisibility Enum=Collapsed, Index=TextIsEmpty(InText=@n1)) -> (ReturnValue:ESlateVisibility Enum)
   )
  ]}
]
variables []
widgets {
 info {inheritedWidgetCount=0, parentClass=&CommonUserWidget, rootWidgetClass=&Overlay, widgetCount=12}
 widgets [
  PlaceHolder {widgetClassPath=&Overlay, widget=PlaceHolder}
  VerticalBox_38 {
   widgetClassPath=&VerticalBox
   parent=PlaceHolder
   widget=VerticalBox_38
   slot {class=OverlaySlot, properties {horizontalAlignment=HAlign_Fill, verticalAlignment=VAlign_Fill}}
  }
  Widget_ItemTypeRarityTag {
   widgetClassPath=/Game/Athena/UI/Season/AthenaRewardItemTypeRarityTag
   bIsVariable=true
   parent=VerticalBox_38
   widget=Widget_ItemTypeRarityTag
   slot {class=VerticalBoxSlot, properties {horizontalAlignment=HAlign_Right}}
  }
  Text_CosmeticStage {
   widgetClassPath=&CommonTextBlock
   parent=VerticalBox_38
   widget=Text_CosmeticStage
   properties {
    justification=Right
    style=/FortUILibrary/Styles/Typography/_TextStyle_SubH2._TextStyle_SubH2_C
    text="{BACK BLING -STYLE  X}"
    textTransformPolicy=ToUpper
    wrapTextAt=512
    font {
     bForceMonospaced=false
     bMaterialIsStencil=false
     fontMaterial=null
     fontObject=&Burbank
     letterSpacing=0
     monospacedWidth=1
     size=33
     skewAmount=0
     typefaceFontName=BurbankRegular-Black
     outlineSettings {
      bApplyOutlineToDropShadows=false
      bMiteredCorners=false
      bSeparateFillAlpha=false
      outlineMaterial=null
      outlineSize=0
      outlineColor=<rgba 0 0 0 1>
     }
    }
    shadowOffset=<0 0>
    shapedTextOptions {bOverride_TextFlowDirection=true, bOverride_TextShapingMethod=false, textFlowDirection=LeftToRight, textShapingMethod=Auto}
    strikeBrush {
     bIsDynamicallyLoaded=false
     drawAs=NoDrawType
     imageType=NoImage
     mirroring=NoMirror
     resourceName=None
     resourceObject=null
     tiling=NoTile
     imageSize=<32 32>
     margin=<ltrb 0 0 0 0>
     outlineSettings {
      bUseBrushTransparency=false
      roundingType=HalfHeightRadius
      width=0
      color {colorUseRule=UseColor_Specified, specifiedColor=<rgba 0 0 0 0>}
      cornerRadii=<0 0 0 0>
     }
     tintColor {colorUseRule=UseColor_Specified, specifiedColor=<rgba 1 1 1 1>}
     uVRegion {bIsValid=false, max=<0 0>, min=<0 0>}
    }
   }
   slot {class=VerticalBoxSlot, properties {padding=<ltrb 0 0 0 10>}}
  }
  Tag_Owned {
   widgetClassPath=&Border
   parent=VerticalBox_38
   widget=Tag_Owned
   properties {
    visibility=HitTestInvisible
    background {
     bIsDynamicallyLoaded=false
     drawAs=Box
     imageType=NoImage
     mirroring=NoMirror
     resourceName=None
     resourceObject=/Game/Athena/UI/ReplayBrowser/M_Ui_Skewed_NoV1
     tiling=NoTile
     imageSize=<32 32>
     margin=<ltrb 0.25 0 0.25 0>
     outlineSettings {
      bUseBrushTransparency=false
      roundingType=HalfHeightRadius
      width=0
      color {colorUseRule=UseColor_Specified, specifiedColor=<rgba 0 0 0 0>}
      cornerRadii=<0 0 0 0>
     }
     tintColor {colorUseRule=UseColor_Specified, specifiedColor=<rgba 0.0343 1 0.227 1>}
     uVRegion {bIsValid=false, max=<0 0>, min=<0 0>}
    }
    contentColorAndOpacity=<rgba 0 0.147 0.063 1>
    padding=<ltrb 10 2 10 2>
   }
   slot {class=VerticalBoxSlot, properties {horizontalAlignment=HAlign_Right, padding=<ltrb 0 6 0 10>}}
  }
  HorizontalBox_5 {
   widgetClassPath=&HorizontalBox
   parent=Tag_Owned
   widget=HorizontalBox_5
   slot {class=BorderSlot, properties {padding=<ltrb 10 2 10 2>}}
  }
  Image_140 {
   widgetClassPath=&Image
   parent=HorizontalBox_5
   widget=Image_140
   properties {
    brush {
     bIsDynamicallyLoaded=false
     drawAs=Image
     imageType=NoImage
     mirroring=NoMirror
     resourceName=None
     resourceObject=/Game/Athena/HUD/Bacchus/HUDLayoutTool/HLT_Icons/T-HLT-CheckMark-32
     tiling=NoTile
     imageSize=<16 16>
     margin=<ltrb 0 0 0 0>
     outlineSettings {
      bUseBrushTransparency=false
      roundingType=HalfHeightRadius
      width=0
      color {colorUseRule=UseColor_Specified, specifiedColor=<rgba 0 0 0 0>}
      cornerRadii=<0 0 0 0>
     }
     tintColor {colorUseRule=UseColor_Specified, specifiedColor=<rgba 1 1 1 1>}
     uVRegion {bIsValid=false, max=<0 0>, min=<0 0>}
    }
    renderTransform {angle=0, scale=<1.3 1.3>, shear=<0 0>, translation=<0 0>}
   }
   slot {class=HorizontalBoxSlot, properties {horizontalAlignment=HAlign_Left, verticalAlignment=VAlign_Center, padding=<ltrb 0 -1.5 6 0>}}
  }
  CommonTextBlock_10 {
   widgetClassPath=&CommonTextBlock
   parent=HorizontalBox_5
   widget=CommonTextBlock_10
   properties {
    text="Owned!"
    textTransformPolicy=ToUpper
    font {
     bForceMonospaced=false
     bMaterialIsStencil=false
     fontMaterial=null
     fontObject=&BurbankSmall
     letterSpacing=0
     monospacedWidth=1
     size=14
     skewAmount=0
     typefaceFontName=Black
     outlineSettings {
      bApplyOutlineToDropShadows=false
      bMiteredCorners=false
      bSeparateFillAlpha=false
      outlineMaterial=null
      outlineSize=0
      outlineColor=<rgba 0 0 0 1>
     }
    }
    renderTransform {angle=0, scale=<1 1>, shear=<-8 0>, translation=<0 0>}
    shadowOffset=<0 0>
   }
   slot {class=HorizontalBoxSlot}
  }
  Widget_UnlockCriteriaContainer {
   widgetClassPath=&HorizontalBox
   bIsVariable=true
   parent=VerticalBox_38
   widget=Widget_UnlockCriteriaContainer
   slot {class=VerticalBoxSlot, properties {horizontalAlignment=HAlign_Right}}
  }
  Image_Lock_6 {
   widgetClassPath=&Image
   parent=Widget_UnlockCriteriaContainer
   widget=Image_Lock_6
   properties {brush {
     bIsDynamicallyLoaded=false
     drawAs=Image
     imageType=NoImage
     mirroring=NoMirror
     resourceName=None
     resourceObject=/FortUILibrary/Foundation/Art_ToolKit/AngledBoxes/M_UI_AngledLockedBox_Red
     tiling=NoTile
     imageSize=<40 40>
     margin=<ltrb 0 0 0 0>
     outlineSettings {
      bUseBrushTransparency=false
      roundingType=HalfHeightRadius
      width=0
      color {colorUseRule=UseColor_Specified, specifiedColor=<rgba 0 0 0 0>}
      cornerRadii=<0 0 0 0>
     }
     tintColor {colorUseRule=UseColor_Specified, specifiedColor=<rgba 1 1 1 1>}
     uVRegion {bIsValid=false, max=<0 0>, min=<0 0>}
    }}
   slot {class=HorizontalBoxSlot, properties {horizontalAlignment=HAlign_Left, verticalAlignment=VAlign_Top, padding=<ltrb 0 4 8 0>}}
  }
  VerticalBox {
   widgetClassPath=&VerticalBox
   parent=Widget_UnlockCriteriaContainer
   widget=VerticalBox
   slot {class=HorizontalBoxSlot, properties {verticalAlignment=VAlign_Center, padding=<ltrb 6 0 0 0>, size {sizeRule=Fill, value=1}}}
  }
  Text_UnlockCriteria {
   widgetClassPath=&CommonRichTextBlock
   bIsVariable=true
   parent=VerticalBox
   widget=Text_UnlockCriteria
   properties {
    bOverrideDefaultStyle=true
    mobileTextBlockScale=1.1
    text="{PH} Unlock the <text color=\"FFF\">Protocol Elite Style</> by being an <text color=\"FFF\">active subscriber</> to the <text color=\"FFF\">Fortnite Crew</> for <text color=\"FFF\">1</> additional month."
    wrapTextAt=625
    defaultTextStyleOverride {
     overflowPolicy=Clip
     transformPolicy=None
     colorAndOpacity {colorUseRule=UseColor_Specified, specifiedColor=<rgba 0.0331 0.855 1 1>}
     font {
      bForceMonospaced=false
      bMaterialIsStencil=false
      fontMaterial=null
      fontObject=&BurbankSmall
      letterSpacing=0
      monospacedWidth=1
      size=18
      skewAmount=0
      typefaceFontName=Bold
      outlineSettings {
       bApplyOutlineToDropShadows=false
       bMiteredCorners=false
       bSeparateFillAlpha=false
       outlineMaterial=null
       outlineSize=0
       outlineColor=<rgba 0 0 0 1>
      }
     }
     highlightColor {colorUseRule=UseColor_Specified, specifiedColor=<rgba 0 0 0 1>}
     highlightShape {
      bIsDynamicallyLoaded=false
      drawAs=Image
      imageType=NoImage
      mirroring=NoMirror
      resourceName=None
      resourceObject=null
      tiling=NoTile
      imageSize=<32 32>
      margin=<ltrb 0 0 0 0>
      outlineSettings {
       bUseBrushTransparency=false
       roundingType=HalfHeightRadius
       width=0
       color {colorUseRule=UseColor_Specified, specifiedColor=<rgba 0 0 0 0>}
       cornerRadii=<0 0 0 0>
      }
      tintColor {colorUseRule=UseColor_Specified, specifiedColor=<rgba 1 1 1 1>}
      uVRegion {bIsValid=false, max=<0 0>, min=<0 0>}
     }
     selectedBackgroundColor {colorUseRule=UseColor_Foreground, specifiedColor=<rgba 1 0 1 1>}
     shadowColorAndOpacity=<rgba 0 0 0 1>
     shadowOffset=<0 0>
     strikeBrush {
      bIsDynamicallyLoaded=false
      drawAs=Image
      imageType=NoImage
      mirroring=NoMirror
      resourceName=None
      resourceObject=null
      tiling=NoTile
      imageSize=<32 32>
      margin=<ltrb 0 0 0 0>
      outlineSettings {
       bUseBrushTransparency=false
       roundingType=HalfHeightRadius
       width=0
       color {colorUseRule=UseColor_Specified, specifiedColor=<rgba 0 0 0 0>}
       cornerRadii=<0 0 0 0>
      }
      tintColor {colorUseRule=UseColor_Specified, specifiedColor=<rgba 1 1 1 1>}
      uVRegion {bIsValid=false, max=<0 0>, min=<0 0>}
     }
     underlineBrush {
      bIsDynamicallyLoaded=false
      drawAs=Image
      imageType=NoImage
      mirroring=NoMirror
      resourceName=None
      resourceObject=null
      tiling=NoTile
      imageSize=<32 32>
      margin=<ltrb 0 0 0 0>
      outlineSettings {
       bUseBrushTransparency=false
       roundingType=HalfHeightRadius
       width=0
       color {colorUseRule=UseColor_Specified, specifiedColor=<rgba 0 0 0 0>}
       cornerRadii=<0 0 0 0>
      }
      tintColor {colorUseRule=UseColor_Specified, specifiedColor=<rgba 1 1 1 1>}
      uVRegion {bIsValid=false, max=<0 0>, min=<0 0>}
     }
    }
   }
   slot {class=VerticalBoxSlot, properties {verticalAlignment=VAlign_Center}}
  }
 ]
}
