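# ProgressiveItemScreenPip is a pip button widget, apparently for stages on the progressive cosmetic item screen.
# It derives from NoInputActionPipButton, adding no widgets, variables or event logic of its own.
# Its Get Pip Text function converts the stage index to a label through an integer add, left at its default increment, so numbering displays one-based.
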WidgetBlueprint /Crew/UI/ProgressiveCosmetic/ProgressiveItemScreenPip
graphs [
 EventGraph {links [], nodes []}
 Get Pip Text {nodes [
   n1 = Get Pip Text!() -> (then, Stage Index:int) {is=title K2Node_FunctionEntry}
   > Return Node!(execute, Result:Text="To Text (Integer)"(Value=int + int(A=@n1)))
  ]}
]
variables []
widgets {
 info {
  inheritedWidgetCount=0
  parentClass=&/NoInputActionPipButton.NoInputActionPipButton_C
  rootWidgetClass=null
  widgetCount=0
 }
 widgets []
}
