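# ProgressiveTagsBase is a tag widget for progressive cosmetic UI, showing an optional icon and uppercase sheared text on a skewed-material background.
# PreConstruct pushes its variables through UpdateColor, UpdateText and Update Image, so instances preview in the designer.
# UpdateText collapses the label when its text is empty.
# Update Image prefers a texture, falls back to a material with tint and rotation applied, and collapses the icon when neither is valid.
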WidgetBlueprint /Crew/UI/ProgressiveCosmetic/ProgressiveTagsBase
graphs [
 EventGraph {nodes [
   Event PreConstruct()
   > UpdateColor!(
    execute
    self
    Background Color:Linear Color Structure=Get Background Color() -> (Background Color:Linear Color Structure)
    Text Color:Slate Color Structure=Get Text Color() -> (Text Color:Slate Color Structure)
   ) -> (then)
   > UpdateText!(execute, self, Text:Text=Get Text!() -> (Text)) -> (then)
   > Update Image!(
    execute
    self
    Texture:Texture 2D Object Reference=Get Image Texture() -> (Image Texture:Texture 2D Object Reference)
    Material:Material Interface Object Reference=Get Image Material() -> (Image Material:Material Interface Object Reference)
    Color:Linear Color Structure=Get Image Color() -> (Image Color:Linear Color Structure)
    Angle:double=Get Image Rotation() -> (Image Rotation:double)
   ) -> (then)
   Event Construct()
   Event Tick()
  ]}
 Update Image {nodes [
   Update Image!() -> (then, Texture:Texture 2D Object Reference, Material:Material Interface Object Reference, Color:Linear Color Structure, Angle:double=~) {is=title K2Node_FunctionEntry}
   > n1 = Get!(execute:Exec) -> (then:Exec, else:Exec, Texture:Texture 2D Object Reference)
   > n2 = SetBrushFromTexture!(
    execute
    self:Image Object Reference=Get Tag_Image() -> (Tag_Image:Image Object Reference)
    Texture:Texture 2D Object Reference=Get Texture() -> (Texture:Texture 2D Object Reference)
    bMatchSize:bool=false
   ) -> (then)
   n3 = Get Tag_Image()
   SetColorAndOpacity(
    execute=[@n2.then, @n5.then]
    self:Image Object Reference=@n3
    InColorAndOpacity:Linear Color Structure=Get Color() -> (Color:Linear Color Structure)
   )
   > SetRenderTransformAngle!(execute, self:Widget Object Reference=@n3, Angle:float=Get Angle() -> (Angle:double)) -> (then)
   > SetVisibility(self=@n3, InVisibility=HitTestInvisible)
   n4 = Get!(execute=@n1.else) -> (then, else, Material:Material Interface Object Reference)
   > n5 = SetBrushFromMaterial(self=Get Tag_Image(), Material=@n4)
   SetVisibility(execute=@n4.else, self=Get Tag_Image(), InVisibility=Collapsed)
  ]}
 UpdateColor {nodes [
   n1 = UpdateColor!() -> (then, Background Color:Linear Color Structure, Text Color:Slate Color Structure) {is=title K2Node_FunctionEntry}
   > SetBrushColor(self=Get Tag_Background() -> (Tag_Background:Border Object Reference), InBrushColor=@n1)
   > SetColorAndOpacity(self:Text Object Reference=Get Tag_Text() -> (Tag_Text:Common Text Object Reference), InColorAndOpacity:Slate Color Structure=@n1.Text Color)
  ]}
 UpdateText {nodes [
   n1 = UpdateText!() -> (then, Text:Text) {is=title K2Node_FunctionEntry}
   > "SetText (Text)"(self=Get Tag_Text(), InText=@n1)
   > SetVisibility(InVisibility=Select(Option 0=HitTestInvisible, Option 1=Collapsed, Index=TextIsEmpty(InText=@n1)))
  ]}
]
variables [
 Background Color {type=LinearColor, default=<rgba 0.2483 0.2483 0.2483 1>}
 Image Color {type=LinearColor, default=<rgba 1 1 1 0>}
 Image Material {type=MaterialInterface, default=null}
 Image Rotation {type=number}
 Image Texture {type=Texture2D}
 Text {type=string, default=Text Goes here}
 Text Color {type=SlateColor, default {colorUseRule=UseColor_Specified, specifiedColor=<rgba 1 1 1 1>}}
]
widgets {
 info {inheritedWidgetCount=0, parentClass=&UserWidget, rootWidgetClass=&Border, widgetCount=4}
 widgets [
  Tag_Background {
   widgetClassPath=&Border
   bIsVariable=true
   widget=Tag_Background
   properties {
    verticalAlignment=VAlign_Center
    visibility=HitTestInvisible
    background {
     bIsDynamicallyLoaded=false
     drawAs=Box
     imageType=NoImage
     mirroring=NoMirror
     resourceName=None
     resourceObject=/Game/Athena/UI/ReplayBrowser/M_Ui_Skewed_NoV1
     tiling=NoTile
     imageSize=<32 32>
     margin=<ltrb 0.25 0 0.25 0>
     outlineSettings {
      bUseBrushTransparency=false
      roundingType=HalfHeightRadius
      width=0
      color {colorUseRule=UseColor_Specified, specifiedColor=<rgba 0 0 0 0>}
      cornerRadii=<0 0 0 0>
     }
     tintColor {colorUseRule=UseColor_Specified, specifiedColor=<rgba 1 1 1 1>}
     uVRegion {bIsValid=false, max=<0 0>, min=<0 0>}
    }
    brushColor=<rgba 0 0.147 0.063 1>
    padding=<ltrb 16 2 16 2>
   }
  }
  HorizontalBox_0 {
   widgetClassPath=&HorizontalBox
   parent=Tag_Background
   widget=HorizontalBox_0
   slot {class=BorderSlot, properties {verticalAlignment=VAlign_Center, padding=<ltrb 16 2 16 2>}}
  }
  Tag_Image {
   widgetClassPath=&Image
   bIsVariable=true
   parent=HorizontalBox_0
   widget=Tag_Image
   properties {
    visibility=Collapsed
    brush {
     bIsDynamicallyLoaded=false
     drawAs=Image
     imageType=NoImage
     mirroring=NoMirror
     resourceName=None
     resourceObject=null
     tiling=NoTile
     imageSize=<24 24>
     margin=<ltrb 0 0 0 0>
     outlineSettings {
      bUseBrushTransparency=false
      roundingType=HalfHeightRadius
      width=0
      color {colorUseRule=UseColor_Specified, specifiedColor=<rgba 0 0 0 0>}
      cornerRadii=<0 0 0 0>
     }
     tintColor {colorUseRule=UseColor_Specified, specifiedColor=<rgba 1 1 1 1>}
     uVRegion {bIsValid=false, max=<0 0>, min=<0 0>}
    }
   }
   slot {class=HorizontalBoxSlot, properties {horizontalAlignment=HAlign_Left, verticalAlignment=VAlign_Center, padding=<ltrb -8 -2 4 -2>}}
  }
  Tag_Text {
   widgetClassPath=&CommonTextBlock
   bIsVariable=true
   parent=HorizontalBox_0
   widget=Tag_Text
   properties {
    mobileFontSizeMultiplier=1.2
    text=Default
    textTransformPolicy=ToUpper
    colorAndOpacity {colorUseRule=UseColor_Specified, specifiedColor=<rgba 0.1329 0.8796 0.3712 1>}
    font {
     bForceMonospaced=false
     bMaterialIsStencil=false
     fontMaterial=null
     fontObject=&BurbankSmall
     letterSpacing=0
     monospacedWidth=1
     size=14
     skewAmount=0
     typefaceFontName=Black
     outlineSettings {
      bApplyOutlineToDropShadows=false
      bMiteredCorners=false
      bSeparateFillAlpha=false
      outlineMaterial=null
      outlineSize=0
      outlineColor=<rgba 0 0 0 1>
     }
    }
    renderTransform {angle=0, scale=<1 1>, shear=<-8 0>, translation=<0 0>}
    shadowOffset=<0 0>
   }
   slot {class=HorizontalBoxSlot, properties {verticalAlignment=VAlign_Center}}
  }
 ]
}
