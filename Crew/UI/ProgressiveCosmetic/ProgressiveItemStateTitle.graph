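# Crew UI header widget for a progressive cosmetic set, showing a title, subheading and multi-month benefit, stage and subscription tags.
# PreConstruct sets the title and benefits tag, plus placeholder tag text at design time.
# Setup Header Info passes subscription and stage data to helpers that update tag text and recolor the subheading.
# The stages tag collapses at zero max stages, otherwise showing Completed with a checkmark or an unlocked count colored by subscription.
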WidgetBlueprint /Crew/UI/ProgressiveCosmetic/ProgressiveItemStateTitle
graphs [
 EventGraph {nodes [
   n1 = Event PreConstruct()
   > Update MultiBenefits Tag!(execute, self, Index:bool=Get Show Multi-Benefits tag() -> (Show Multi-Benefits tag:bool)) -> (then)
   > Update Title Text()
   > Branch(Condition=@n1.IsDesignTime)
   > UpdateText!(
    execute
    self:Progressive Tags Base Object Reference=Get ProgressiveTags_Stages() -> (ProgressiveTags_Stages:Progressive Tags Base Object Reference)
    Text:Text="LOCTEXT(\"[PH] Completed\")"
   ) -> (then)
   > UpdateText(
    self=Get ProgressiveTags_Subscribed() -> (ProgressiveTags_Subscribed:Progressive Tags Base Object Reference)
    Text="LOCTEXT(\"[PH] Subscribe\")"
   )
  ]}
 Setup Header Info {nodes [
   n1 = Setup Header Info!() -> (then, InputPin:"Text (by ref)", Subscribed:bool=~, Current Stage:int, Max Stage:int) {is=title K2Node_FunctionEntry}
   > Update Stages Tag!(execute, self, Subscribed:bool=@n1.Subscribed, Current Stage:int=@n1.Current Stage, Max Stage:int=@n1.Max Stage) -> (then)
   > Update Subheading!(execute, self, Text:Text=@n1, Subscribed:bool=@n1.Subscribed) -> (then)
   > Update Subscribed Tag!(execute, self, Subscribed:bool=@n1.Subscribed) -> (then)
  ]}
 Update MultiBenefits Tag {nodes [
   n1 = Update MultiBenefits Tag!() -> (then, Index:bool=~) {is=title K2Node_FunctionEntry}
   > UpdateText(
    self=Get ProgressiveTags_Benefits() -> (ProgressiveTags_Benefits:Progressive Tags Base Object Reference)
    Text=Select(Option 0:Text, Option 1:Text=Format Text(Format="LOCTEXT(\"Multi-Month Benefit\")"), Index=@n1) -> (ReturnValue:Text)
   )
  ]}
 Update Stages Tag {nodes [
   n1 = Update Stages Tag!() -> (then, Subscribed:bool=~, Current Stage:int, Max Stage:int) {is=title K2Node_FunctionEntry}
   > n2 = Branch(Condition="Equal (Integer)"(A=@n1.Max Stage))
   > SetVisibility(self=Get ProgressiveTags_Stages(), InVisibility=Collapsed)
   n3 = Branch(execute=@n2.else, Condition="Equal (Integer)"(A=Get Current Stage() -> (Current Stage:int=0), B=Get Max Stage() -> (Max Stage:int=0)))
   n4 = Get ProgressiveTags_Stages()
   > UpdateText(self=@n4, Text=Format Text(Format="LOCTEXT(\"Completed\")"))
   n5 = Get Subbed Dark Green() -> (Subbed Dark Green:Linear Color Structure)
   > Update Image!(
    execute
    self:Progressive Tags Base Object Reference=@n4
    Texture:Texture 2D Object Reference=/Game/Athena/HUD/Bacchus/HUDLayoutTool/HLT_Icons/T-HLT-Checkmark-64
    Material:Material Interface Object Reference
    Color:Linear Color Structure=@n5
    Angle:double=0
   ) -> (then)
   > UpdateColor!(
    execute
    self:Progressive Tags Base Object Reference=@n4
    Background Color:Linear Color Structure=Get Subbed Light Green() -> (Subbed Light Green:Linear Color Structure)
    Text Color:Slate Color Structure=Make Slate Color(SpecifiedColor=@n5)
   ) -> (then)
   n6 = Get ProgressiveTags_Stages()
   UpdateText(
    execute=@n3.else
    self=@n6
    Text=Format Text!(Format="LOCTEXT(\"Unlocked {0}/{1} Stages\")", 0:int=Get Current Stage() -> (Current Stage=0), 1:int=Get Max Stage() -> (Max Stage=0)) -> (Result)
   )
   > Update Image(self=@n6, Angle=0)
   n7 = Get Subscribed() -> (Subscribed:bool)
   n8 = Get Subbed Light Green()
   > UpdateColor(
    self=@n6
    Background Color=SelectColor(A=Get Subbed Dark Green(), B=Get Unsubbed Pink BG() -> (Unsubbed Pink BG:Linear Color Structure), bPickA=@n7)
    Text Color=Select(
     Option 0:Slate Color Structure=Get Unsubbed White() -> (Unsubbed White:Slate Color Structure)
     Option 1:Slate Color Structure=Make Slate Color(SpecifiedColor=@n8)
     Index=@n7
    ) -> (ReturnValue:Slate Color Structure)
   )
  ]}
 Update Subheading {nodes [
   n1 = Update Subheading!() -> (then, Text:Text, Subscribed:bool=~) {is=title K2Node_FunctionEntry}
   n2 = Get Text_Subheading() -> (Text_Subheading:Common Text Object Reference)
   > Set Text(Text=@n1, self=@n2)
   > SetColorAndOpacity(
    self=@n2
    InColorAndOpacity=Select(
     Option 0=Get Unsubscribed Subheader Color() -> (Unsubscribed Subheader Color:Slate Color Structure)
     Option 1=Get Subscibed Subheader Color() -> (Subscibed Subheader Color:Slate Color Structure)
     Index=@n1.Subscribed
    )
   )
  ]}
 Update Subscribed Tag {nodes [
   n1 = Update Subscribed Tag!() -> (then, Subscribed:bool=~) {is=title K2Node_FunctionEntry}
   > UpdateText(
    self=Get ProgressiveTags_Subscribed()
    Text=Select(Option 0:Text, Option 1:Text=Format Text!(Format="LOCTEXT(\"Subscribed!\")") -> (Result), Index=@n1) -> (ReturnValue:Text)
   )
  ]}
 Update Title Text {nodes [
   Update Title Text() {is=title K2Node_FunctionEntry}
   > "SetText (Text)"(self=Get Text_Title() -> (Text_Title:Common Text Object Reference), InText=Get Title Text() -> (Title Text:Text))
  ]}
]
variables [
 Show Multi-Benefits tag {type=boolean}
 Subbed Dark Green {type=LinearColor, default=<rgba 0 0.147 0.063 1>}
 Subbed Light Green {type=LinearColor, default=<rgba 0.0343 1 0.227 1>}
 Subscibed Subheader Color {type=SlateColor, default {colorUseRule=UseColor_Specified, specifiedColor=<rgba 0.9734 0.9216 0.3278 1>}}
 Text {type=string, default=""}
 Title Text {type=string, default=Fortnite Crew}
 Unsubbed Pink BG {type=LinearColor, default=<rgba 0.7913 0.0086 0.1845 1>}
 Unsubbed White {type=SlateColor, default {colorUseRule=UseColor_Specified, specifiedColor=<rgba 1 1 1 1>}}
 Unsubscribed Subheader Color {type=SlateColor, default {colorUseRule=UseColor_Specified, specifiedColor=<rgba 0.0331 0.855 1 1>}}
]
widgets {
 info {inheritedWidgetCount=0, parentClass=&CommonUserWidget, rootWidgetClass=&HorizontalBox, widgetCount=10}
 widgets [
  HorizontalBox_0 {widgetClassPath=&HorizontalBox, widget=HorizontalBox_0}
  Overlay_0 {
   widgetClassPath=&Overlay
   parent=HorizontalBox_0
   widget=Overlay_0
   slot {class=HorizontalBoxSlot, properties {horizontalAlignment=HAlign_Left, verticalAlignment=VAlign_Center}}
  }
  VerticalBox_0 {
   widgetClassPath=&VerticalBox
   parent=HorizontalBox_0
   widget=VerticalBox_0
   slot {class=HorizontalBoxSlot, properties {verticalAlignment=VAlign_Center, size {sizeRule=Fill, value=1}}}
  }
  Text_Title {
   widgetClassPath=&CommonTextBlock
   bIsVariable=true
   parent=VerticalBox_0
   widget=Text_Title
   properties {
    style=/FortUILibrary/Styles/Typography/_TextStyle_SubH2._TextStyle_SubH2_C
    text="[PH] Crew Legacy Set"
    textTransformPolicy=ToUpper
    font {
     bForceMonospaced=false
     bMaterialIsStencil=false
     fontMaterial=null
     fontObject=&Burbank
     letterSpacing=0
     monospacedWidth=1
     size=33
     skewAmount=0
     typefaceFontName=BurbankRegular-Black
     outlineSettings {
      bApplyOutlineToDropShadows=false
      bMiteredCorners=false
      bSeparateFillAlpha=false
      outlineMaterial=null
      outlineSize=0
      outlineColor=<rgba 0 0 0 1>
     }
    }
    shadowOffset=<0 0>
    strikeBrush {
     bIsDynamicallyLoaded=false
     drawAs=NoDrawType
     imageType=NoImage
     mirroring=NoMirror
     resourceName=None
     resourceObject=null
     tiling=NoTile
     imageSize=<32 32>
     margin=<ltrb 0 0 0 0>
     outlineSettings {
      bUseBrushTransparency=false
      roundingType=HalfHeightRadius
      width=0
      color {colorUseRule=UseColor_Specified, specifiedColor=<rgba 0 0 0 0>}
      cornerRadii=<0 0 0 0>
     }
     tintColor {colorUseRule=UseColor_Specified, specifiedColor=<rgba 1 1 1 1>}
     uVRegion {bIsValid=false, max=<0 0>, min=<0 0>}
    }
   }
   slot {class=VerticalBoxSlot, properties {verticalAlignment=VAlign_Top, padding=<ltrb 16 0 0 -16>}}
  }
  ScaleBox_1 {
   widgetClassPath=&ScaleBox
   parent=VerticalBox_0
   widget=ScaleBox_1
   properties {stretch=ScaleToFitX, stretchDirection=DownOnly}
   slot {class=VerticalBoxSlot, properties {verticalAlignment=VAlign_Top, padding=<ltrb 16 8 0 8>}}
  }
  Text_Subheading {
   widgetClassPath=&CommonTextBlock
   bIsVariable=true
   parent=ScaleBox_1
   widget=Text_Subheading
   properties {
    style=/FortUILibrary/Styles/Typography/_TextStyle_SubH4._TextStyle_SubH4_C
    text="[PH] Progressive Pickaxe"
    textTransformPolicy=ToUpper
    font {
     bForceMonospaced=false
     bMaterialIsStencil=false
     fontMaterial=null
     fontObject=&BurbankSmall
     letterSpacing=0
     monospacedWidth=1
     size=24
     skewAmount=0
     typefaceFontName=Black
     outlineSettings {
      bApplyOutlineToDropShadows=false
      bMiteredCorners=false
      bSeparateFillAlpha=false
      outlineMaterial=null
      outlineSize=0
      outlineColor=<rgba 0 0 0 1>
     }
    }
    shadowOffset=<0 0>
    strikeBrush {
     bIsDynamicallyLoaded=false
     drawAs=NoDrawType
     imageType=NoImage
     mirroring=NoMirror
     resourceName=None
     resourceObject=null
     tiling=NoTile
     imageSize=<32 32>
     margin=<ltrb 0 0 0 0>
     outlineSettings {
      bUseBrushTransparency=false
      roundingType=HalfHeightRadius
      width=0
      color {colorUseRule=UseColor_Specified, specifiedColor=<rgba 0 0 0 0>}
      cornerRadii=<0 0 0 0>
     }
     tintColor {colorUseRule=UseColor_Specified, specifiedColor=<rgba 1 1 1 1>}
     uVRegion {bIsValid=false, max=<0 0>, min=<0 0>}
    }
   }
   slot {class=ScaleBoxSlot, properties {horizontalAlignment=HAlign_Fill}}
  }
  HorizontalBox_1 {
   widgetClassPath=&HorizontalBox
   parent=VerticalBox_0
   widget=HorizontalBox_1
   slot {class=VerticalBoxSlot, properties {verticalAlignment=VAlign_Top}}
  }
  ProgressiveTags_Benefits {
   widgetClassPath=&/ProgressiveTagsBase
   bIsVariable=true
   parent=HorizontalBox_1
   widget=ProgressiveTags_Benefits
   properties {text="", toolTipText=Multi-Month Benefits, visibility=HitTestInvisible, background Color=<rgba 0.0275 0.0275 0.0275 1>}
   slot {class=HorizontalBoxSlot, properties {horizontalAlignment=HAlign_Left, verticalAlignment=VAlign_Center, padding=<ltrb 0 0 8 0>}}
  }
  ProgressiveTags_Stages {
   widgetClassPath=&ProgressiveTagsBase
   bIsVariable=true
   parent=HorizontalBox_1
   widget=ProgressiveTags_Stages
   properties {
    image Texture=/Game/Athena/HUD/Bacchus/HUDLayoutTool/HLT_Icons/T-HLT-CheckMark-32
    text=""
    toolTipText=Completed
    visibility=HitTestInvisible
    background Color=<rgba 0 0.147 0.063 1>
    text Color {colorUseRule=UseColor_Specified, specifiedColor=<rgba 0.1329 0.8796 0.3712 1>}
   }
   slot {class=HorizontalBoxSlot, properties {horizontalAlignment=HAlign_Left, verticalAlignment=VAlign_Center, padding=<ltrb 0 0 8 0>}}
  }
  ProgressiveTags_Subscribed {
   widgetClassPath=&ProgressiveTagsBase
   bIsVariable=true
   parent=HorizontalBox_1
   widget=ProgressiveTags_Subscribed
   properties {
    text=""
    visibility=HitTestInvisible
    background Color=<rgba 0.147 0.0284 0.0044 1>
    text Color {colorUseRule=UseColor_Specified, specifiedColor=<rgba 0.9734 0.9216 0.314 1>}
   }
   slot {class=HorizontalBoxSlot, properties {horizontalAlignment=HAlign_Left, verticalAlignment=VAlign_Center}}
  }
 ]
}
