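# Button tile for one reward in the Crew progressive cosmetic screen, showing item art, name stripe and locked/claimed icons.
# Click or focus writes its stage and reward to the selection view model.
# Stage and reward selection functions drive reveal, row-dimming, grow and highlight animations; hover plays only while unselected.
# Reward tile material loads asynchronously through an event dispatcher, and loading-state changes recolor overlays and card by subscription status.
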WidgetBlueprint /Crew/UI/ProgressiveCosmetic/ProgressiveItemWidget
animations [
 GrowTile {
  binding_count=3
  length_seconds=0.15
  object=&self:GrowTile
  sequence {bindings [
    SizeBoxSlot_0 {
     id=#1
     parent_id=#2
     possessed_class=SizeBoxSlot
     tracks [{
       type=MovieSceneMarginTrack
       display_name=Padding
       sections [{
         type=MovieSceneMarginSection
         channels [
          Left {type=MovieSceneScriptingFloatChannel, keys [[0,0],[3,-6]]}
          Top {type=MovieSceneScriptingFloatChannel, keys [[0,0],[3,-6]]}
          Right {type=MovieSceneScriptingFloatChannel, keys [[0,0],[3,-6]]}
          Bottom {type=MovieSceneScriptingFloatChannel, keys [[0,0],[3,-6]]}
         ]
        }]
      }]
    }
    Overlay_1 {
     id=#3
     possessed_class=Overlay
     tracks [{
       type=MovieSceneByteTrack
       display_name=Clipping
       sections [{type=MovieSceneByteSection, end_frame=3, start_frame=0, channels [None {type=MovieSceneScriptingByteChannel, const [0,1]}]}]
      }]
    }
    Overlay_0 {id=#2, possessed_class=Overlay, tracks []}
   ]}
 }
 Reveal_Hide {
  binding_count=3
  length_seconds=0.15
  object=&self:Reveal_Hide
  sequence {bindings [
    TileText_ItemName {
     id=#4
     display_name=Item_Name
     possessed_class=CommonVisibilitySwitcher
     tracks [
      {
       type=MovieScene2DTransformTrack
       sections [{
         type=MovieScene2DTransformSection
         channels [
          Translation.X {type=MovieSceneScriptingFloatChannel}
          Translation.Y {type=MovieSceneScriptingFloatChannel, keys [[0,15],[2,1.5],[3.0007,0]]}
          Angle {type=MovieSceneScriptingFloatChannel}
          Scale.X {type=MovieSceneScriptingFloatChannel}
          Scale.Y {type=MovieSceneScriptingFloatChannel}
          Shear.X {type=MovieSceneScriptingFloatChannel}
          Shear.Y {type=MovieSceneScriptingFloatChannel}
         ]
        }]
      }
      {
       type=MovieSceneFloatTrack
       display_name=Render Opacity
       sections [{type=MovieSceneFloatSection, channels [None {type=MovieSceneScriptingFloatChannel, keys [[0,0],[2,0],[3.0007,1]]}]}]
      }
     ]
    }
    HBox_locks {
     id=#5
     possessed_class=HorizontalBox
     tracks [{
       type=MovieSceneFloatTrack
       display_name=Render Opacity
       sections [{type=MovieSceneFloatSection, channels [None {type=MovieSceneScriptingFloatChannel, keys [[0,0],[1.9993,1]]}]}]
      }]
    }
    ItemName {
     id=#6
     display_name=Overlay
     possessed_class=Overlay
     tracks [{
       type=MovieScene2DTransformTrack
       sections [{
         type=MovieScene2DTransformSection
         channels [
          Translation.X {type=MovieSceneScriptingFloatChannel}
          Translation.Y {type=MovieSceneScriptingFloatChannel, keys [[0,25],[1.9993,0]]}
         ]
        }]
      }]
    }
   ]}
 }
 NotCurrentRow {
  binding_count=2
  length_seconds=0.25
  object=&self:NotCurrentRow
  sequence {bindings [
    Locked_Overlay {
     id=#7
     display_name=DisableTintOverlay
     possessed_class=Image
     tracks [{
       type=MovieSceneFloatTrack
       display_name=Render Opacity
       sections [{type=MovieSceneFloatSection, channels [None {type=MovieSceneScriptingFloatChannel, keys [[0,0.5],[5,0]]}]}]
      }]
    }
    OffscreenOverlay {
     id=#8
     possessed_class=Image
     tracks [{
       type=MovieSceneColorTrack
       display_name=Color and Opacity
       sections [{
         type=MovieSceneColorSection
         channels [
          Color.R {type=MovieSceneScriptingFloatChannel}
          Color.G {type=MovieSceneScriptingFloatChannel}
          Color.B {type=MovieSceneScriptingFloatChannel}
          Color.A {type=MovieSceneScriptingFloatChannel, keys [[0,0],[5,0.7]]}
         ]
        }]
      }]
    }
   ]}
 }
 Select {
  binding_count=3
  length_seconds=0.1
  object=&self:Select
  sequence {bindings [
    ItemCardImage {id=#9, possessed_class=Image, tracks [{type=MovieSceneWidgetMaterialTrack, sections []}]}
    TileText_ItemName {
     id=#10
     possessed_class=AthenaItemShopReactiveTileText
     tracks [
      {
       type=MovieSceneColorTrack
       display_name=Color and Opacity
       sections [{
         type=MovieSceneColorSection
         end_frame=2
         start_frame=0
         channels [
          Color.R {type=MovieSceneScriptingFloatChannel, keys [[0,1],[2,0.0116]]}
          Color.G {type=MovieSceneScriptingFloatChannel, keys [[0,1],[2,0.0116]]}
          Color.B {type=MovieSceneScriptingFloatChannel, keys [[0,1],[2,0.0116]]}
          Color.A {type=MovieSceneScriptingFloatChannel, const [1,2]}
         ]
        }]
      }
      {
       type=MovieScene2DTransformTrack
       sections [{
         type=MovieScene2DTransformSection
         end_frame=2
         start_frame=0
         channels [
          Translation.X {type=MovieSceneScriptingFloatChannel, const [0,1]}
          Translation.Y {type=MovieSceneScriptingFloatChannel, keys [[0,0],[2,-2]]}
         ]
        }]
      }
     ]
    }
    Image_TextStripe {
     id=#11
     display_name=Image_1
     possessed_class=Image
     tracks [{
       type=MovieScene2DTransformTrack
       sections [{
         type=MovieScene2DTransformSection
         channels [
          Scale.X {type=MovieSceneScriptingFloatChannel}
          Scale.Y {type=MovieSceneScriptingFloatChannel, keys [[0,1],[2,1.1]]}
         ]
        }]
      }]
    }
   ]}
 }
 Hover_Burst {
  binding_count=3
  length_seconds=2
  object=&self:Hover_Burst
  sequence {bindings [
    ItemCardImage {
     id=#12
     possessed_class=Image
     tracks [{
       type=MovieSceneWidgetMaterialTrack
       sections [{type=MovieSceneParameterSection, channels [HoverBurst {type=MovieSceneScriptingFloatChannel, keys [[0,0],[3,3],[40,0]]}]}]
      }]
    }
    DisableTintOverlay {
     id=#13
     possessed_class=Image
     tracks [{
       type=MovieSceneFloatTrack
       display_name=Render Opacity
       sections [{
         type=MovieSceneFloatSection
         end_frame=40
         start_frame=0
         channels [None {type=MovieSceneScriptingFloatChannel, keys [[0,0.5],[3,0.75],[40,1]]}]
        }]
      }]
    }
   ]}
 }
 Item_Hover {
  binding_count=1
  length_seconds=0.2
  object=&self:Item_Hover
  sequence {bindings [ItemCardImage {
     id=#14
     possessed_class=Image
     tracks [{
       type=MovieSceneWidgetMaterialTrack
       sections [{type=MovieSceneParameterSection, channels [Hover {type=MovieSceneScriptingFloatChannel, keys [[0,0],[4,1]]}]}]
      }]
    }]}
 }
]
graphs [
 EventGraph {nodes [
   "On Loading State Changed (ItemCardImage)"!() -> (OutputDelegate, then, bIsLoading:bool) {is=title K2Node_ComponentBoundEvent}
   > Update Subscribed Visuals!(
    execute
    self
    IsSubscribed:bool=IsSubscribed!(self:Crew VM Object Reference=Get CrewVM() -> (CrewVM:Crew VM Object Reference)) -> (ReturnValue:bool)
   ) -> (then)
   n1 = LoadTileMaterialItemImage_Event!() -> (OutputDelegate, then, ProgressiveRewardVM:Crew Progressive Reward VM Object Reference) {is=title K2Node_CustomEvent}
   > ApplyRewardTileMaterial!(
    execute:Exec
    ContextWidget:Fort Lazy Image Object Reference=Get ItemCardImage() -> (ItemCardImage:Fort Lazy Image Object Reference)
    RewardVM:Crew Progressive Reward VM Object Reference=@n1.ProgressiveRewardVM
   ) -> (then:Exec, OnOperationCompleted:Exec, Success:bool=~) {is=title K2Node_AsyncAction}
   n2 = Event On Clicked()
   n3 = Get CrewProgressiveSelectionVM() -> (CrewProgressiveSelectionVM:Crew Progressive Selection VM Object Reference)
   Set SelectedStage!(
    execute=[@n2.then, @n9.then]
    SelectedStage:Crew Progressive Stage VM Object Reference=Get CrewProgressiveStageVM() -> (CrewProgressiveStageVM:Crew Progressive Stage VM Object Reference)
    self:Crew Progressive Selection VM Object Reference=@n3
   ) -> (then, Output_Get:Crew Progressive Stage VM Object Reference)
   > Set SelectedReward!(
    execute
    SelectedReward:Crew Progressive Reward VM Object Reference=Get CrewProgressiveRewardVM() -> (CrewProgressiveRewardVM:Crew Progressive Reward VM Object Reference)
    self:Crew Progressive Selection VM Object Reference=@n3
   ) -> (then, Output_Get:Crew Progressive Reward VM Object Reference)
   Event OnInitialized()
   > Bind Event to Load Tile Material Item Image(self:Progressive Item Widget Object Reference, Delegate=@n1) {is=title K2Node_AddDelegate}
   Event Construct()
   > Play Stage Selected Anim()
   > Play Reward Selected Anim()
   Event On Hovered()
   n4 = Get Selected() -> (Selected:bool)
   > n5 = Branch(Condition=@n4)
   Set Hovered(execute=@n5.else, Hovered:bool=true) -> (Output_Get:bool)
   n6 = Get Item_hover() -> (Item_hover:Widget Animation Object Reference)
   > PlayAnimationForward(InAnimation=@n6)
   n7 = Get Hover_Burst() -> (Hover_Burst:Widget Animation Object Reference)
   > PlayAnimationForward(InAnimation=@n7)
   Event On Unhovered()
   > n8 = Branch(Condition=@n4)
   Set Hovered(execute=@n8.else, Hovered=false)
   > PlayAnimationReverse(InAnimation=@n6)
   > StopAnimation(InAnimation=@n7)
   n9 = Event On Focused() {is=title K2Node_Event}
  ]}
 LoadTileMaterialItemImage {nodes [LoadTileMaterialItemImage!() -> (then, ProgressiveRewardVM:Crew Progressive Reward VM Object Reference) {is=title K2Node_FunctionEntry}]}
 Play Reward Selected Anim {nodes [
   Play Reward Selected Anim() {is=title K2Node_FunctionEntry}
   > n1 = Branch(Condition=Get Selected())
   n2 = Get Image_TextStripe() -> (Image_TextStripe:Image Object Reference)
   > n3 = GetDynamicMaterial(self=@n2)
   > SetScalarParameterValue(self=@n3, ParameterName=Animate_BackingColor, Value=1)
   n4 = Get Select() -> (Select:Widget Animation Object Reference)
   > PlayAnimationForward(InAnimation=@n4)
   n5 = Get GrowTIle() -> (GrowTIle:Widget Animation Object Reference)
   > PlayAnimationForward(InAnimation=@n5)
   > Branch(Condition=Get Hovered() -> (Hovered:bool))
   > Set Hovered(Hovered=false)
   n6 = Get Item_hover()
   > PlayAnimationForward(InAnimation=@n6)
   n7 = Get Hover_Burst()
   > PlayAnimationForward(InAnimation=@n7)
   n8 = GetDynamicMaterial(execute=@n1.else, self=@n2)
   > SetScalarParameterValue(self=@n8, ParameterName=Animate_BackingColor, Value=0)
   > PlayAnimationReverse(InAnimation=@n4, PlaybackSpeed=2)
   > PlayAnimationReverse(InAnimation=@n5, PlaybackSpeed=2)
   > PlayAnimationReverse(InAnimation=@n6)
   > StopAnimation(InAnimation=@n7)
  ]}
 Play Stage Selected Anim {nodes [
   Play Stage Selected Anim() {is=title K2Node_FunctionEntry}
   n1 = Get Stage Was Selected() -> (Stage Was Selected:bool)
   > PlayAnimation(
    InAnimation=Get Reveal_Hide() -> (Reveal_Hide:Widget Animation Object Reference)
    PlayMode=Select(Option 0:EUMGSequencePlayMode Enum=Reverse, Option 1:EUMGSequencePlayMode Enum, Index=@n1) -> (ReturnValue:EUMGSequencePlayMode Enum)
   )
   > PlayAnimation(InAnimation=Get NotCurrentRow() -> (NotCurrentRow:Widget Animation Object Reference), PlayMode=Select(Option 1=Reverse, Index=@n1))
  ]}
 Reward Selected {nodes [
   n1 = Reward Selected!() -> (then, Selected:bool=~) {is=title K2Node_FunctionEntry}
   n2 = Set Selected(execute=@n4.then, Selected:bool=@n1) -> (Output_Get:bool)
   n3 = SetIsSelected!(execute, self:Common Button Base Object Reference, InSelected:bool=@n2, bGiveClickFeedback:bool=false) -> (then)
   Play Reward Selected Anim(execute=[@n3.then, @n2.then]) {is=title K2Node_CallFunction}
   n4 = Branch(execute=@n1.then, Condition="Not Equal (Boolean)"(A=@n1, B=Get Selected()) -> (ReturnValue=~))
  ]}
 Set CrewProgressiveRewardVM {nodes [
   n1 = Set CrewProgressiveRewardVM!() -> (then, ProgressiveRewardVM:Crew Progressive Reward VM Object Reference) {is=title K2Node_FunctionEntry}
   > Call Load Tile Material Item Image!(execute:Exec, self:Progressive Item Widget Object Reference, ProgressiveRewardVM:Crew Progressive Reward VM Object Reference=@n1) -> (then:Exec) {is=title K2Node_CallDelegate}
  ]}
 Stage Selected {nodes [
   n1 = Stage Selected!() -> (then, Selected:bool=~) {is=title K2Node_FunctionEntry}
   > Branch(Condition="Not Equal (Boolean)"(A=@n1, B=Get Stage Was Selected()) -> (ReturnValue=~))
   > Set Stage Was Selected(Stage Was Selected:bool=@n1) -> (Output_Get:bool)
   > Play Stage Selected Anim() {is=title K2Node_CallFunction}
  ]}
 Update Subscribed Visuals {nodes [
   n1 = Update Subscribed Visuals!() -> (then, IsSubscribed:bool=~) {is=title K2Node_FunctionEntry}
   > SetColorAndOpacity(
    self:Image Object Reference=[Get OffscreenOverlay() -> (OffscreenOverlay:Image Object Reference), Get DisableTintOverlay() -> (DisableTintOverlay:Image Object Reference)]
    InColorAndOpacity:Linear Color Structure=SelectColor(A=(R=0.147,G=0.0284,B=0.0044,A=0.5), B=(R=0.0027,G=0.0561,B=0.4564,A=0.502), bPickA=@n1)
   )
   > n2 = GetDynamicMaterial(self=Get ItemCardImage())
   n3 = Get IsSubscribed() -> (IsSubscribed:bool)
   > SetVectorParameterValue(self=@n2, ParameterName=Premium_A, Value=SelectColor(A=(R=0.3467,G=0.1441,B=0.0331,A=1), B=(R=0.0353,G=0.0353,B=0.3608,A=1), bPickA=@n3))
   > SetVectorParameterValue(
    self=@n2
    ParameterName=Premium_Owned_A
    Value=SelectColor(A=(R=0.3005,G=0.1095,B=0.0203,A=1), B=(R=0.0118,G=0.0863,B=0.8314,A=1), bPickA=@n3)
   )
   > SetVectorParameterValue(self=@n2, ParameterName=Premium_B, Value=SelectColor(A=(R=1,G=0.7835,B=0.0976,A=1), B=(R=0,G=0.4678,B=1,A=1), bPickA=@n3))
   > SetVectorParameterValue(self=@n2, ParameterName=Premium_Owned_B, Value=SelectColor(A=(R=0.807,G=0.6867,B=0.3005,A=1), B=(R=0.2157,G=0.6588,B=1,A=1), bPickA=@n3))
   > SetVectorParameterValue(self=@n2, ParameterName=FallOff_Color, Value=SelectColor(A=(R=0.276,G=0.0794,B=0.0197,A=1), B=(R=0.0612,G=0.2232,B=0.8963,A=1), bPickA=@n3))
  ]}
]
variables [
 Hovered {type=boolean}
 LoadTileMaterialItemImage {}
 Selected {type=boolean}
 Stage Was Selected {type=boolean}
]
widgets {
 info {inheritedWidgetCount=1, parentClass=&CommonButtonBase, rootWidgetClass=&Overlay, widgetCount=13}
 widgets [
  Overlay_1 {widgetClassPath=&Overlay, widget=Overlay_1}
  SizeBox_0 {
   widgetClassPath=&SizeBox
   parent=Overlay_1
   widget=SizeBox_0
   properties {bOverride_HeightOverride=true, bOverride_WidthOverride=true, heightOverride=176, widthOverride=256}
   slot {class=OverlaySlot}
  }
  Overlay_0 {widgetClassPath=&Overlay, parent=SizeBox_0, widget=Overlay_0, properties {clipping=ClipToBounds}, slot {class=SizeBoxSlot}}
  ItemCardImage {
   widgetClassPath=&FortLazyImage
   bIsVariable=true
   parent=Overlay_0
   widget=ItemCardImage
   properties {brush {
     bIsDynamicallyLoaded=false
     drawAs=Image
     imageType=NoImage
     mirroring=NoMirror
     resourceName=None
     resourceObject=/FortniteCoreUI/UI/Subscriptions/Art/M_UI_ProgressiveItemTile_Empty
     tiling=NoTile
     imageSize=<32 32>
     margin=<ltrb 0 0 0 0>
     outlineSettings {
      bUseBrushTransparency=false
      roundingType=HalfHeightRadius
      width=0
      color {colorUseRule=UseColor_Specified, specifiedColor=<rgba 0 0 0 0>}
      cornerRadii=<0 0 0 0>
     }
     tintColor {colorUseRule=UseColor_Specified, specifiedColor=<rgba 1 1 1 1>}
     uVRegion {bIsValid=false, max=<0 0>, min=<0 0>}
    }}
   slot {class=OverlaySlot, properties {horizontalAlignment=HAlign_Fill, verticalAlignment=VAlign_Fill}}
  }
  DisableTintOverlay {
   widgetClassPath=&Image
   bIsVariable=true
   parent=Overlay_0
   widget=DisableTintOverlay
   properties {renderOpacity=0.7, visibility=SelfHitTestInvisible, colorAndOpacity=<rgba 0.0027 0.0561 0.4564 0.502>}
   slot {class=OverlaySlot, properties {horizontalAlignment=HAlign_Fill, verticalAlignment=VAlign_Fill}}
  }
  ItemName {
   widgetClassPath=&Overlay
   parent=Overlay_0
   widget=ItemName
   slot {class=OverlaySlot, properties {horizontalAlignment=HAlign_Fill, verticalAlignment=VAlign_Bottom}}
  }
  Image_TextStripe {
   widgetClassPath=&Image
   bIsVariable=true
   parent=ItemName
   widget=Image_TextStripe
   properties {
    brush {
     bIsDynamicallyLoaded=false
     drawAs=Box
     imageType=FullColor
     mirroring=NoMirror
     resourceName=None
     resourceObject=/Game/UI/Frontend/ItemShop/Art/M_UI_OfferTile_UI_Layer_NoRarity
     tiling=NoTile
     imageSize=<36 46>
     margin=<ltrb 0 0.25 0 0>
     outlineSettings {
      bUseBrushTransparency=false
      roundingType=HalfHeightRadius
      width=0
      color {colorUseRule=UseColor_Specified, specifiedColor=<rgba 0 0 0 0>}
      cornerRadii=<0 0 0 0>
     }
     tintColor {colorUseRule=UseColor_Specified, specifiedColor=<rgba 1 1 1 1>}
     uVRegion {bIsValid=false, max=<0 0>, min=<0 0>}
    }
    renderTransformPivot=<0.5 1>
   }
   slot {class=OverlaySlot, properties {horizontalAlignment=HAlign_Fill, verticalAlignment=VAlign_Fill}}
  }
  TileText_ItemName {
   widgetClassPath=/Script/FortniteUI.AthenaItemShopReactiveTileText
   parent=ItemName
   widget=TileText_ItemName
   properties {
    justification=Center
    lineHeightPercentage=0.85
    maxCharacterCountBeforeScaling=17
    maxScaleDownPercentage=0.8
    mobileFontSizeMultiplier=1.2
    scrollStyle=&TextScrollStyle-Base_C
    text=wwwwwwwwwwwwwwwwwwwwww
    textTransformPolicy=ToUpper
    font {
     bForceMonospaced=false
     bMaterialIsStencil=false
     fontMaterial=null
     fontObject=&BurbankSmall
     letterSpacing=0
     monospacedWidth=1
     size=16
     skewAmount=0
     typefaceFontName=Black
     outlineSettings {
      bApplyOutlineToDropShadows=false
      bMiteredCorners=false
      bSeparateFillAlpha=false
      outlineMaterial=null
      outlineSize=0
      outlineColor=<rgba 0 0 0 1>
     }
    }
    renderTransform {angle=0, scale=<1 1>, shear=<-12 0>, translation=<0 0>}
    shadowOffset=<0 0>
   }
   slot {class=OverlaySlot, properties {horizontalAlignment=HAlign_Fill, verticalAlignment=VAlign_Center, padding=<ltrb 15 1 15 0>}}
  }
  HBox_locks {
   widgetClassPath=&HorizontalBox
   parent=Overlay_0
   widget=HBox_locks
   slot {class=OverlaySlot, properties {horizontalAlignment=HAlign_Right, padding=<ltrb 4 4 4 4>}}
  }
  Widget_LockedImage {
   widgetClassPath=/Script/FortniteUI.FortMobileImage
   bIsVariable=true
   parent=HBox_locks
   widget=Widget_LockedImage
   properties {
    mobileSizeMultiplier=1.4
    brush {
     bIsDynamicallyLoaded=false
     drawAs=Image
     imageType=NoImage
     mirroring=NoMirror
     resourceName=None
     resourceObject=/FortUILibrary/Foundation/Art_ToolKit/AngledBoxes/M_UI_AngledLockedBox_Red
     tiling=NoTile
     imageSize=<30 30>
     margin=<ltrb 0 0 0 0>
     outlineSettings {
      bUseBrushTransparency=false
      roundingType=HalfHeightRadius
      width=0
      color {colorUseRule=UseColor_Specified, specifiedColor=<rgba 0 0 0 0>}
      cornerRadii=<0 0 0 0>
     }
     tintColor {colorUseRule=UseColor_Specified, specifiedColor=<rgba 1 1 1 1>}
     uVRegion {bIsValid=false, max=<0 0>, min=<0 0>}
    }
   }
   slot {class=HorizontalBoxSlot, properties {horizontalAlignment=HAlign_Right}}
  }
  Widget_ClaimedImage {
   widgetClassPath=&FortMobileImage
   bIsVariable=true
   parent=HBox_locks
   widget=Widget_ClaimedImage
   properties {
    mobileSizeMultiplier=1.4
    brush {
     bIsDynamicallyLoaded=false
     drawAs=Image
     imageType=NoImage
     mirroring=NoMirror
     resourceName=None
     resourceObject=/FortUILibrary/Foundation/Art_ToolKit/AngledBoxes/M_UI_AngledCheckBox_Crafting
     tiling=NoTile
     imageSize=<30 30>
     margin=<ltrb 0 0 0 0>
     outlineSettings {
      bUseBrushTransparency=false
      roundingType=HalfHeightRadius
      width=0
      color {colorUseRule=UseColor_Specified, specifiedColor=<rgba 0 0 0 0>}
      cornerRadii=<0 0 0 0>
     }
     tintColor {colorUseRule=UseColor_Specified, specifiedColor=<rgba 1 1 1 1>}
     uVRegion {bIsValid=false, max=<0 0>, min=<0 0>}
    }
   }
   slot {class=HorizontalBoxSlot, properties {horizontalAlignment=HAlign_Right}}
  }
  OffscreenOverlay {
   widgetClassPath=&Image
   bIsVariable=true
   parent=Overlay_0
   widget=OffscreenOverlay
   properties {visibility=SelfHitTestInvisible, colorAndOpacity=<rgba 0.0027 0.0561 0.4564 0>}
   slot {class=OverlaySlot, properties {horizontalAlignment=HAlign_Fill, verticalAlignment=VAlign_Fill}}
  }
  InputActionWidget {widgetClassPath=&CommonActionWidget, bInherited=true, widget=null}
 ]
}
