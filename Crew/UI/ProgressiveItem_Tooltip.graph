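# Tutorial-styled tooltip widget for Crew UI progressive items, with rich text, optional icon, burst backdrop and arrow.
# PreConstruct applies exposed text, wrapping, icon, arrow alignment and transform, and padding; Construct starts it collapsed.
# ToggleTooltip ignores unchanged state, otherwise plays Show forward or reversed, setting final visibility when it finishes.
# The burstLoop and FloatLoop animations are defined but not played by this graph.
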WidgetBlueprint /Crew/UI/ProgressiveItem_Tooltip
animations [
 Show {
  binding_count=4
  length_seconds=0.35
  object=&self:Show
  sequence {bindings [
    Overlay_0 {
     id=#1
     possessed_class=Overlay
     tracks [{
       type=MovieSceneFloatTrack
       display_name=Render Opacity
       sections [{type=MovieSceneFloatSection, channels [None {type=MovieSceneScriptingFloatChannel, keys [[0,0],[2,0],[5,1]]}]}]
      }]
    }
    SizeBox_0 {
     id=#2
     possessed_class=SizeBox
     tracks [{
       type=MovieScene2DTransformTrack
       sections [{type=MovieScene2DTransformSection, channels [Scale.Y {type=MovieSceneScriptingFloatChannel, keys [[2,0],[5,1]]}]}]
      }]
    }
    Icon {
     id=#3
     possessed_class=Image
     tracks [{
       type=MovieSceneFloatTrack
       display_name=Render Opacity
       sections [{type=MovieSceneFloatSection, channels [None {type=MovieSceneScriptingFloatChannel, keys [[3,0],[6,1]]}]}]
      }]
    }
    Text_Tooltip {
     id=#4
     possessed_class=CommonRichTextBlock
     tracks [{
       type=MovieSceneFloatTrack
       display_name=Render Opacity
       sections [{type=MovieSceneFloatSection, channels [None {type=MovieSceneScriptingFloatChannel, keys [[4,0],[7,1]]}]}]
      }]
    }
   ]}
 }
 burstLoop {
  binding_count=1
  length_seconds=1.75
  object=&self:burstLoop
  sequence {bindings [Image_burst {
     id=#5
     possessed_class=Image
     tracks [
      {
       type=MovieSceneFloatTrack
       display_name=Render Opacity
       sections [{type=MovieSceneFloatSection, channels [None {type=MovieSceneScriptingFloatChannel, keys [[0,0],[1,1],[20,0],[35,0]]}]}]
      }
      {
       type=MovieScene2DTransformTrack
       sections [{
         type=MovieScene2DTransformSection
         channels [
          Translation.X {type=MovieSceneScriptingFloatChannel}
          Translation.Y {type=MovieSceneScriptingFloatChannel}
          Angle {type=MovieSceneScriptingFloatChannel}
          Scale.X {type=MovieSceneScriptingFloatChannel, keys [[1,1],[20,1.1]]}
          Scale.Y {type=MovieSceneScriptingFloatChannel, keys [[1,1],[20,1.4]]}
          Shear.X {type=MovieSceneScriptingFloatChannel}
          Shear.Y {type=MovieSceneScriptingFloatChannel}
         ]
        }]
      }
      {
       type=MovieSceneWidgetMaterialTrack
       sections [{
         type=MovieSceneParameterSection
         channels [
          Border_Offset_Top {type=MovieSceneScriptingFloatChannel, keys [[1,-0.1345],[20,-0.0745]]}
          Border_Offset_Right {type=MovieSceneScriptingFloatChannel, keys [[1,-0.1445],[20,-0.0845]]}
          Border_Offset_Left {type=MovieSceneScriptingFloatChannel, keys [[1,-0.1064],[20,-0.0464]]}
          Border_Offset_Bottom {type=MovieSceneScriptingFloatChannel, keys [[1,-0.1155],[20,-0.0555]]}
         ]
        }]
      }
     ]
    }]}
 }
 FloatLoop {
  binding_count=3
  length_seconds=2.5
  object=&self:FloatLoop
  sequence {bindings [
    Image_burst {
     id=#6
     possessed_class=Image
     tracks [
      {
       type=MovieSceneFloatTrack
       display_name=Render Opacity
       sections [{type=MovieSceneFloatSection, channels [None {type=MovieSceneScriptingFloatChannel, keys [[0,0],[1,1],[20,0],[35,0]]}]}]
      }
      {
       type=MovieScene2DTransformTrack
       sections [{
         type=MovieScene2DTransformSection
         channels [
          Translation.X {type=MovieSceneScriptingFloatChannel, const [0,3]}
          Translation.Y {type=MovieSceneScriptingFloatChannel, keys [[0,0],[25,4],[50,0]]}
          Angle {type=MovieSceneScriptingFloatChannel}
          Scale.X {type=MovieSceneScriptingFloatChannel, keys [[1,1],[20,1.1]]}
          Scale.Y {type=MovieSceneScriptingFloatChannel, keys [[1,1],[20,1.4]]}
          Shear.X {type=MovieSceneScriptingFloatChannel}
          Shear.Y {type=MovieSceneScriptingFloatChannel}
         ]
        }]
      }
      {
       type=MovieSceneWidgetMaterialTrack
       sections [{
         type=MovieSceneParameterSection
         channels [
          Border_Offset_Top {type=MovieSceneScriptingFloatChannel, keys [[1,-0.1345],[20,-0.0745]]}
          Border_Offset_Right {type=MovieSceneScriptingFloatChannel, keys [[1,-0.1445],[20,-0.0845]]}
          Border_Offset_Left {type=MovieSceneScriptingFloatChannel, keys [[1,-0.1064],[20,-0.0464]]}
          Border_Offset_Bottom {type=MovieSceneScriptingFloatChannel, keys [[1,-0.1155],[20,-0.0555]]}
         ]
        }]
      }
     ]
    }
    Border_79 {
     id=#7
     possessed_class=Border
     tracks [{
       type=MovieScene2DTransformTrack
       sections [{
         type=MovieScene2DTransformSection
         end_frame=50
         start_frame=0
         channels [
          Translation.X {type=MovieSceneScriptingFloatChannel, const [0,3]}
          Translation.Y {type=MovieSceneScriptingFloatChannel, keys [[0,0],[25,4],[50,0]]}
          Angle {type=MovieSceneScriptingFloatChannel}
          Scale.X {type=MovieSceneScriptingFloatChannel}
          Scale.Y {type=MovieSceneScriptingFloatChannel}
          Shear.X {type=MovieSceneScriptingFloatChannel}
          Shear.Y {type=MovieSceneScriptingFloatChannel}
         ]
        }]
      }]
    }
    SizeBox_0 {
     id=#8
     possessed_class=SizeBox
     tracks [{
       type=MovieScene2DTransformTrack
       sections [{
         type=MovieScene2DTransformSection
         end_frame=50
         start_frame=0
         channels [
          Translation.X {type=MovieSceneScriptingFloatChannel, const [0,3]}
          Translation.Y {type=MovieSceneScriptingFloatChannel, keys [[0,0],[25,8],[50,0]]}
          Angle {type=MovieSceneScriptingFloatChannel, const [0,1]}
          Scale.X {type=MovieSceneScriptingFloatChannel, const [1,1]}
          Scale.Y {type=MovieSceneScriptingFloatChannel, const [1,1]}
          Shear.X {type=MovieSceneScriptingFloatChannel, const [0,1]}
          Shear.Y {type=MovieSceneScriptingFloatChannel, const [0,1]}
         ]
        }]
      }]
    }
   ]}
 }
]
graphs [
 EventGraph {nodes [
   n1 = ToggleTooltip!() -> (OutputDelegate, then, Visible:bool=~) {is=title K2Node_CustomEvent}
   > Branch(Condition="Not Equal (Boolean)"(A=@n1.Visible, B=Get ShouldShowToolTip() -> (ShouldShowToolTip:bool)) -> (ReturnValue=~))
   > SetVisibility(InVisibility=HitTestInvisible)
   > Set ShouldShowToolTip(ShouldShowToolTip:bool=@n1.Visible) -> (Output_Get:bool)
   > n2 = "Play Animation with Finished event (legacy)"!(
    execute:Exec
    Widget:User Widget Object Reference=Self-Reference()
    InAnimation:Widget Animation Object Reference=Get Show() -> (Show:Widget Animation Object Reference)
    StartAtTime:float=0
    NumLoopsToPlay:int=1
    PlayMode:EUMGSequencePlayMode Enum=Select(Option 0:EUMGSequencePlayMode Enum=Reverse, Option 1:EUMGSequencePlayMode Enum, Index=@n1.Visible) -> (ReturnValue:EUMGSequencePlayMode Enum)
    PlaybackSpeed:float=1
   ) -> (then:Exec, Result:UMGSequence Player Object Reference, Finished:Exec) {is=title K2Node_PlayAnimation}
   SetVisibility(
    execute=@n2.Finished
    InVisibility=Select(Option 0:ESlateVisibility Enum=Collapsed, Option 1:ESlateVisibility Enum=HitTestInvisible, Index=@n1.Visible) -> (ReturnValue:ESlateVisibility Enum)
   )
   Event PreConstruct()
   > Update Text!(execute, self, InText:"Text (by ref)"=Get Tooltip Text() -> (Tooltip Text:Text)) -> (then)
   > SetAutoWrapText!(
    execute
    self:Rich Text Block Object Reference=Get Text_Tooltip() -> (Text_Tooltip:Common Rich Text Block Object Reference)
    InAutoTextWrap:bool=Get In Auto Wrap Text() -> (In Auto Wrap Text:bool)
   ) -> (then)
   n3 = Get Icon() -> (Icon:Image Object Reference)
   > SetBrush(self=@n3, InBrush=Get In Brush() -> (In Brush:Slate Brush Structure))
   > SetVisibility(self=@n3, InVisibility=Select(Option 0=Collapsed, Option 1=SelfHitTestInvisible, Index=Get Show Icon() -> (Show Icon:bool)))
   n4 = SlotAsOverlaySlot(Widget=Get Image_Arrow() -> (Image_Arrow:Image Object Reference))
   > SetHorizontalAlignment(self=@n4, InHorizontalAlignment=Get In Horizontal Alignment() -> (In Horizontal Alignment:EHorizontalAlignment Enum=HAlign_Fill))
   > SetVerticalAlignment(self=@n4, InVerticalAlignment=Get In Vertical Alignment() -> (In Vertical Alignment:EVerticalAlignment Enum=VAlign_Fill))
   > SetRenderTransform!(
    execute
    self:Widget Object Reference=Get Image_Arrow()
    InTransform:Widget Transform Structure=Get In Transform() -> (In Transform:Widget Transform Structure)
   ) -> (then)
   > SetPadding!(
    execute
    self:Border Slot Object Reference=SlotAsBorderSlot!(Widget:Widget Object Reference=Get HorizontalBoxcontent() -> (HorizontalBoxcontent:Horizontal Box Object Reference)) -> (ReturnValue:Border Slot Object Reference)
    InPadding:Margin Structure=Get Content Padding() -> (Content Padding:Margin Structure)
   ) -> (then)
   Event Construct()
   > Set ShouldShowToolTip(ShouldShowToolTip=false)
   > SetVisibility(InVisibility=Collapsed)
  ]}
 Update Text {nodes [
   n1 = Update Text!() -> (then, InText:"Text (by ref)") {is=title K2Node_FunctionEntry}
   > SetText(self=Get Text_Tooltip(), InText=@n1)
  ]}
]
variables [
 Content Padding {type=Margin, default=<ltrb 24 2 24 4>}
 In Auto Wrap Text {type=boolean}
 In Brush {
  type=SlateBrush
  default {
   bIsDynamicallyLoaded=false
   drawAs=Image
   imageType=NoImage
   mirroring=NoMirror
   resourceName=None
   resourceObject=/FortUILibrary/Foundation/Art_ToolKit/AngledBoxes/M_UI_AngledLockedBox_Red
   tiling=NoTile
   imageSize=<50 50>
   margin=<ltrb 0 0 0 0>
   outlineSettings {
    bUseBrushTransparency=false
    roundingType=HalfHeightRadius
    width=0
    color {colorUseRule=UseColor_Specified, specifiedColor=<rgba 0 0 0 0>}
    cornerRadii=<0 0 0 1>
   }
   tintColor {colorUseRule=UseColor_Specified, specifiedColor=<rgba 1 1 1 1>}
   uVRegion {bIsValid=false, max=<0 0>, min=<0 0>}
  }
 }
 In Horizontal Alignment {type=EHorizontalAlignment, default=HAlign_Left}
 In Transform {type=WidgetTransform, default {angle=0, scale=<1 1>, shear=<0 0>, translation=<0 0>}}
 In Vertical Alignment {type=EVerticalAlignment, default=VAlign_Center}
 ShouldShowToolTip {type=boolean}
 Show Icon {type=boolean, default=true}
 Tooltip Text {type=string, default="Placeholder <img id=\"MTXSmall\"/>\r\n<text color=\"fff\">tutorial</> text."}
]
widgets {
 info {inheritedWidgetCount=0, parentClass=&UserWidget, rootWidgetClass=&SizeBox, widgetCount=8}
 widgets [
  SizeBox_0 {
   widgetClassPath=&SizeBox
   widget=SizeBox_0
   properties {bOverride_MinDesiredWidth=true, heightOverride=90, maxDesiredWidth=400, minDesiredHeight=96, minDesiredWidth=400, widthOverride=490}
  }
  Overlay_0 {widgetClassPath=&Overlay, parent=SizeBox_0, widget=Overlay_0, slot {class=SizeBoxSlot}}
  Image_burst {
   widgetClassPath=&Image
   parent=Overlay_0
   widget=Image_burst
   properties {
    visibility=SelfHitTestInvisible
    brush {
     bIsDynamicallyLoaded=false
     drawAs=Box
     imageType=NoImage
     mirroring=NoMirror
     resourceName=None
     resourceObject=/FortUILibrary/Foundation/Art_ToolKit/AngledBoxes/M_UI_AngledBox_Tuto_burst
     tiling=NoTile
     imageSize=<90 90>
     margin=<ltrb 0.5 0 0.5 0>
     outlineSettings {
      bUseBrushTransparency=false
      roundingType=HalfHeightRadius
      width=0
      color {colorUseRule=UseColor_Specified, specifiedColor=<rgba 0 0 0 0>}
      cornerRadii=<0 0 0 0>
     }
     tintColor {colorUseRule=UseColor_Specified, specifiedColor=<rgba 1 1 1 1>}
     uVRegion {bIsValid=false, max=<0 0>, min=<0 0>}
    }
   }
   slot {class=OverlaySlot, properties {horizontalAlignment=HAlign_Fill, verticalAlignment=VAlign_Fill}}
  }
  Border_79 {
   widgetClassPath=&Border
   parent=Overlay_0
   widget=Border_79
   properties {
    background {
     bIsDynamicallyLoaded=false
     drawAs=Box
     imageType=NoImage
     mirroring=NoMirror
     resourceName=None
     resourceObject=/FortUILibrary/Foundation/Art_ToolKit/AngledBoxes/M_UI_AngledBox_Tuto_Inst
     tiling=NoTile
     imageSize=<90 90>
     margin=<ltrb 0.5 0 0.5 0>
     outlineSettings {
      bUseBrushTransparency=false
      roundingType=HalfHeightRadius
      width=0
      color {colorUseRule=UseColor_Specified, specifiedColor=<rgba 0 0 0 0>}
      cornerRadii=<0 0 0 0>
     }
     tintColor {colorUseRule=UseColor_Specified, specifiedColor=<rgba 1 1 1 1>}
     uVRegion {bIsValid=false, max=<0 0>, min=<0 0>}
    }
    padding=<ltrb 24 2 24 2>
   }
   slot {class=OverlaySlot, properties {horizontalAlignment=HAlign_Fill, verticalAlignment=VAlign_Fill}}
  }
  HorizontalBoxcontent {
   widgetClassPath=&HorizontalBox
   bIsVariable=true
   parent=Border_79
   widget=HorizontalBoxcontent
   slot {class=BorderSlot, properties {padding=<ltrb 24 2 24 2>}}
  }
  Text_Tooltip {
   widgetClassPath=&CommonRichTextBlock
   bIsVariable=true
   parent=HorizontalBoxcontent
   widget=Text_Tooltip
   properties {
    autoWrapText=true
    bOverrideDefaultStyle=true
    justification=Center
    lineHeightPercentage=0.95
    mobileTextBlockScale=1.2
    text="Placeholder <img id=\"MTXSmall\"/><text color=\"fff\">tutorial</> text goes here and then something"
    wrapTextAt=400
    decoratorClasses [/FortniteCoreUI/Foundation/RichTextData/CommonRichTextDecorators.CommonRichTextDecorators_C]
    defaultTextStyleOverride {
     overflowPolicy=Clip
     transformPolicy=None
     colorAndOpacity {colorUseRule=UseColor_Specified, specifiedColor=<rgba 0.4219 0.7481 1 1>}
     font {
      bForceMonospaced=false
      bMaterialIsStencil=false
      fontMaterial=null
      fontObject=&Burbank
      letterSpacing=0
      monospacedWidth=1
      size=14
      skewAmount=0
      typefaceFontName=Burbank-PlayerPower
      outlineSettings {
       bApplyOutlineToDropShadows=false
       bMiteredCorners=false
       bSeparateFillAlpha=false
       outlineMaterial=null
       outlineSize=0
       outlineColor=<rgba 0 0 0 1>
      }
     }
     highlightColor {colorUseRule=UseColor_Specified, specifiedColor=<rgba 0 0 0 1>}
     highlightShape {
      bIsDynamicallyLoaded=false
      drawAs=Image
      imageType=NoImage
      mirroring=NoMirror
      resourceName=None
      resourceObject=null
      tiling=NoTile
      imageSize=<32 32>
      margin=<ltrb 0 0 0 0>
      outlineSettings {
       bUseBrushTransparency=false
       roundingType=HalfHeightRadius
       width=0
       color {colorUseRule=UseColor_Specified, specifiedColor=<rgba 0 0 0 0>}
       cornerRadii=<0 0 0 0>
      }
      tintColor {colorUseRule=UseColor_Specified, specifiedColor=<rgba 1 1 1 1>}
      uVRegion {bIsValid=false, max=<0 0>, min=<0 0>}
     }
     selectedBackgroundColor {colorUseRule=UseColor_Foreground, specifiedColor=<rgba 1 0 1 1>}
     shadowColorAndOpacity=<rgba 0 0 0 1>
     shadowOffset=<0 0>
     strikeBrush {
      bIsDynamicallyLoaded=false
      drawAs=Image
      imageType=NoImage
      mirroring=NoMirror
      resourceName=None
      resourceObject=null
      tiling=NoTile
      imageSize=<32 32>
      margin=<ltrb 0 0 0 0>
      outlineSettings {
       bUseBrushTransparency=false
       roundingType=HalfHeightRadius
       width=0
       color {colorUseRule=UseColor_Specified, specifiedColor=<rgba 0 0 0 0>}
       cornerRadii=<0 0 0 0>
      }
      tintColor {colorUseRule=UseColor_Specified, specifiedColor=<rgba 1 1 1 1>}
      uVRegion {bIsValid=false, max=<0 0>, min=<0 0>}
     }
     underlineBrush {
      bIsDynamicallyLoaded=false
      drawAs=Image
      imageType=NoImage
      mirroring=NoMirror
      resourceName=None
      resourceObject=null
      tiling=NoTile
      imageSize=<32 32>
      margin=<ltrb 0 0 0 0>
      outlineSettings {
       bUseBrushTransparency=false
       roundingType=HalfHeightRadius
       width=0
       color {colorUseRule=UseColor_Specified, specifiedColor=<rgba 0 0 0 0>}
       cornerRadii=<0 0 0 0>
      }
      tintColor {colorUseRule=UseColor_Specified, specifiedColor=<rgba 1 1 1 1>}
      uVRegion {bIsValid=false, max=<0 0>, min=<0 0>}
     }
    }
    renderTransform {angle=0, scale=<1 1>, shear=<-8 0>, translation=<0 0>}
   }
   slot {class=HorizontalBoxSlot, properties {verticalAlignment=VAlign_Center, padding=<ltrb 0 17 0 17>, size {sizeRule=Fill, value=1}}}
  }
  Icon {
   widgetClassPath=&Image
   bIsVariable=true
   parent=HorizontalBoxcontent
   widget=Icon
   properties {brush {
     bIsDynamicallyLoaded=false
     drawAs=Image
     imageType=NoImage
     mirroring=NoMirror
     resourceName=None
     resourceObject=&M_UI_AngledLockedBox_Red
     tiling=NoTile
     imageSize=<50 50>
     margin=<ltrb 0 0 0 0>
     outlineSettings {
      bUseBrushTransparency=false
      roundingType=HalfHeightRadius
      width=0
      color {colorUseRule=UseColor_Specified, specifiedColor=<rgba 0 0 0 0>}
      cornerRadii=<0 0 0 0>
     }
     tintColor {colorUseRule=UseColor_Specified, specifiedColor=<rgba 1 1 1 1>}
     uVRegion {bIsValid=false, max=<0 0>, min=<0 0>}
    }}
   slot {class=HorizontalBoxSlot, properties {horizontalAlignment=HAlign_Right, verticalAlignment=VAlign_Center, padding=<ltrb 16 17 0 17>}}
  }
  Image_Arrow {
   widgetClassPath=&Image
   bIsVariable=true
   parent=Overlay_0
   widget=Image_Arrow
   properties {
    visibility=SelfHitTestInvisible
    brush {
     bIsDynamicallyLoaded=false
     drawAs=Image
     imageType=NoImage
     mirroring=NoMirror
     resourceName=None
     resourceObject=/Game/Athena/HUD/T_UI_BP_TutoArrow_64
     tiling=NoTile
     imageSize=<32 32>
     margin=<ltrb 0 0 0 0>
     outlineSettings {
      bUseBrushTransparency=false
      roundingType=HalfHeightRadius
      width=0
      color {colorUseRule=UseColor_Specified, specifiedColor=<rgba 0 0 0 0>}
      cornerRadii=<0 0 0 0>
     }
     tintColor {colorUseRule=UseColor_Specified, specifiedColor=<rgba 1 1 1 1>}
     uVRegion {bIsValid=false, max=<0 0>, min=<0 0>}
    }
   }
   slot {class=OverlaySlot, properties {horizontalAlignment=HAlign_Right, verticalAlignment=VAlign_Center}}
  }
 ]
}
