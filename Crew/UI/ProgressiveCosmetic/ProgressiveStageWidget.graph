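# CommonButton row for one stage of a Crew progressive cosmetic set, showing up to two reward items with a plus separator.
# Stage data fills the item widgets and shows an unlock or subscribe-by tooltip on the next unowned stage when subscription is inactive.
# Custom left/right navigation sets the selected reward on the selection view model.
# Selection changes play slide-and-fade animation, and deselection clears both items and hides the tooltip.
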WidgetBlueprint /Crew/UI/ProgressiveCosmetic/ProgressiveStageWidget
animations [
 Intro {
  binding_count=2
  length_seconds=0.25
  object=&self:Intro
  sequence {bindings [
    SizeBox_0 {
     id=#1
     possessed_class=SizeBox
     tracks [
      {
       type=MovieSceneFloatTrack
       display_name=Render Opacity
       sections [{type=MovieSceneFloatSection, channels [None {type=MovieSceneScriptingFloatChannel, keys [[0,0],[2,0],[5,1]]}]}]
      }
      {
       type=MovieScene2DTransformTrack
       sections [{type=MovieScene2DTransformSection, channels [Translation.X {type=MovieSceneScriptingFloatChannel, keys [[2,-256],[5,0]]}]}]
      }
     ]
    }
    Image {
     id=#2
     possessed_class=Image
     tracks [{
       type=MovieSceneFloatTrack
       display_name=Render Opacity
       sections [{type=MovieSceneFloatSection, channels [None {type=MovieSceneScriptingFloatChannel, keys [[0,0.7],[2,1]]}]}]
      }]
    }
   ]}
 }
 Outro {
  binding_count=2
  length_seconds=0.15
  object=&self:Outro
  sequence {bindings [
    SizeBox_0 {
     id=#3
     possessed_class=SizeBox
     tracks [
      {
       type=MovieScene2DTransformTrack
       sections [{type=MovieScene2DTransformSection, channels [Translation.X {type=MovieSceneScriptingFloatChannel, keys [[0,0],[3,-256]]}]}]
      }
      {
       type=MovieSceneFloatTrack
       display_name=Render Opacity
       sections [{type=MovieSceneFloatSection, channels [None {type=MovieSceneScriptingFloatChannel, keys [[0,1],[3,0]]}]}]
      }
     ]
    }
    Image {
     id=#4
     possessed_class=Image
     tracks [{
       type=MovieSceneFloatTrack
       display_name=Render Opacity
       sections [{type=MovieSceneFloatSection, channels [None {type=MovieSceneScriptingFloatChannel, keys [[0,1],[3,0.5]]}]}]
      }]
    }
   ]}
 }
]
graphs [
 EventGraph {nodes [
   Event On Selected()
   > SetFocus(self=Get ProgressiveItemWidget_Left() -> (ProgressiveItemWidget_Left:Progressive Item Widget Object Reference))
   Event Construct()
   > PlayAnimation(
    InAnimation=Get Intro()
    StartAtTime=1
    PlayMode=Select(Option 0:EUMGSequencePlayMode Enum=Reverse, Option 1:EUMGSequencePlayMode Enum, Index=Get Was Selected() -> (Was Selected:bool)) -> (ReturnValue:EUMGSequencePlayMode Enum)
   )
   > SetNavigationRuleCustom!(execute, self:Widget Object Reference, Direction:EUINavigation Enum=Left, InCustomDelegate:Delegate=Create Event()) -> (then)
   > SetNavigationRuleCustom(Direction=Right, InCustomDelegate=Create Event())
   n1 = Get CrewProgressiveSelectionVM() -> (CrewProgressiveSelectionVM:Crew Progressive Selection VM Object Reference)
   n2 = Get Rewards!(self:Crew Progressive Stage VM Object Reference=Get SelectedStage!(self:Crew Progressive Selection VM Object Reference=@n1) -> (SelectedStage:Crew Progressive Stage VM Object Reference)) -> (Rewards:Array of Crew Progressive Reward VM Object References)
   n3 = Get SelectedReward!(self:Crew Progressive Selection VM Object Reference=@n1) -> (SelectedReward:Crew Progressive Reward VM Object Reference)
   Select(
    Option 0:Widget Object Reference=Get ProgressiveItemWidget_Left()
    Option 1:Widget Object Reference=Get ProgressiveItemWidget_Right() -> (ProgressiveItemWidget_Right:Progressive Item Widget Object Reference)
    Index="integer > integer"(A=Find Item!(TargetArray:Array of Crew Progressive Reward VM Object References=@n2, ItemToFind:"Crew Progressive Reward VM Object Reference (by ref)"=@n3) -> (ReturnValue))
   ) -> (ReturnValue:Widget Object Reference)
  ]}
 GetMonthText {nodes [
   n1 = GetMonthText!() -> (then, Stage Index:int) {is=title K2Node_FunctionEntry}
   > Return Node!(execute, Result:Text=Format Text!(Format="LOCTEXT(\"Stage {Index}\")", Index:int=int + int(A=@n1)) -> (Result))
  ]}
 SelectLeftWidget {nodes [
   SelectLeftWidget!() -> (then, Navigation:EUINavigation Enum) {is=title K2Node_FunctionEntry}
   n1 = Get CrewProgressiveStageVM() -> (CrewProgressiveStageVM:Crew Progressive Stage VM Object Reference)
   > Set SelectedReward!(
    execute
    SelectedReward:Crew Progressive Reward VM Object Reference="Get (a copy)"(Array:Array of Crew Progressive Reward VM Object References=Get Rewards(self=@n1)) -> (Output:Crew Progressive Reward VM Object Reference)
    self:Crew Progressive Selection VM Object Reference=Get CrewProgressiveSelectionVM()
   ) -> (then, Output_Get:Crew Progressive Reward VM Object Reference)
   > Return Node!(execute, ReturnValue:Widget Object Reference)
  ]}
 SelectRightWidget {nodes [
   SelectRightWidget!() -> (then, Navigation:EUINavigation Enum) {is=title K2Node_FunctionEntry}
   n1 = Get Rewards(self=Get CrewProgressiveStageVM())
   > n2 = Branch(Condition="integer > integer"(A=Length(TargetArray:Array of Crew Progressive Reward VM Object References=@n1), B=1))
   n3 = Get CrewProgressiveStageVM()
   > n4 = Set SelectedReward(SelectedReward="Get (a copy)"(Array=Get Rewards(self=@n3), Dimension 1=1), self=Get CrewProgressiveSelectionVM())
   Return Node(execute=[@n2.else, @n4.then], ReturnValue=Self-Reference())
  ]}
 ShowTooltip {nodes [
   ShowTooltip() {is=title K2Node_FunctionEntry}
   n1 = Get CrewProgressiveStageVM()
   n2 = Get bIsOwned!(self:Crew Progressive Stage VM Object Reference=@n1) -> (bIsOwned:bool)
   n3 = Get SubscriptionState!(self:Crew VM Object Reference=Get CrewVM() -> (CrewVM:Crew VM Object Reference)) -> (SubscriptionState:EMcpSubscriptionState Enum=Inactive)
   n4 = GetNextStage!(self:Crew Progressive Set VM Object Reference=Get SelectedSet!(self:Crew Progressive Selection VM Object Reference=Get CrewProgressiveSelectionVM()) -> (SelectedSet:Crew Progressive Set VM Object Reference)) -> (ReturnValue:Crew Progressive Stage VM Object Reference)
   > Branch(Condition=AND Boolean!(
     A="Equal (Enum)"(A:EMcpSubscriptionState Enum=@n3, B:EMcpSubscriptionState Enum=Inactive)
     B=NOT Boolean(A=@n2)
     C:bool="Equal (Object)"(A=@n1, B=@n4)
    ) -> (ReturnValue))
   n5 = Get CrewProgressiveStageVM()
   > n6 = Branch(Condition="Equal (Integer)"(A=Get StageIndex!(self:Crew Progressive Stage VM Object Reference=@n5) -> (StageIndex:int=0)))
   n7 = Get SelectedSet(self=Get CrewProgressiveSelectionVM())
   > Branch(Condition="Not Equal (DateTime)"!(
     A:Date Time Structure="MinValue (DateTime)"!() -> (ReturnValue:Date Time Structure)
     B:Date Time Structure=Get AcquisitionCutoffDate!(self:Crew Progressive Set VM Object Reference=@n7) -> (AcquisitionCutoffDate:Date Time Structure)
    ) -> (ReturnValue:bool))
   n8 = Get ProgressiveItem_Tooltip() -> (ProgressiveItem_Tooltip:Progressive Item Tooltip Object Reference)
   > Update Text!(
    execute
    self:Progressive Item Tooltip Object Reference=@n8
    InText:"Text (by ref)"=Format Text!(
     Format="LOCTEXT(\"Subscribe by <text color=\\\"FFF\\\" fontface=\\\"black\\\">{ExpirationDate}</> to claim Stage 1. Once claimed, <text color=\\\"FFF\\\" fontface=\\\"black\\\">there is no time limit</> to complete remaining stages\")"
     ExpirationDate:Text=Get AcquisitionCutoffDateText!(self:Crew Progressive Set VM Object Reference=@n7) -> (AcquisitionCutoffDateText:Text)
    ) -> (Result)
   ) -> (then)
   > ToggleTooltip!(execute, self:Progressive Item Tooltip Object Reference=@n8, Visible:bool=true) -> (then)
   n9 = Get StageIndex(self=Get CrewProgressiveStageVM()) -> (StageIndex=0)
   n10 = Get ProgressiveItem_Tooltip()
   Update Text(
    execute=@n6.else
    self=@n10
    InText=Format Text!(
     Format="LOCTEXT(\"<text color=\\\"FFF\\\" fontface=\\\"black\\\">Stage {StageNum}</> unlocks on the <text color=\\\"FFF\\\" fontface=\\\"black\\\">{UnlockDay}{UnlockDay}|ordinal(one=st,two=nd,few=rd,other=th) of the month</> as long as you\\'re subscribed\\\"\")"
     StageNum:int=int + int(A=@n9)
     UnlockDay:int=MakeLiteralInt!(Value:int=1) -> (ReturnValue:int=0)
    ) -> (Result)
   )
   > ToggleTooltip(self=@n10, Visible=true)
  ]}
 Stage Selected {nodes [
   n1 = Stage Selected!() -> (then, Selected:bool=~) {is=title K2Node_FunctionEntry}
   > Branch(Condition="Not Equal (Boolean)"(A=@n1, B=Get Was Selected()) -> (ReturnValue=~))
   > n2 = Set Was Selected(Was Selected:bool=@n1) -> (Output_Get:bool)
   n3 = Get Intro()
   PlayAnimation(
    execute=[@n6.then, @n5.then, @n2.then]
    InAnimation=@n3
    PlayMode=Select(Option 0:EUMGSequencePlayMode Enum=Reverse, Option 1:EUMGSequencePlayMode Enum, Index=@n2) -> (ReturnValue:EUMGSequencePlayMode Enum)
   )
   > n4 = Branch(Condition=@n2)
   SetVisibility(execute=@n4.else, self=Get ProgressiveItem_Tooltip(), InVisibility=Collapsed)
   n5 = Branch(Condition=@n2)
   n6 = SetIsSelected!(
    execute=@n5.else
    self:Common Button Base Object Reference=[Self-Reference(), Get ProgressiveItemWidget_Right(), Get ProgressiveItemWidget_Left()]
    InSelected:bool=false
    bGiveClickFeedback:bool=false
   ) -> (then)
   Select(
    Option 0:Widget Animation Object Reference=Get Outro() -> (Outro:Widget Animation Object Reference)
    Option 1:Widget Animation Object Reference=@n3
    Index=@n2
   ) -> (ReturnValue:Widget Animation Object Reference)
  ]}
 Stage Set {nodes [
   n1 = Stage Set!() -> (then, Stage:Crew Progressive Stage VM Object Reference) {is=title K2Node_FunctionEntry}
   > n2 = Is Valid(InputObject=@n1)
   SetVisibility(
    execute=[@n2.Is Not Valid, @n3.0]
    self=[
     Get ItemRightContainer() -> (ItemRightContainer:Overlay Object Reference)
     Get ItemLeftContainer() -> (ItemLeftContainer:Overlay Object Reference)
     Get CenterSeparator() -> (CenterSeparator:Overlay Object Reference)
    ]
    InVisibility=Collapsed
   )
   n3 = Switch on Int!(execute:Exec=@n2.Is Valid, Selection:int=Length(TargetArray=Get Rewards(self=@n1))) -> (Default:Exec, 0:Exec, 1:Exec)
   n4 = SetVisibility(execute=@n3.1, self=[Get ItemRightContainer(), Get CenterSeparator()], InVisibility=Collapsed)
   SetVisibility(execute=[@n4.then, @n7.then_0], self=Get ItemLeftContainer(), InVisibility=SelfHitTestInvisible)
   n5 = Get Stage() -> (Stage:Crew Progressive Stage VM Object Reference)
   n6 = Get ProgressiveItemWidget_Left()
   > SetCrewProgressiveStageVM!(execute, self:Progressive Item Widget Object Reference=@n6, Viewmodel:Crew Progressive Stage VM Object Reference=@n5) -> (then)
   > SetCrewProgressiveRewardVM!(
    execute
    self:Progressive Item Widget Object Reference=@n6
    Viewmodel:Crew Progressive Reward VM Object Reference="Get (a copy)"(Array=Get Rewards(self=@n5))
   ) -> (then)
   > ShowTooltip() {is=title K2Node_CallFunction}
   n7 = Sequence(execute=@n3.Default)
   SetVisibility(execute=@n7.then_1, self=[Get ItemRightContainer(), Get CenterSeparator()], InVisibility=SelfHitTestInvisible)
   n8 = Get ProgressiveItemWidget_Right()
   n9 = Get Stage()
   > SetCrewProgressiveStageVM(self=@n8, Viewmodel=@n9)
   > SetCrewProgressiveRewardVM(self=@n8, Viewmodel="Get (a copy)"(Array=Get Rewards(self=@n9), Dimension 1=1))
  ]}
]
variables [Was Selected {type=boolean}]
widgets {
 info {inheritedWidgetCount=1, parentClass=&CommonButtonBase, rootWidgetClass=&Overlay, widgetCount=15}
 widgets [
  MainListRowContainer {widgetClassPath=&Overlay, widget=MainListRowContainer}
  VB_ListRowPaddings {
   widgetClassPath=&VerticalBox
   parent=MainListRowContainer
   widget=VB_ListRowPaddings
   slot {class=OverlaySlot, properties {padding=<ltrb 8 12 8 0>}}
  }
  SizeBox_0 {
   widgetClassPath=&SizeBox
   parent=VB_ListRowPaddings
   widget=SizeBox_0
   properties {bOverride_HeightOverride=true, heightOverride=48}
   slot {class=VerticalBoxSlot}
  }
  Text_Month {
   widgetClassPath=&CommonTextBlock
   parent=SizeBox_0
   widget=Text_Month
   properties {
    text="{PH} MONTH X"
    textTransformPolicy=ToUpper
    font {
     bForceMonospaced=false
     bMaterialIsStencil=false
     fontMaterial=null
     fontObject=&Burbank
     letterSpacing=0
     monospacedWidth=1
     size=23
     skewAmount=0
     typefaceFontName=BurbankRegular-Black
     outlineSettings {
      bApplyOutlineToDropShadows=false
      bMiteredCorners=false
      bSeparateFillAlpha=false
      outlineMaterial=null
      outlineSize=0
      outlineColor=<rgba 0 0 0 1>
     }
    }
    renderTransform {angle=0, scale=<1 1>, shear=<-12 0>, translation=<0 0>}
    shadowOffset=<0 0>
   }
   slot {class=SizeBoxSlot, properties {verticalAlignment=VAlign_Center}}
  }
  Spacer {widgetClassPath=&Spacer, parent=VB_ListRowPaddings, widget=Spacer, properties {size=<1 5>}, slot {class=VerticalBoxSlot}}
  HorizontalBox_63 {
   widgetClassPath=&HorizontalBox
   parent=VB_ListRowPaddings
   widget=HorizontalBox_63
   slot {class=VerticalBoxSlot, properties {horizontalAlignment=HAlign_Left}}
  }
  ItemLeftContainer {
   widgetClassPath=&Overlay
   bIsVariable=true
   parent=HorizontalBox_63
   widget=ItemLeftContainer
   properties {visibility=Collapsed}
   slot {class=HorizontalBoxSlot, properties {horizontalAlignment=HAlign_Left}}
  }
  ProgressiveItemWidget_Left {
   widgetClassPath=&/ProgressiveItemWidget
   bIsVariable=true
   parent=ItemLeftContainer
   widget=ProgressiveItemWidget_Left
   properties {bSelectable=true, bShouldSelectUponReceivingFocus=true}
   slot {class=OverlaySlot}
  }
  CenterSeparator {
   widgetClassPath=&Overlay
   bIsVariable=true
   parent=HorizontalBox_63
   widget=CenterSeparator
   properties {visibility=Collapsed}
   slot {class=HorizontalBoxSlot, properties {horizontalAlignment=HAlign_Center}}
  }
  Image {
   widgetClassPath=&Image
   parent=CenterSeparator
   widget=Image
   properties {
    visibility=SelfHitTestInvisible
    brush {
     bIsDynamicallyLoaded=false
     drawAs=Image
     imageType=NoImage
     mirroring=NoMirror
     resourceName=None
     resourceObject=/Game/Athena/UI/Frontend/Parties/Art/T_UI_PlusSign_64x
     tiling=NoTile
     imageSize=<50 50>
     margin=<ltrb 0 0 0 0>
     outlineSettings {
      bUseBrushTransparency=false
      roundingType=HalfHeightRadius
      width=0
      color {colorUseRule=UseColor_Specified, specifiedColor=<rgba 0 0 0 0>}
      cornerRadii=<0 0 0 0>
     }
     tintColor {colorUseRule=UseColor_Specified, specifiedColor=<rgba 1 1 1 1>}
     uVRegion {bIsValid=false, max=<0 0>, min=<0 0>}
    }
    colorAndOpacity=<rgba 0.0319 0.855 1 1>
   }
   slot {class=OverlaySlot, properties {horizontalAlignment=HAlign_Center, verticalAlignment=VAlign_Center, padding=<ltrb 10 0 10 0>}}
  }
  ItemRightContainer {
   widgetClassPath=&Overlay
   bIsVariable=true
   parent=HorizontalBox_63
   widget=ItemRightContainer
   properties {visibility=Collapsed}
   slot {class=HorizontalBoxSlot, properties {horizontalAlignment=HAlign_Right}}
  }
  ProgressiveItemWidget_Right {
   widgetClassPath=&ProgressiveItemWidget
   bIsVariable=true
   parent=ItemRightContainer
   widget=ProgressiveItemWidget_Right
   properties {bSelectable=true, bShouldSelectUponReceivingFocus=true}
   slot {class=OverlaySlot}
  }
  ProgressiveItem_Tooltip {
   widgetClassPath=/Crew/UI/ProgressiveItem_Tooltip
   bIsVariable=true
   parent=HorizontalBox_63
   widget=ProgressiveItem_Tooltip
   properties {
    in Vertical Alignment=VAlign_Bottom
    tooltip Text="[PH] <text color=\"fff\">Subscribe by, APR 9, 2022 to </>claim Stage 1. Once claimed,<text color=\"fff\"> there is no time limit to complete remaining stages</>"
    in Transform {angle=180, scale=<1 1>, shear=<0 0>, translation=<32 26>}
   }
   slot {class=HorizontalBoxSlot, properties {horizontalAlignment=HAlign_Right, verticalAlignment=VAlign_Center, padding=<ltrb 24 -100 0 0>}}
  }
  Spacer_84 {widgetClassPath=&Spacer, parent=VB_ListRowPaddings, widget=Spacer_84, properties {size=<1 5>}, slot {class=VerticalBoxSlot}}
  InputActionWidget {widgetClassPath=&CommonActionWidget, bInherited=true, widget=null}
 ]
}
