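# Full-screen Fortnite Crew screen for a progressive cosmetic set, showing a stage list, 3D item preview, details, subscribe button and legal fine print.
# Crew view models drive stage selection, header and fine print, with gradient, text color and preview material reflecting subscription state.
# Parental content blocking swaps the right-side purchase column for a blocker, which can take focus.
# Also handles intro animations, mobile padding, styles preview push and analytics events.
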WidgetBlueprint /Crew/UI/ProgressiveCosmetic/ProgressiveItemScreen
animations [
 ProgDetailsItemChange {
  binding_count=1
  length_seconds=0.25
  object=&self:ProgDetailsItemChange
  sequence {bindings [Widget_ProgressiveItemDetails {
     id=#1
     possessed_class=ProgressiveItemDetails_C
     tracks [
      {
       type=MovieSceneFloatTrack
       display_name=Render Opacity
       sections [{type=MovieSceneFloatSection, channels [None {type=MovieSceneScriptingFloatChannel, keys [[0,0],[5,1]]}]}]
      }
      {
       type=MovieScene2DTransformTrack
       sections [{
         type=MovieScene2DTransformSection
         channels [
          Translation.X {type=MovieSceneScriptingFloatChannel, keys [[0,100],[5,0]]}
          Translation.Y {type=MovieSceneScriptingFloatChannel}
         ]
        }]
      }
     ]
    }]}
 }
 Intro {
  binding_count=6
  length_seconds=0.25
  object=&self:Intro
  sequence {bindings [
    BottomScreenGradient {
     id=#2
     possessed_class=Image
     tracks [{
       type=MovieSceneFloatTrack
       display_name=Render Opacity
       sections [{type=MovieSceneFloatSection, channels [None {type=MovieSceneScriptingFloatChannel, keys [[0,0],[5,0.7],[5.0003,0.8]]}]}]
      }]
    }
    RightSideVertBox {
     id=#3
     possessed_class=VerticalBox
     tracks [
      {
       type=MovieScene2DTransformTrack
       sections [{
         type=MovieScene2DTransformSection
         channels [
          Translation.X {type=MovieSceneScriptingFloatChannel, keys [[0,100],[5,0]]}
          Translation.Y {type=MovieSceneScriptingFloatChannel, const [0,1]}
          Angle {type=MovieSceneScriptingFloatChannel, const [0,1]}
          Scale.X {type=MovieSceneScriptingFloatChannel, const [1,1]}
          Scale.Y {type=MovieSceneScriptingFloatChannel, const [1,1]}
          Shear.X {type=MovieSceneScriptingFloatChannel, const [0,1]}
          Shear.Y {type=MovieSceneScriptingFloatChannel, const [0,1]}
         ]
        }]
      }
      {
       type=MovieSceneFloatTrack
       display_name=Render Opacity
       sections [{type=MovieSceneFloatSection, channels [None {type=MovieSceneScriptingFloatChannel, keys [[0,0],[5,1]]}]}]
      }
     ]
    }
    LegalVertBox {
     id=#4
     possessed_class=VerticalBox
     tracks [
      {
       type=MovieScene2DTransformTrack
       sections [{
         type=MovieScene2DTransformSection
         channels [
          Translation.X {type=MovieSceneScriptingFloatChannel, keys [[0,-100],[5,0]]}
          Translation.Y {type=MovieSceneScriptingFloatChannel, const [0,1]}
          Angle {type=MovieSceneScriptingFloatChannel, const [0,1]}
          Scale.X {type=MovieSceneScriptingFloatChannel, const [1,1]}
          Scale.Y {type=MovieSceneScriptingFloatChannel, const [1,1]}
          Shear.X {type=MovieSceneScriptingFloatChannel, const [0,1]}
          Shear.Y {type=MovieSceneScriptingFloatChannel, const [0,1]}
         ]
        }]
      }
      {
       type=MovieSceneFloatTrack
       display_name=Render Opacity
       sections [{type=MovieSceneFloatSection, channels [None {type=MovieSceneScriptingFloatChannel, keys [[0,0],[5,1]]}]}]
      }
     ]
    }
    VerticalBox_Pips {
     id=#5
     possessed_class=VerticalBox
     tracks [{
       type=MovieSceneFloatTrack
       display_name=Render Opacity
       sections [{type=MovieSceneFloatSection, channels [None {type=MovieSceneScriptingFloatChannel, keys [[0,0],[5,1]]}]}]
      }]
    }
    ProgressiveItem_Header {
     id=#6
     possessed_class=ProgressiveItemStateTitle_C
     tracks [
      {
       type=MovieSceneFloatTrack
       display_name=Render Opacity
       sections [{type=MovieSceneFloatSection, channels [None {type=MovieSceneScriptingFloatChannel, keys [[0,0],[5,1]]}]}]
      }
      {
       type=MovieScene2DTransformTrack
       sections [{
         type=MovieScene2DTransformSection
         channels [
          Translation.X {type=MovieSceneScriptingFloatChannel, const [0,1]}
          Translation.Y {type=MovieSceneScriptingFloatChannel, keys [[0,35],[5,0]]}
         ]
        }]
      }
     ]
    }
    ScrollContainer {
     id=#7
     possessed_class=Overlay
     tracks [
      {
       type=MovieSceneFloatTrack
       display_name=Render Opacity
       sections [{type=MovieSceneFloatSection, channels [None {type=MovieSceneScriptingFloatChannel, keys [[0.0003,0],[5.0003,1]]}]}]
      }
      {
       type=MovieScene2DTransformTrack
       sections [{
         type=MovieScene2DTransformSection
         channels [
          Translation.X {type=MovieSceneScriptingFloatChannel, const [0,1]}
          Translation.Y {type=MovieSceneScriptingFloatChannel, keys [[0.0003,35],[5.0003,0]]}
         ]
        }]
      }
     ]
    }
   ]}
 }
]
graphs [
 BP_GetDesiredFocusTarget {nodes [
   Get Desired Focus Target()
   > n1 = IsCabinModeEnabled!(execute, Context:User Widget Object Reference=Self-Reference()) -> (then, ReturnValue:bool)
   n2 = GetIsCabinModeEnabled!(self:Fort Local Player VM Object Reference=Get FortLocalPlayerVM() -> (FortLocalPlayerVM:Fort Local Player VM Object Reference)) -> (ReturnValue:bool)
   n3 = Get WBP_ContentPurchaseBlocker() -> (WBP_ContentPurchaseBlocker:WBP Content Purchase Blocker Object Reference)
   > Return Node!(
    execute
    ReturnValue:Widget Object Reference=Select(
     Option 0:Widget Object Reference=Get ListView_Stages() -> (ListView_Stages:Fort Tile View Object Reference)
     Option 1:Widget Object Reference=Get WBP_UIKit_Button_Loud!(self:WBP Content Purchase Blocker Object Reference=@n3) -> (WBP_UIKit_Button_Loud:WBP UIKit Button Loud Object Reference)
     Index=AND Boolean(A=@n1, B=@n2)
    ) -> (ReturnValue:Widget Object Reference)
   )
  ]}
 BP_GetDesiredInputConfig {nodes [
   Get Desired Input Config() {is=title K2Node_FunctionEntry}
   > Return Node(ReturnValue:UIInput Config Structure=Make UIInput Config!(
     bIgnoreMoveInput:bool=False
     bIgnoreLookInput:bool=False
     InputMode:ECommonInputMode Enum=All
     MouseCaptureMode:EMouseCaptureMode Enum=CaptureDuringMouseDown
     MouseLockMode:EMouseLockMode Enum=DoNotLock
     bHideCursorDuringViewportCapture:bool=True
    ) -> (UIInputConfig:UIInput Config Structure) {is=title K2Node_MakeStruct})
  ]}
 ContentBlocking {nodes [
   n1 = ContentBlocking!() -> (then, IsContentBlocked:bool=~) {is=title K2Node_FunctionEntry}
   > SetVisibility(
    self=[Get BottomContentBlockerGradient() -> (BottomContentBlockerGradient:Image Object Reference), Get WBP_ContentPurchaseBlocker()]
    InVisibility=Select(Option 0:ESlateVisibility Enum=Collapsed, Option 1:ESlateVisibility Enum=SelfHitTestInvisible, Index=@n1) -> (ReturnValue:ESlateVisibility Enum)
   )
   > SetVisibility(
    self=Get RightSideVertBox() -> (RightSideVertBox:Vertical Box Object Reference)
    InVisibility=Select(Option 0=SelfHitTestInvisible, Option 1=Collapsed, Index=@n1)
   )
  ]}
 EventGraph {nodes [
   n1 = "On Clicked (ConfirmPurchase_Button)"!() -> (OutputDelegate, then, Button:Common Button Base Object Reference) {is=title K2Node_ComponentBoundEvent}
   > SetIsEnabled(self=@n1.Button, bInIsEnabled=false)
   > Send Analytics Event!(execute, self, Event:String=SubscribeClicked) -> (then)
   > HandleSubscribeEngaged(self:Fort Progressive Item Screen Object Reference)
   "On Clicked (Button_Back)"!() -> (OutputDelegate, then, Button:Common Button Base Object Reference) {is=title K2Node_ComponentBoundEvent}
   > Send Analytics Event(Event=BackClicked)
   > Fire Event Element Interacted(Element=Get Button_Back() -> (Button_Back:WBP UIKit Button Quiet Object Reference))
   > HandleBackClicked(self:Fort Progressive Item Screen Object Reference)
   n2 = "On Clicked (Button_MoreInfo)"!() -> (OutputDelegate, then, Button:Common Button Base Object Reference) {is=title K2Node_ComponentBoundEvent}
   > Send Analytics Event(Event=MoreInfoClicked)
   > Fire Event Element Interacted(Element=@n2.Button)
   > HandleMoreInfoClicked(self:Fort Progressive Item Screen Object Reference)
   n3 = "On Clicked (Button_PreviewStyles)"!() -> (OutputDelegate, then, Button:Common Button Base Object Reference) {is=title K2Node_ComponentBoundEvent}
   > Send Analytics Event(Event=CustomizationOptionsClicked)
   > Fire Event Element Interacted(Element=@n3.Button)
   > n4 = PushPrimaryContentClassAsync!(execute:Exec, WidgetClass:Common Activatable Widget Class Reference=/Crew/UI/Preview/CrewItemPreviewSreen.CrewItemPreviewSreen_C) -> (then:Exec, OnPrimaryContentPushed:Exec, ActivatableWidget:Common Activatable Widget Object Reference) {is=title K2Node_AsyncAction}
   n5 = Cast To CrewItemPreviewSreen!(execute=@n4.OnPrimaryContentPushed, Object=@n4) -> (then, CastFailed, AsCrew Item Preview Sreen:Crew Item Preview Sreen Object Reference) {is=title K2Node_DynamicCast}
   n6 = Get CrewProgressiveSelectionVM() -> (CrewProgressiveSelectionVM:Crew Progressive Selection VM Object Reference)
   > Set with Broadcast Rewards!(
    execute
    Rewards:Array of Fort Item VM Object References=Make Array("[0]":Fort Item VM Object Reference=Get PreviewItem!(self:Crew Progressive Selection VM Object Reference=@n6) -> (PreviewItem:Fort Item VM Object Reference)) -> (Array:Array of Fort Item VM Object References)
    self:Crew Item Preview Sreen Object Reference=@n5
   ) -> (then, Output_Get:Array of Fort Item VM Object References)
   n7 = "On Item Selection Changed (ListView_Stages)"!() -> (OutputDelegate, then, Item:Object Reference, bIsSelected:bool) {is=title K2Node_ComponentBoundEvent}
   > Branch(Condition=IsActivated!(self:Common Activatable Widget Object Reference=Self-Reference()) -> (ReturnValue:bool))
   > n8 = Branch(Condition=@n7.bIsSelected)
   > FireEventImpression()
   n9 = Cast To FortPlayerController(execute=@n8.else, Object=GetOwningPlayer())
   > Fire Event Flush Impressions!(execute, FortPC:Fort Player Controller Object Reference=@n9) -> (then)
   Event Construct()
   > SetCompactUIStyle() {is=title K2Node_CallFunction}
   Event On Activated()
   n10 = Get FortItemPreviewWidget() -> (FortItemPreviewWidget:WBP Item Preview Object Reference)
   > ActivateWidget(self=@n10)
   > UpdateSubscribedVisuals!(
    execute
    self
    Subscribed:bool=IsSubscribed!(self:Crew VM Object Reference=Get CrewVM() -> (CrewVM:Crew VM Object Reference)) -> (ReturnValue:bool)
   ) -> (then)
   > n11 = "Play Animation with Finished event (legacy)"!(
    execute:Exec
    Widget:User Widget Object Reference=Self-Reference()
    InAnimation:Widget Animation Object Reference=Get Intro()
    StartAtTime:float=0
    NumLoopsToPlay:int=1
    PlayMode:EUMGSequencePlayMode Enum
    PlaybackSpeed:float=1
   ) -> (then:Exec, Result:UMGSequence Player Object Reference, Finished:Exec) {is=title K2Node_PlayAnimation}
   FireSubView(execute=@n11.Finished)
   n12 = Get ListView_Stages()
   n13 = Get SelectedStage!(self:Crew Progressive Selection VM Object Reference=Get CrewProgressiveSelectionVM()) -> (SelectedStage:Crew Progressive Stage VM Object Reference)
   > Set Selected Item!(execute, self:List View Object Reference=@n12, Item:Object Reference=@n13) -> (then) {is=title K2Node_CallFunction}
   > Navigate To Item!(execute, self:List View Object Reference=@n12, Item:Object Reference=@n13) -> (then)
   > Scroll Item Into View!(execute, self:List View Object Reference=@n12, Item:Object Reference=@n13) -> (then)
   > FireEventImpression()
   Event PreConstruct()
   > n14 = Create Widget!(
    execute:Exec
    Class:User Widget Class Reference=/FortUILibrary/Systems/ItemShopTile/WBP_Price_PlatformStoreIcon.WBP_Price_PlatformStoreIcon_C
    OwningPlayer:Player Controller Object Reference=GetOwningPlayer()
    PreviewPlayStationPlatform:bool=Get PreviewPlatformIcon() -> (PreviewPlatformIcon:bool)
    ImageSize:Vector 2D Structure=Get PlatformIconSize() -> (PlatformIconSize:Vector 2D Structure)
   ) -> (then:Exec, ReturnValue:WBP Price Platform Store Icon Object Reference)
   > n15 = AddChildToGrid!(
    execute
    self:Grid Panel Object Reference=Get Root!(self:WBP UIKit Button CTA Base Object Reference=Get ConfirmPurchase_Button() -> (ConfirmPurchase_Button:WBP UIKit Button Loud Object Reference)) -> (Root:Grid Panel Object Reference)
    Content:Widget Object Reference=@n14
    InRow:int=1
    InColumn:int=4
   ) -> (then, ReturnValue:Grid Slot Object Reference)
   > Set VerticalAlignment!(execute, VerticalAlignment:EVerticalAlignment Enum=VAlign_Center, self:Grid Slot Object Reference=@n15) -> (then, Output_Get:EVerticalAlignment Enum=VAlign_Fill)
   > Set HorizontalAlignment!(execute, HorizontalAlignment:EHorizontalAlignment Enum=HAlign_Center, self:Grid Slot Object Reference=@n15) -> (then, Output_Get:EHorizontalAlignment Enum=HAlign_Fill)
   > SetGridSlotPadding!(execute, Widget:Widget Object Reference=@n14, Padding:Margin Structure=Make Margin(Left=4, Top=0, Right=8, Bottom=0)) -> (then)
   Event OnSubscriptionPurchaseComplete!() -> (OutputDelegate, then, bSuccess:bool) {is=title K2Node_Event}
   > SetIsEnabled(self=@n1.Button, bInIsEnabled=true)
   Event On Deactivated()
   > DeactivateWidget(self=@n10)
  ]}
 FireEventImpression {nodes [
   FireEventImpression() {is=title K2Node_FunctionEntry}
   n1 = Get SelectedStage(self=Get CrewProgressiveSelectionVM())
   n2 = Format Text!(
    Format="LOCTEXT(\"Stage {0}\")"
    0:int=int + int(A=Get StageIndex!(self:Crew Progressive Stage VM Object Reference=@n1) -> (StageIndex:int=0))
   ) -> (Result)
   n3 = Get CrewProgressiveSelectionVM()
   > "Fire Event Add Impression (Localized)"!(
    execute
    SectionID:"Text (by ref)"=Get Name!(self:Crew Progressive Set VM Object Reference=Get SelectedSet!(self:Crew Progressive Selection VM Object Reference=@n3) -> (SelectedSet:Crew Progressive Set VM Object Reference)) -> (Name:Text)
    SectionContents:Array of Flow Analytics Impression Section Content Structures=Make Array("[0]":Flow Analytics Impression Section Content Structure=Make Flow Analytics Impression Section Content!(ContentID:Text=@n2, OfferID:Text, Row:int=-1, Column:int=-1, Index:int=-1) -> (FlowAnalyticsImpressionSectionContent:Flow Analytics Impression Section Content Structure) {is=title K2Node_MakeStruct}) -> (Array:Array of Flow Analytics Impression Section Content Structures)
   ) -> (then)
  ]}
 FireSubView {nodes [
   FireSubView() {is=title K2Node_FunctionEntry}
   > n1 = Cast To FortPlayerController(Object=GetOwningPlayer())
   n2 = Get SelectedSet(self=Get CrewProgressiveSelectionVM())
   > "Fire Event Sub View Opened (Non-Localized)"!(
    execute
    FortPC:Fort Player Controller Object Reference=@n1
    ViewName:String=CrewStylesItems
    AdditionalInfo:Flow Analytics Additional Info Structure="Add Text (Localized)"!(
     Info:"Flow Analytics Additional Info Structure (by ref)"=Make Flow Analytics Additional Info!() -> (FlowAnalyticsAdditionalInfo:Flow Analytics Additional Info Structure) {is=title K2Node_MakeStruct}
     Key:String=ItemName
     Value:"Text (by ref)"=Get Name(self=@n2)
    ) -> (ReturnValue:Flow Analytics Additional Info Structure)
    bIsOverlay:bool=false
   ) -> (then)
  ]}
 Get Expiration Text {nodes [
   n1 = Get Expiration Text!() -> (then, Expiration Date:Date Time Structure) {is=title K2Node_FunctionEntry}
   > Return Node!(
    execute
    NewParam:Text=Format Text!(
     Format="LOCTEXT(\"Ability to unlock available until {ExpirationDate}\")"
     ExpirationDate:Text=As DateTime!(In:"Date Time Structure (by ref)"=@n1, InDateStyle:EDateTimeStyle Enum=Default, InTimeStyle:EDateTimeStyle Enum=Default) -> (ReturnValue:Text)
    ) -> (Result)
   )
  ]}
 Get Purchase Visibility {nodes [
   n1 = Get Purchase Visibility!() -> (then, Is Cabined:bool=~, Has CMS Data:bool=~, Sub State:EMcpSubscriptionState Enum=Inactive) {is=title K2Node_FunctionEntry}
   n2 = NOT Boolean(A=@n1)
   n3 = "Equal (Enum)"(A:EMcpSubscriptionState Enum=@n1.Sub State, B:EMcpSubscriptionState Enum=Inactive)
   > Return Node(NewParam:ESlateVisibility Enum=Select(Option 0=Collapsed, Option 1=SelfHitTestInvisible, Index=AND Boolean!(A=@n2, B=@n1.Has CMS Data, C:bool=@n3) -> (ReturnValue)))
  ]}
 Navigate Right from Items List {nodes [
   Navigate Right from Items List!() -> (then, Navigation:EUINavigation Enum) {is=title K2Node_FunctionEntry}
   > Return Node!(execute, ReturnValue:Widget Object Reference=Get WBP_UIKit_Button_Loud(self=Get WBP_ContentPurchaseBlocker()))
  ]}
 Reward Selection Changed {nodes [
   Reward Selection Changed!() -> (then, NewParam:Crew Progressive Reward VM Object Reference) {is=title K2Node_FunctionEntry}
   > PlayAnimation(InAnimation=Get ProgDetailsItemChange() -> (ProgDetailsItemChange:Widget Animation Object Reference))
  ]}
 Send Analytics Event {nodes [
   n1 = Send Analytics Event!() -> (then, Event:String) {is=title K2Node_FunctionEntry}
   > Branch(Condition=IsActivated())
   n2 = Get CrewProgressiveSelectionVM()
   > SendAnalyticsEvent!(
    execute
    self:Fort Progressive Item Screen Object Reference
    Event:String=@n1
    FullfillmentId:String=Get FulfillmentId!(self:Crew Progressive Set VM Object Reference=Get CrewProgressiveSetVM() -> (CrewProgressiveSetVM:Crew Progressive Set VM Object Reference)) -> (FulfillmentId:String)
    StageIndex:int=Get StageIndex(self=Get SelectedStage(self=@n2)) -> (StageIndex=0)
   ) -> (then)
  ]}
 SetCompactUIStyle {nodes [
   SetCompactUIStyle() {is=title K2Node_FunctionEntry}
   > Branch(Condition=IsMobileGame())
   > SetPadding!(
    execute
    self:Overlay Slot Object Reference=SlotAsOverlaySlot(Widget=Get SafeZone_Mobile() -> (SafeZone_Mobile:Safe Zone Object Reference))
    InPadding_Left:float=48
    InPadding_Top:float=24
    InPadding_Right:float=48
    InPadding_Bottom:float=24
   ) -> (then)
   > SetPadding(
    self=SlotAsOverlaySlot(Widget=Get Overlay_CameraFraming() -> (Overlay_CameraFraming:Overlay Object Reference))
    InPadding_Left=0
    InPadding_Top=0
    InPadding_Right=0
    InPadding_Bottom=0
   )
   > SetPadding(
    self=SlotAsOverlaySlot(Widget=Get HB_MainContent() -> (HB_MainContent:Horizontal Box Object Reference))
    InPadding_Left=0
    InPadding_Top=0
    InPadding_Right=0
    InPadding_Bottom=0
   )
   > SetPadding(
    self=SlotAsOverlaySlot(Widget=Get ProgressiveItem_Header() -> (ProgressiveItem_Header:Progressive Item State Title Object Reference))
    InPadding_Left=48
    InPadding_Top=128
    InPadding_Right=0
    InPadding_Bottom=0
   )
   > SetPadding(
    self=SlotAsOverlaySlot(Widget=Get WBP_PlayerInfo() -> (WBP_PlayerInfo:WBP Player Info Object Reference))
    InPadding_Left=0
    InPadding_Top=0
    InPadding_Right=0
    InPadding_Bottom=0
   )
   > SetPadding(
    self=SlotAsOverlaySlot(Widget=Get WBP_ActionBarContainer() -> (WBP_ActionBarContainer:WBP Action Bar Container Object Reference))
    InPadding_Left=0
    InPadding_Top=0
    InPadding_Right=0
    InPadding_Bottom=0
   )
   > SetPadding(self=SlotAsOverlaySlot(Widget=Get RightSideVertBox()), InPadding_Left=0, InPadding_Top=0, InPadding_Right=0, InPadding_Bottom=128)
   > SetPadding(
    self=SlotAsOverlaySlot(Widget=Get SizeBox_ProgressiveStageList() -> (SizeBox_ProgressiveStageList:Size Box Object Reference))
    InPadding_Left=48
    InPadding_Top=272
    InPadding_Right=0
    InPadding_Bottom=64
   )
   > SetPadding(self=SlotAsOverlaySlot(Widget=Get FortItemPreviewWidget()), InPadding_Left=0, InPadding_Top=128, InPadding_Right=400, InPadding_Bottom=0)
   > SetPadding(self=SlotAsOverlaySlot(Widget=Get Overlay_CameraFraming()), InPadding_Left=0, InPadding_Top=0, InPadding_Right=0, InPadding_Bottom=0)
   > SetRequiresHold!(execute, self:Common Button Base Object Reference=Get ConfirmPurchase_Button(), bInRequiresHold:bool=false) -> (then)
  ]}
 Stage Selection Changed {nodes [
   n1 = Stage Selection Changed!() -> (then, Stage:Crew Progressive Stage VM Object Reference) {is=title K2Node_FunctionEntry}
   > Send Analytics Event(Event=StageItemSelected) {is=title K2Node_CallFunction}
   n2 = Get ListView_Stages()
   > Set Selected Item(self=@n2, Item=@n1)
   > Navigate To Item(self=@n2, Item=@n1)
   > Scroll Item Into View(self=@n2, Item=@n1)
  ]}
 Update Selected Set {nodes [
   n1 = Update Selected Set!() -> (then, Set:Crew Progressive Set VM Object Reference) {is=title K2Node_FunctionEntry}
   n2 = Get CrewVM()
   > Branch(Condition=AND Boolean!(A=IsValid(Object=@n2), B=IsValid(Object=@n1)) -> (ReturnValue))
   > Setup Header Info!(
    execute
    self:Progressive Item State Title Object Reference=Get ProgressiveItem_Header()
    InputPin:"Text (by ref)"=Get Name(self=@n1)
    Subscribed:bool=IsSubscribed(self=@n2)
    Current Stage:int=Get UnlockedStages!(self:Crew Progressive Set VM Object Reference=@n1) -> (UnlockedStages:int=0)
    Max Stage:int=Length(TargetArray:Array of Crew Progressive Stage VM Object References=Get Stages!(self:Crew Progressive Set VM Object Reference=@n1) -> (Stages:Array of Crew Progressive Stage VM Object References))
   ) -> (then)
  ]}
 Update Unlocked Stages {nodes [
   Update Unlocked Stages!() -> (then, Unlocked Stages:int) {is=title K2Node_FunctionEntry}
   n1 = GetStageCount!(self:Crew Progressive Set VM Object Reference=Get CrewProgressiveSetVM()) -> (ReturnValue:int=0)
   n2 = Get Unlocked Stages() -> (Unlocked Stages:int=0)
   > SetVisibility(
    self=Get Text_NewStagesUnlockFinePrint() -> (Text_NewStagesUnlockFinePrint:Common Text Object Reference)
    InVisibility=Select(Option 0=Collapsed, Option 1=SelfHitTestInvisible, Index="integer < integer"(A=@n2, B=@n1))
   )
   n3 = Get Unlocked Stages() -> (Unlocked Stages=0)
   > SetVisibility(
    self=Get Text_ExpirationFinePrint() -> (Text_ExpirationFinePrint:Common Text Object Reference)
    InVisibility=Select(Option 0=Collapsed, Option 1=SelfHitTestInvisible, Index="Equal (Integer)"(A=@n3))
   )
  ]}
 UpdateSubscribedVisuals {nodes [
   UpdateSubscribedVisuals!() -> (then, Subscribed:bool=~) {is=title K2Node_FunctionEntry}
   > Update Selected Set!(execute, self, Set:Crew Progressive Set VM Object Reference=Get CrewProgressiveSetVM()) -> (then) {is=title K2Node_CallFunction}
   > n1 = Sequence()
   > n2 = Branch(Condition=Get Subscribed() -> (Subscribed:bool))
   n3 = Get BottomScreenGradient() -> (BottomScreenGradient:Image Object Reference)
   > n4 = GetDynamicMaterial(self=@n3)
   > SetVectorParameterValue(self=@n4, ParameterName=Color1, Value=(R=0.2462,G=0.0685,B=0.0103,A=1))
   > SetVectorParameterValue(self=@n4, ParameterName=Color2, Value=(R=0.0999,G=0.0232,B=0.0037,A=0))
   n5 = Get Text_ExpirationFinePrint()
   n6 = Get SubscribedTextColor_Legal() -> (SubscribedTextColor_Legal:Slate Color Structure)
   > SetColorAndOpacity(self=@n5, InColorAndOpacity=@n6)
   n7 = Get Text_NewStagesUnlockFinePrint()
   > SetColorAndOpacity(self=@n7, InColorAndOpacity=@n6)
   n8 = GetDynamicMaterial(execute=@n2.else, self=@n3)
   > SetVectorParameterValue(self=@n8, ParameterName=Color1, Value=(R=0.0018,G=0.0356,B=0.2918,A=0.702))
   > SetVectorParameterValue(self=@n8, ParameterName=Color2, Value=(R=0.0021,G=0.0437,B=0.3564,A=0))
   n9 = Get UnsubscribedTextColor_Legal() -> (UnsubscribedTextColor_Legal:Slate Color Structure)
   > SetColorAndOpacity(self=@n5, InColorAndOpacity=@n9)
   > SetColorAndOpacity(self=@n7, InColorAndOpacity=@n9)
   n10 = SelectFloat(A=1, B=0, bPickA=Get Subscribed())
   n11 = Get FortItemPreviewWidget()
   RequestChangeBackgroundMaterialScalarParameter!(execute=@n1.then_1, self:Fort Item Preview Screen Object Reference=@n11, ParameterName:Name=CrewLegacyUnsubscribed, Value:float=@n10) -> (then)
   > RequestChangeFloorMaterialScalarParameter!(execute, self:Fort Item Preview Screen Object Reference=@n11, ParameterName:Name=CrewLegacyUnsubscribed, Value:float=@n10) -> (then)
  ]}
]
variables [
 BaseColor {type=LinearColor, default=<rgba 0 0.25 1 1>}
 HoverColor {type=LinearColor, default=<rgba 0 0.5 1 1>}
 PlatformIconSize {type=Vector2D, default=<40 40>}
 PreviewPlatformIcon {type=boolean, default=true}
 SelectColor {type=LinearColor, default=<rgba 0.5 1 1 1>}
]
widgets {
 info {
  inheritedWidgetCount=0
  namedSlotCount=1
  parentClass=/Script/CrewUI.FortProgressiveItemScreen
  rootWidgetClass=&Overlay
  widgetCount=37
 }
 widgets [
  Overlay_1 {widgetClassPath=&Overlay, widget=Overlay_1}
  BottomContentBlockerGradient {
   widgetClassPath=&Image
   bIsVariable=true
   parent=Overlay_1
   widget=BottomContentBlockerGradient
   properties {
    visibility=Collapsed
    brush {
     bIsDynamicallyLoaded=false
     drawAs=Image
     imageType=NoImage
     mirroring=NoMirror
     resourceName=None
     resourceObject=/Game/UI/ContentControls/Materials/M_UI_ContentControls_LinearGradient
     tiling=NoTile
     imageSize=<32 32>
     margin=<ltrb 0 0 0 0>
     outlineSettings {
      bUseBrushTransparency=false
      roundingType=HalfHeightRadius
      width=0
      color {colorUseRule=UseColor_Specified, specifiedColor=<rgba 0 0 0 0>}
      cornerRadii=<0 0 0 0>
     }
     tintColor {colorUseRule=UseColor_Specified, specifiedColor=<rgba 1 1 1 1>}
     uVRegion {bIsValid=false, max=<0 0>, min=<0 0>}
    }
    colorAndOpacity=<rgba 1 1 1 0.6>
   }
   slot {class=OverlaySlot, properties {horizontalAlignment=HAlign_Fill, verticalAlignment=VAlign_Fill}}
  }
  BottomScreenGradient {
   widgetClassPath=&Image
   bIsVariable=true
   parent=Overlay_1
   widget=BottomScreenGradient
   properties {
    visibility=SelfHitTestInvisible
    brush {
     bIsDynamicallyLoaded=false
     drawAs=Image
     imageType=NoImage
     mirroring=NoMirror
     resourceName=None
     resourceObject=/Crew/UI/Materials/MI_UI_CrewLegacy_GradientOverlay
     tiling=NoTile
     imageSize=<1200 704>
     margin=<ltrb 0 0 0 0>
     outlineSettings {
      bUseBrushTransparency=false
      roundingType=HalfHeightRadius
      width=0
      color {colorUseRule=UseColor_Specified, specifiedColor=<rgba 0 0 0 0>}
      cornerRadii=<0 0 0 0>
     }
     tintColor {colorUseRule=UseColor_Specified, specifiedColor=<rgba 1 1 1 1>}
     uVRegion {bIsValid=false, max=<0 0>, min=<0 0>}
    }
   }
   slot {class=OverlaySlot, properties {horizontalAlignment=HAlign_Fill, verticalAlignment=VAlign_Bottom}}
  }
  SafeZone_Mobile {
   widgetClassPath=&SafeZone
   bIsVariable=true
   parent=Overlay_1
   widget=SafeZone_Mobile
   slot {class=OverlaySlot, properties {horizontalAlignment=HAlign_Fill, verticalAlignment=VAlign_Fill}}
  }
  Overlay_2 {widgetClassPath=&Overlay, parent=SafeZone_Mobile, widget=Overlay_2, slot {class=SafeZoneSlot}}
  Overlay_CameraFraming {
   widgetClassPath=&Overlay
   bIsVariable=true
   parent=Overlay_2
   widget=Overlay_CameraFraming
   slot {class=OverlaySlot, properties {horizontalAlignment=HAlign_Fill, verticalAlignment=VAlign_Fill, padding=<ltrb 40 0 0 0>}}
  }
  Widget_CameraFrame {
   widgetClassPath=&Overlay
   parent=Overlay_CameraFraming
   widget=Widget_CameraFrame
   slot {class=OverlaySlot, properties {horizontalAlignment=HAlign_Right, verticalAlignment=VAlign_Fill}}
  }
  Spacer_FrameSize {widgetClassPath=&Spacer, parent=Widget_CameraFrame, widget=Spacer_FrameSize, properties {size=<1200 0>}, slot {class=OverlaySlot}}
  FortItemPreviewWidget {
   widgetClassPath=/FortniteCoreUI/Frontend/ItemPreview/WBP_ItemPreview
   bIsVariable=true
   parent=Widget_CameraFrame
   widget=FortItemPreviewWidget
   properties {
    bSupportsActivationFocus=false
    itemViewObject=&self:WidgetTree.FortItemPreviewWidget.ItemViewObject
    previewActionButtons [&self:WidgetTree.Button_PreviewAction]
    previewSceneChangerTag {tagName=UI.ItemPreview.SceneChanger.FortniteCrew}
   }
   slot {class=OverlaySlot, properties {horizontalAlignment=HAlign_Fill, verticalAlignment=VAlign_Fill}}
  }
  RightSideVertBox {
   widgetClassPath=&VerticalBox
   bIsVariable=true
   parent=Overlay_CameraFraming
   widget=RightSideVertBox
   slot {class=OverlaySlot, properties {horizontalAlignment=HAlign_Right, verticalAlignment=VAlign_Bottom, padding=<ltrb 0 0 40 110>}}
  }
  ErrorInfoFlag {
   widgetClassPath=/Crew/UI/ErrorInfoFlag
   parent=RightSideVertBox
   widget=ErrorInfoFlag
   properties {accentColor=<rgba 0.7922 0.0078 0.1843 1>}
   slot {class=VerticalBoxSlot, properties {horizontalAlignment=HAlign_Right, padding=<ltrb 0 0 0 20>}}
  }
  Widget_ProgressiveItemDetails {
   widgetClassPath=&/ProgressiveItemDetails
   parent=RightSideVertBox
   widget=Widget_ProgressiveItemDetails
   slot {class=VerticalBoxSlot, properties {padding=<ltrb 0 0 0 20>}}
  }
  ConfirmPurchase_Button {
   widgetClassPath=/FortUILibrary/UIKit/Buttons/WBP_UIKit_Button_Loud
   bIsVariable=true
   parent=RightSideVertBox
   widget=ConfirmPurchase_Button
   properties {
    bIsFocusable=false
    bRequiresHold=true
    buttonSize=Large
    minHeight=56
    minWidth=632
    overrideSecondaryTextSizeMultiplier=0.5
    secondaryText=$11.99
    text=Confirm Purchase
    triggeringInputAction {dataTable=/FortUILibrary/Input/ItemShopInputActionDataTable, rowName=PurchaseCrewSetHold}
   }
   slot {class=VerticalBoxSlot, properties {horizontalAlignment=HAlign_Right, verticalAlignment=VAlign_Center}}
  }
  WBP_ContentPurchaseBlocker {
   widgetClassPath=/FortniteCoreUI/UI/ContentControls/WBP_ContentPurchaseBlocker
   bIsVariable=true
   parent=Overlay_CameraFraming
   widget=WBP_ContentPurchaseBlocker
   properties {
    text_Details=To purchase this item or bundle ask your parent or guardian to set up your account from Parental Controls.
    visibility=Collapsed
   }
   slot {class=OverlaySlot, properties {horizontalAlignment=HAlign_Right, verticalAlignment=VAlign_Center, padding=<ltrb 0 0 40 0>}}
  }
  HB_MainContent {
   widgetClassPath=&HorizontalBox
   bIsVariable=true
   parent=Overlay_2
   widget=HB_MainContent
   slot {class=OverlaySlot, properties {horizontalAlignment=HAlign_Fill, verticalAlignment=VAlign_Fill, padding=<ltrb 24 64 24 24>}}
  }
  Overlay_MainContent {
   widgetClassPath=&Overlay
   parent=HB_MainContent
   widget=Overlay_MainContent
   properties {renderTransformPivot=<0 0.5>}
   slot {class=HorizontalBoxSlot, properties {horizontalAlignment=HAlign_Left, size {sizeRule=Automatic, value=0.4}}}
  }
  ScrollContainer {
   widgetClassPath=&Overlay
   parent=Overlay_MainContent
   widget=ScrollContainer
   slot {class=OverlaySlot, properties {horizontalAlignment=HAlign_Fill, verticalAlignment=VAlign_Fill}}
  }
  SizeBox_ProgressiveStageList {
   widgetClassPath=&SizeBox
   bIsVariable=true
   parent=ScrollContainer
   widget=SizeBox_ProgressiveStageList
   properties {heightOverride=795}
   slot {class=OverlaySlot, properties {horizontalAlignment=HAlign_Fill, verticalAlignment=VAlign_Fill, padding=<ltrb 92 190 0 58>}}
  }
  ListView_Stages {
   widgetClassPath=/Script/FortniteUI.FortTileView
   bIsVariable=true
   parent=SizeBox_ProgressiveStageList
   widget=ListView_Stages
   properties {
    entryHeight=246
    entryWidgetClass=&/ProgressiveStageWidget.ProgressiveStageWidget_C
    entryWidth=1120
    scrollBarStyle {
     thickness=0
     draggedThumbImage {
      bIsDynamicallyLoaded=false
      drawAs=Box
      imageType=FullColor
      mirroring=NoMirror
      resourceName=../../../Engine/Content/Slate/Common/Scrollbar_Thumb.png
      resourceObject=null
      tiling=NoTile
      imageSize=<0 0>
      margin=<ltrb 0.25 0.25 0.25 0.25>
      outlineSettings {
       bUseBrushTransparency=false
       roundingType=HalfHeightRadius
       width=0
       color {colorUseRule=UseColor_Specified, specifiedColor=<rgba 0 0 0 0>}
       cornerRadii=<0 0 0 0>
      }
      tintColor {colorUseRule=UseColor_Specified, specifiedColor=<rgba 1 1 1 1>}
      uVRegion {bIsValid=false, max=<0 0>, min=<0 0>}
     }
     horizontalBackgroundImage {
      bIsDynamicallyLoaded=false
      drawAs=NoDrawType
      imageType=NoImage
      mirroring=NoMirror
      resourceName=None
      resourceObject=null
      tiling=NoTile
      imageSize=<0 0>
      margin=<ltrb 0 0 0 0>
      outlineSettings {
       bUseBrushTransparency=false
       roundingType=HalfHeightRadius
       width=0
       color {colorUseRule=UseColor_Specified, specifiedColor=<rgba 0 0 0 0>}
       cornerRadii=<0 0 0 0>
      }
      tintColor {colorUseRule=UseColor_Specified, specifiedColor=<rgba 1 1 1 1>}
      uVRegion {bIsValid=false, max=<0 0>, min=<0 0>}
     }
     horizontalBottomSlotImage {
      bIsDynamicallyLoaded=false
      drawAs=Image
      imageType=FullColor
      mirroring=NoMirror
      resourceName=../../../Engine/Content/Slate/Common/Scrollbar_Background_Horizontal.png
      resourceObject=null
      tiling=NoTile
      imageSize=<8 8>
      margin=<ltrb 0 0 0 0>
      outlineSettings {
       bUseBrushTransparency=false
       roundingType=HalfHeightRadius
       width=0
       color {colorUseRule=UseColor_Specified, specifiedColor=<rgba 0 0 0 0>}
       cornerRadii=<0 0 0 0>
      }
      tintColor {colorUseRule=UseColor_Specified, specifiedColor=<rgba 1 1 1 1>}
      uVRegion {bIsValid=false, max=<0 0>, min=<0 0>}
     }
     horizontalTopSlotImage {
      bIsDynamicallyLoaded=false
      drawAs=Image
      imageType=FullColor
      mirroring=NoMirror
      resourceName=../../../Engine/Content/Slate/Common/Scrollbar_Background_Horizontal.png
      resourceObject=null
      tiling=NoTile
      imageSize=<8 8>
      margin=<ltrb 0 0 0 0>
      outlineSettings {
       bUseBrushTransparency=false
       roundingType=HalfHeightRadius
       width=0
       color {colorUseRule=UseColor_Specified, specifiedColor=<rgba 0 0 0 0>}
       cornerRadii=<0 0 0 0>
      }
      tintColor {colorUseRule=UseColor_Specified, specifiedColor=<rgba 1 1 1 1>}
      uVRegion {bIsValid=false, max=<0 0>, min=<0 0>}
     }
     hoveredThumbImage {
      bIsDynamicallyLoaded=false
      drawAs=Box
      imageType=FullColor
      mirroring=NoMirror
      resourceName=../../../Engine/Content/Slate/Common/Scrollbar_Thumb.png
      resourceObject=null
      tiling=NoTile
      imageSize=<0 0>
      margin=<ltrb 0.25 0.25 0.25 0.25>
      outlineSettings {
       bUseBrushTransparency=false
       roundingType=HalfHeightRadius
       width=0
       color {colorUseRule=UseColor_Specified, specifiedColor=<rgba 0 0 0 0>}
       cornerRadii=<0 0 0 0>
      }
      tintColor {colorUseRule=UseColor_Specified, specifiedColor=<rgba 1 1 1 1>}
      uVRegion {bIsValid=false, max=<0 0>, min=<0 0>}
     }
     normalThumbImage {
      bIsDynamicallyLoaded=false
      drawAs=Box
      imageType=FullColor
      mirroring=NoMirror
      resourceName=../../../Engine/Content/Slate/Common/Scrollbar_Thumb.png
      resourceObject=null
      tiling=NoTile
      imageSize=<0 0>
      margin=<ltrb 0.25 0.25 0.25 0.25>
      outlineSettings {
       bUseBrushTransparency=false
       roundingType=HalfHeightRadius
       width=0
       color {colorUseRule=UseColor_Specified, specifiedColor=<rgba 0 0 0 0>}
       cornerRadii=<0 0 0 0>
      }
      tintColor {colorUseRule=UseColor_Specified, specifiedColor=<rgba 1 1 1 1>}
      uVRegion {bIsValid=false, max=<0 0>, min=<0 0>}
     }
     verticalBackgroundImage {
      bIsDynamicallyLoaded=false
      drawAs=NoDrawType
      imageType=NoImage
      mirroring=NoMirror
      resourceName=None
      resourceObject=null
      tiling=NoTile
      imageSize=<0 0>
      margin=<ltrb 0 0 0 0>
      outlineSettings {
       bUseBrushTransparency=false
       roundingType=HalfHeightRadius
       width=0
       color {colorUseRule=UseColor_Specified, specifiedColor=<rgba 0 0 0 0>}
       cornerRadii=<0 0 0 0>
      }
      tintColor {colorUseRule=UseColor_Specified, specifiedColor=<rgba 1 1 1 1>}
      uVRegion {bIsValid=false, max=<0 0>, min=<0 0>}
     }
     verticalBottomSlotImage {
      bIsDynamicallyLoaded=false
      drawAs=Image
      imageType=FullColor
      mirroring=NoMirror
      resourceName=../../../Engine/Content/Slate/Common/Scrollbar_Background_Vertical.png
      resourceObject=null
      tiling=NoTile
      imageSize=<8 8>
      margin=<ltrb 0 0 0 0>
      outlineSettings {
       bUseBrushTransparency=false
       roundingType=HalfHeightRadius
       width=0
       color {colorUseRule=UseColor_Specified, specifiedColor=<rgba 0 0 0 0>}
       cornerRadii=<0 0 0 0>
      }
      tintColor {colorUseRule=UseColor_Specified, specifiedColor=<rgba 1 1 1 1>}
      uVRegion {bIsValid=false, max=<0 0>, min=<0 0>}
     }
     verticalTopSlotImage {
      bIsDynamicallyLoaded=false
      drawAs=Image
      imageType=FullColor
      mirroring=NoMirror
      resourceName=../../../Engine/Content/Slate/Common/Scrollbar_Background_Vertical.png
      resourceObject=null
      tiling=NoTile
      imageSize=<8 8>
      margin=<ltrb 0 0 0 0>
      outlineSettings {
       bUseBrushTransparency=false
       roundingType=HalfHeightRadius
       width=0
       color {colorUseRule=UseColor_Specified, specifiedColor=<rgba 0 0 0 0>}
       cornerRadii=<0 0 0 0>
      }
      tintColor {colorUseRule=UseColor_Specified, specifiedColor=<rgba 1 1 1 1>}
      uVRegion {bIsValid=false, max=<0 0>, min=<0 0>}
     }
    }
   }
   slot {class=SizeBoxSlot}
  }
  LegalVertBox {
   widgetClassPath=&VerticalBox
   parent=Overlay_MainContent
   widget=LegalVertBox
   slot {class=OverlaySlot, properties {verticalAlignment=VAlign_Bottom, padding=<ltrb 8 8 8 -16>}}
  }
  Text_NewStagesUnlockFinePrint {
   widgetClassPath=&CommonTextBlock
   bIsVariable=true
   parent=LegalVertBox
   widget=Text_NewStagesUnlockFinePrint
   properties {
    mobileFontSizeMultiplier=1.12
    text="[PH] New Stages unlock on the 15th of each month while active"
    wrapTextAt=660
    colorAndOpacity {colorUseRule=UseColor_Specified, specifiedColor=<rgba 0.0331 0.855 1 1>}
    font {
     bForceMonospaced=false
     bMaterialIsStencil=false
     fontMaterial=null
     fontObject=&BurbankSmall
     letterSpacing=0
     monospacedWidth=1
     size=13
     skewAmount=0
     typefaceFontName=Bold
     outlineSettings {
      bApplyOutlineToDropShadows=false
      bMiteredCorners=false
      bSeparateFillAlpha=false
      outlineMaterial=null
      outlineSize=0
      outlineColor=<rgba 0 0 0 1>
     }
    }
    shadowOffset=<0 0>
   }
   slot {class=VerticalBoxSlot}
  }
  Text_ExpirationFinePrint {
   widgetClassPath=&CommonTextBlock
   bIsVariable=true
   parent=LegalVertBox
   widget=Text_ExpirationFinePrint
   properties {
    mobileFontSizeMultiplier=1.12
    text="[PH] Ability to unlock until 9/15/22, 8:00:00 PM"
    wrapTextAt=660
    colorAndOpacity {colorUseRule=UseColor_Specified, specifiedColor=<rgba 0.0331 0.855 1 1>}
    font {
     bForceMonospaced=false
     bMaterialIsStencil=false
     fontMaterial=null
     fontObject=&BurbankSmall
     letterSpacing=0
     monospacedWidth=1
     size=13
     skewAmount=0
     typefaceFontName=Bold
     outlineSettings {
      bApplyOutlineToDropShadows=false
      bMiteredCorners=false
      bSeparateFillAlpha=false
      outlineMaterial=null
      outlineSize=0
      outlineColor=<rgba 0 0 0 1>
     }
    }
    shadowOffset=<0 0>
   }
   slot {class=VerticalBoxSlot}
  }
  VerticalBox_Pips {
   widgetClassPath=&VerticalBox
   parent=Overlay_2
   widget=VerticalBox_Pips
   slot {class=OverlaySlot, properties {horizontalAlignment=HAlign_Fill, verticalAlignment=VAlign_Fill}}
  }
  Spacer_66 {
   widgetClassPath=&Spacer
   parent=VerticalBox_Pips
   widget=Spacer_66
   slot {class=VerticalBoxSlot, properties {horizontalAlignment=HAlign_Left, size {sizeRule=Fill, value=0.19}}}
  }
  ScaleBox_0 {
   widgetClassPath=&ScaleBox
   parent=VerticalBox_Pips
   widget=ScaleBox_0
   properties {stretch=ScaleToFitY, stretchDirection=DownOnly}
   slot {class=VerticalBoxSlot, properties {size {sizeRule=Fill, value=0.6}}}
  }
  SizeBox_EntryBox_Pips {
   widgetClassPath=&SizeBox
   parent=ScaleBox_0
   widget=SizeBox_EntryBox_Pips
   properties {bOverride_WidthOverride=true, widthOverride=64}
   slot {class=ScaleBoxSlot, properties {horizontalAlignment=HAlign_Left}}
  }
  ListView_StagePips {
   widgetClassPath=/Script/FortniteUI.FortItemListView
   parent=SizeBox_EntryBox_Pips
   widget=ListView_StagePips
   properties {bIsFocusable=false, entryWidgetClass=&/ProgressiveItemScreenPip.ProgressiveItemScreenPip_C}
   slot {class=SizeBoxSlot}
  }
  Spacer {
   widgetClassPath=&Spacer
   parent=VerticalBox_Pips
   widget=Spacer
   slot {class=VerticalBoxSlot, properties {horizontalAlignment=HAlign_Left, size {sizeRule=Fill, value=0.19}}}
  }
  ProgressiveItem_Header {
   widgetClassPath=&/ProgressiveItemStateTitle
   bIsVariable=true
   parent=Overlay_2
   widget=ProgressiveItem_Header
   slot {class=OverlaySlot, properties {padding=<ltrb 110 92 0 0>}}
  }
  WBP_PlayerInfo {
   widgetClassPath=/Game/UI/Frontend/ItemShop/MultiProduct/WBP_PlayerInfo
   bIsVariable=true
   parent=Overlay_2
   widget=WBP_PlayerInfo
   slot {class=OverlaySlot, properties {horizontalAlignment=HAlign_Right, padding=<ltrb 0 70 48 0>}}
  }
  WBP_ActionBarContainer {
   widgetClassPath=/Game/UI/Frontend/BottomBar/WBP_ActionBarContainer
   bIsVariable=true
   parent=Overlay_2
   widget=WBP_ActionBarContainer
   slot {class=OverlaySlot, properties {horizontalAlignment=HAlign_Right, verticalAlignment=VAlign_Bottom, padding=<ltrb 0 0 48 24>}}
  }
  HB_ActionButtons {widgetClassPath=&HorizontalBox, namedSlotHost=WBP_ActionBarContainer, widget=HB_ActionButtons}
  Button_PreviewStyles {
   widgetClassPath=/FortUILibrary/UIKit/Buttons/WBP_UIKit_Button_Quiet
   bIsVariable=true
   parent=HB_ActionButtons
   widget=Button_PreviewStyles
   properties {
    text=Preview Styles
    visibility=Collapsed
    triggeringInputAction {dataTable=&InputActionDataTable, rowName=GenericForward}
   }
   slot {class=HorizontalBoxSlot}
  }
  Button_PreviewAction {
   widgetClassPath=&WBP_UIKit_Button_Quiet
   bIsVariable=true
   parent=HB_ActionButtons
   widget=Button_PreviewAction
   properties {bAutoUpdateFromInputAction=true, text=Preview Action}
   slot {class=HorizontalBoxSlot, properties {padding=<ltrb 24 0 0 0>}}
  }
  Button_MoreInfo {
   widgetClassPath=&WBP_UIKit_Button_Quiet
   bIsVariable=true
   parent=HB_ActionButtons
   widget=Button_MoreInfo
   properties {text=More Info, triggeringInputAction {dataTable=&InputActionDataTable, rowName=GenericFaceButtonTop}}
   slot {class=HorizontalBoxSlot, properties {padding=<ltrb 24 0 0 0>}}
  }
  WBP_AthenaBottomBar_WithNotifications {
   widgetClassPath=/Game/UI/Frontend/BottomBar/WBP_AthenaBottomBar_WithNotifications
   bIsVariable=true
   parent=HB_ActionButtons
   widget=WBP_AthenaBottomBar_WithNotifications
   slot {class=HorizontalBoxSlot, properties {padding=<ltrb 24 0 0 0>}}
  }
  Button_Back {
   widgetClassPath=&WBP_UIKit_Button_Quiet
   bIsVariable=true
   parent=HB_ActionButtons
   widget=Button_Back
   properties {text=Back, triggeringInputAction {dataTable=&InputActionDataTable, rowName=GenericBack}}
   slot {class=HorizontalBoxSlot, properties {padding=<ltrb 24 0 0 0>}}
  }
 ]
}
