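# Tall CommonButtonBase tile for a Crew progressive cosmetic set, showing item art, stage progress and a check or clock badge.
# Setting the set view model fires a dispatcher that async-applies the tile material; loading-state changes recolor materials by subscription.
# Clicking writes the set to the selection view model.
# StageSelected and hover changes drive guarded animations, while helper functions derive stage text, subtitle and badge from progress.
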WidgetBlueprint /Crew/UI/ProgressiveCosmetic/ProgressiveItem_Tall
animations [
 Selected {
  binding_count=9
  length_seconds=0.1024
  object=&self:Selected
  sequence {bindings [
    TileText_ItemName {
     id=#1
     possessed_class=AthenaItemShopReactiveTileText
     tracks [
      {
       type=MovieScene2DTransformTrack
       sections [{
         type=MovieScene2DTransformSection
         channels [
          Translation.Y {type=MovieSceneScriptingFloatChannel, keys [[0.048,0],[2.048,-16]]}
          Angle {type=MovieSceneScriptingFloatChannel, const [0,1]}
          Shear.Y {type=MovieSceneScriptingFloatChannel, keys [[0.048,0],[2.048,-3]]}
         ]
        }]
      }
      {
       type=MovieSceneFloatTrack
       display_name=Render Opacity
       sections [{type=MovieSceneFloatSection, channels [None {type=MovieSceneScriptingFloatChannel, const [1,2]}]}]
      }
      {
       type=MovieSceneByteTrack
       display_name=Visibility
       sections [{type=MovieSceneByteSection, channels [None {type=MovieSceneScriptingByteChannel, const [4,1]}]}]
      }
     ]
    }
    OverlaySlot_2 {
     id=#2
     parent_id=#3
     possessed_class=OverlaySlot
     tracks [{
       type=MovieSceneMarginTrack
       display_name=Padding
       sections [{
         type=MovieSceneMarginSection
         end_frame=2
         start_frame=0
         channels [
          Left {type=MovieSceneScriptingFloatChannel, keys [[0,0],[2.0483,-10]]}
          Top {type=MovieSceneScriptingFloatChannel, keys [[0,0],[2.0483,-10]]}
          Right {type=MovieSceneScriptingFloatChannel, keys [[0,0],[2.0483,-10]]}
          Bottom {type=MovieSceneScriptingFloatChannel, keys [[0,0],[2.0483,-10]]}
         ]
        }]
      }]
    }
    Text_Stages {id=#4, possessed_class=CommonTextBlock, tracks []}
    Image_Item {
     id=#5
     possessed_class=ScaleBox
     tracks [{
       type=MovieScene2DTransformTrack
       sections [{
         type=MovieScene2DTransformSection
         channels [
          Scale.X {type=MovieSceneScriptingFloatChannel, keys [[0,1],[2.048,1.1]]}
          Scale.Y {type=MovieSceneScriptingFloatChannel, keys [[0,1],[2.048,1.1]]}
         ]
        }]
      }]
    }
    Image_Dimmer {
     id=#6
     possessed_class=Image
     tracks [{
       type=MovieSceneFloatTrack
       display_name=Render Opacity
       sections [{type=MovieSceneFloatSection, channels [None {type=MovieSceneScriptingFloatChannel, keys [[0,0.9],[2,0]]}]}]
      }]
    }
    OverlaySlot_1 {
     id=#7
     parent_id=#4
     possessed_class=OverlaySlot
     tracks [{
       type=MovieSceneMarginTrack
       display_name=Padding
       sections [{
         type=MovieSceneMarginSection
         channels [
          Left {type=MovieSceneScriptingFloatChannel, const [0,1]}
          Top {type=MovieSceneScriptingFloatChannel, const [0,1]}
          Right {type=MovieSceneScriptingFloatChannel, const [10,2]}
          Bottom {type=MovieSceneScriptingFloatChannel, keys [[0,0],[2,4]]}
         ]
        }]
      }]
    }
    Overlay_67 {
     id=#8
     possessed_class=Overlay
     tracks [{
       type=MovieScene2DTransformTrack
       sections [{
         type=MovieScene2DTransformSection
         channels [
          Scale.X {type=MovieSceneScriptingFloatChannel, keys [[0,1],[2,1.05]]}
          Scale.Y {type=MovieSceneScriptingFloatChannel, keys [[0,1],[2,1.05]]}
         ]
        }]
      }]
    }
    Image_Frame {id=#3, possessed_class=Image, tracks []}
    Image_PriceBg {
     id=#9
     possessed_class=Image
     tracks [{
       type=MovieSceneWidgetMaterialTrack
       sections [{
         type=MovieSceneParameterSection
         channels [
          TopAngle {type=MovieSceneScriptingFloatChannel, keys [[0,1],[2,10]]}
          TopOffset {type=MovieSceneScriptingFloatChannel, keys [[0,-0.35],[2,-0.2]]}
          HorizontalSharpness {type=MovieSceneScriptingFloatChannel, const [50,1]}
         ]
        }]
      }]
    }
   ]}
 }
 Deselected {
  binding_count=9
  length_seconds=0.1024
  object=&self:Deselected
  sequence {bindings [
    Image_Item {
     id=#10
     possessed_class=ScaleBox
     tracks [{
       type=MovieScene2DTransformTrack
       sections [{
         type=MovieScene2DTransformSection
         channels [
          Scale.X {type=MovieSceneScriptingFloatChannel, keys [[0,1.1],[2.048,1]]}
          Scale.Y {type=MovieSceneScriptingFloatChannel, keys [[0,1.1],[2.048,1]]}
         ]
        }]
      }]
    }
    Image_Frame {id=#11, possessed_class=Image, tracks []}
    OverlaySlot_2 {
     id=#12
     parent_id=#11
     possessed_class=OverlaySlot
     tracks [{
       type=MovieSceneMarginTrack
       display_name=Padding
       sections [{
         type=MovieSceneMarginSection
         end_frame=2
         start_frame=0
         channels [
          Left {type=MovieSceneScriptingFloatChannel, keys [[0,-10],[2.048,0]]}
          Top {type=MovieSceneScriptingFloatChannel, keys [[0,-10],[2.048,0]]}
          Right {type=MovieSceneScriptingFloatChannel, keys [[0,-10],[2.048,0]]}
          Bottom {type=MovieSceneScriptingFloatChannel, keys [[0,-10],[2.048,0]]}
         ]
        }]
      }]
    }
    Image_PriceBg {
     id=#13
     possessed_class=Image
     tracks [{
       type=MovieSceneWidgetMaterialTrack
       sections [{
         type=MovieSceneParameterSection
         channels [
          TopOffset {type=MovieSceneScriptingFloatChannel, keys [[0,-0.2],[2,-0.35]]}
          TopAngle {type=MovieSceneScriptingFloatChannel, keys [[0,10],[2,1]]}
          HorizontalSharpness {type=MovieSceneScriptingFloatChannel, const [50,1]}
         ]
        }]
      }]
    }
    Text_Stages {id=#14, possessed_class=CommonTextBlock, tracks []}
    TileText_ItemName {
     id=#15
     possessed_class=AthenaItemShopReactiveTileText
     tracks [
      {
       type=MovieSceneByteTrack
       display_name=Visibility
       sections [{type=MovieSceneByteSection, channels [None {type=MovieSceneScriptingByteChannel, const [2,1]}]}]
      }
      {
       type=MovieScene2DTransformTrack
       sections [{
         type=MovieScene2DTransformSection
         channels [
          Translation.Y {type=MovieSceneScriptingFloatChannel, keys [[0,-16],[2.048,0]]}
          Angle {type=MovieSceneScriptingFloatChannel, const [0,1]}
          Shear.Y {type=MovieSceneScriptingFloatChannel, keys [[0,-3],[0.048,0],[2.048,0]]}
         ]
        }]
      }
     ]
    }
    Overlay_67 {
     id=#16
     possessed_class=Overlay
     tracks [{
       type=MovieScene2DTransformTrack
       sections [{
         type=MovieScene2DTransformSection
         channels [
          Scale.X {type=MovieSceneScriptingFloatChannel, keys [[0,1.05],[2,1]]}
          Scale.Y {type=MovieSceneScriptingFloatChannel, keys [[0,1.05],[2,1]]}
         ]
        }]
      }]
    }
    Image_Dimmer {
     id=#17
     possessed_class=Image
     tracks [{
       type=MovieSceneFloatTrack
       display_name=Render Opacity
       sections [{type=MovieSceneFloatSection, channels [None {type=MovieSceneScriptingFloatChannel, keys [[0,0],[2,0.85],[2.048,0.9]]}]}]
      }]
    }
    OverlaySlot_1 {
     id=#18
     parent_id=#14
     possessed_class=OverlaySlot
     tracks [{
       type=MovieSceneMarginTrack
       display_name=Padding
       sections [{
         type=MovieSceneMarginSection
         channels [
          Left {type=MovieSceneScriptingFloatChannel, const [0,1]}
          Top {type=MovieSceneScriptingFloatChannel, const [0,1]}
          Right {type=MovieSceneScriptingFloatChannel, const [10,1]}
          Bottom {type=MovieSceneScriptingFloatChannel, keys [[0,4],[2,2],[2.048,0]]}
         ]
        }]
      }]
    }
   ]}
 }
 HoverBurst {
  binding_count=1
  length_seconds=0.1024
  object=&self:HoverBurst
  sequence {bindings [Image_Item {
     id=#19
     possessed_class=FortLazyImage
     tracks [{
       type=MovieSceneWidgetMaterialTrack
       sections [{
         type=MovieSceneParameterSection
         channels [
          Hover {type=MovieSceneScriptingFloatChannel, keys [[0,0],[2,1]]}
          HoverBurst {type=MovieSceneScriptingFloatChannel, keys [[0,0],[2,1],[2.048,2]]}
         ]
        }]
      }]
    }]}
 }
]
graphs [
 EventGraph {nodes [
   "On Loading State Changed (Image_Item)"!() -> (OutputDelegate, then, bIsLoading:bool) {is=title K2Node_ComponentBoundEvent}
   > UpdateSubscribedState!(
    execute
    self
    bInSubscribed:bool=IsSubscribed!(self:Crew VM Object Reference=Get CrewVM() -> (CrewVM:Crew VM Object Reference)) -> (ReturnValue:bool)
   ) -> (then)
   n1 = LoadTileMaterialItemImage_Event!() -> (OutputDelegate, then, CrewSetVM:Crew Progressive Set VM Object Reference) {is=title K2Node_CustomEvent}
   > ApplySetTileMaterial!(
    execute:Exec
    ContextWidget:Fort Lazy Image Object Reference=Get Image_Item() -> (Image_Item:Fort Lazy Image Object Reference)
    SetVM:Crew Progressive Set VM Object Reference=@n1.CrewSetVM
   ) -> (then:Exec, OnOperationCompleted:Exec, Success:bool=~) {is=title K2Node_AsyncAction}
   n2 = Event On Selected()
   Set SelectedSet!(
    execute=[@n3.then, @n2.then]
    SelectedSet:Crew Progressive Set VM Object Reference=Get CrewProgressiveSetVM() -> (CrewProgressiveSetVM:Crew Progressive Set VM Object Reference)
    self:Crew Progressive Selection VM Object Reference=Get CrewProgressiveSelectionVM() -> (CrewProgressiveSelectionVM:Crew Progressive Selection VM Object Reference)
   ) -> (then, Output_Get:Crew Progressive Set VM Object Reference)
   Event OnInitialized()
   > Bind Event to Load Tile Material Item Image(self:Progressive Item Tall Object Reference, Delegate=@n1) {is=title K2Node_AddDelegate}
   Event On Hovered()
   > n3 = Hover Changed!(execute, self, Hovered:bool=true) -> (then)
   Event On Unhovered()
   > Hover Changed(Hovered=false)
  ]}
 Get Stage Text {nodes [
   n1 = Get Stage Text!() -> (then, Unlocked Stages:Integer64, Total Stages:Integer64, Subscribed:bool=~, Progress:EProgressiveSetProgress Enum=NoProgress) {is=title K2Node_FunctionEntry}
   n2 = Format Text!(Format="LOCTEXT(\"Stage {0}/{1}\")", 0:Integer64=@n1, 1:Integer64=@n1.Total Stages) -> (Result)
   n3 = Format Text(Format="LOCTEXT(\"{0}/{1} Stages Unlocked\")", 0=@n1, 1=@n1.Total Stages)
   > Return Node!(
    execute
    Result:Text=Select(
     Option 0:Text=Select(Option 0=@n2, Option 1=@n3, Index=@n1.Subscribed)
     Option 1:Text="LOCTEXT(\"Completed!\")"
     Index="Equal (Enum)"(A:EProgressiveSetProgress Enum=@n1.Progress, B:EProgressiveSetProgress Enum=Completed)
    ) -> (ReturnValue:Text)
   )
  ]}
 Get Subtitle Test {nodes [
   n1 = Get Subtitle Test!() -> (then, Progress:EProgressiveSetProgress Enum=NoProgress, Unlock Fine Print:Text) {is=title K2Node_FunctionEntry}
   > Return Node(Result=Select!(
     NoProgress:Text="LOCTEXT(\"Stage 1 unlocks instantly\")"
     PartialProgress:Text=@n1.Unlock Fine Print
     Completed:Text
     Expired:Text=@n1.Unlock Fine Print
     Index:EProgressiveSetProgress Enum=@n1
    ) -> (ReturnValue))
  ]}
 Hover Changed {nodes [
   n1 = Hover Changed!() -> (then, Hovered:bool=~) {is=title K2Node_FunctionEntry}
   > Branch(Condition="Not Equal (Boolean)"(A=@n1, B=Get WasHovered() -> (WasHovered:bool)) -> (ReturnValue=~))
   > Set WasHovered(WasHovered:bool=@n1) -> (Output_Get:bool)
   > PlayAnimation(
    InAnimation=Get HoverBurst() -> (HoverBurst:Widget Animation Object Reference)
    PlayMode=Select!(Option 0:EUMGSequencePlayMode Enum=Reverse, Option 1:EUMGSequencePlayMode Enum, Index:bool=@n1) -> (ReturnValue:EUMGSequencePlayMode Enum)
   )
  ]}
 LoadTileMaterialItemImage {nodes [LoadTileMaterialItemImage!() -> (then, CrewSetVM:Crew Progressive Set VM Object Reference) {is=title K2Node_FunctionEntry}]}
 Set CrewProgressiveSetVM {nodes [
   n1 = Set CrewProgressiveSetVM!() -> (then, CrewProgressiveSetVM:Crew Progressive Set VM Object Reference) {is=title K2Node_FunctionEntry}
   > Call Load Tile Material Item Image!(execute:Exec, self:Progressive Item Tall Object Reference, CrewSetVM:Crew Progressive Set VM Object Reference=@n1) -> (then:Exec) {is=title K2Node_CallDelegate}
  ]}
 StageSelected {nodes [
   n1 = StageSelected!() -> (then, Selected:bool=~) {is=title K2Node_FunctionEntry}
   > Branch(Condition="Not Equal (Boolean)"(A=@n1, B=Get WasSelected() -> (WasSelected:bool)) -> (ReturnValue=~))
   > Set WasSelected(WasSelected:bool=@n1) -> (Output_Get:bool)
   > n2 = Branch(Condition=@n1)
   > StopAnimation(InAnimation=Get Deselected() -> (Deselected:Widget Animation Object Reference))
   > PlayAnimation(InAnimation=Get Selected() -> (Selected:Widget Animation Object Reference))
   n3 = Get Text_Stages() -> (Text_Stages:Athena Item Shop Reactive Tile Text Object Reference)
   > SetColorAndOpacity(self=@n3, InColorAndOpacity=Make Slate Color!(SpecifiedColor=(R=0,G=0,B=0,A=1)) -> (SlateColor))
   n4 = Get Image_PriceBg() -> (Image_PriceBg:Image Object Reference)
   > SetColorAndOpacity(self:Image Object Reference=@n4, InColorAndOpacity:Linear Color Structure=(R=1,G=1,B=1,A=1))
   n5 = Get CrewProgressiveSetVM()
   n6 = Get Overlay_InspectionText() -> (Overlay_InspectionText:Overlay Object Reference)
   > SetVisibility(
    self=@n6
    InVisibility=Select(
     Option 0:ESlateVisibility Enum=HitTestInvisible
     Option 1:ESlateVisibility Enum=Collapsed
     Index=GetIsCompleted!(self:Crew Progressive Set VM Object Reference=@n5) -> (ReturnValue:bool)
    ) -> (ReturnValue:ESlateVisibility Enum)
   )
   StopAnimation(execute=@n2.else, InAnimation=Get Selected())
   > PlayAnimation(InAnimation=Get Deselected())
   n7 = GetIsCompleted(self=Get CrewProgressiveSetVM())
   > SetColorAndOpacity(
    self:Text Object Reference=@n3
    InColorAndOpacity:Slate Color Structure=Make Slate Color(SpecifiedColor=SelectColor(A=(R=0.0343,G=1,B=0.227,A=1), B=(R=0.4722,G=0.4722,B=0.4722,A=1), bPickA=@n7))
   )
   > SetColorAndOpacity(self:Image Object Reference=@n4, InColorAndOpacity:Linear Color Structure=(R=0.0118,G=0.0118,B=0.0118,A=1))
   > SetVisibility(self=@n6, InVisibility=Collapsed)
  ]}
 UpdateCompletedState {nodes [
   n1 = UpdateCompletedState!() -> (then, Unlocked Stages:int) {is=title K2Node_FunctionEntry}
   n2 = Get CrewProgressiveSetVM()
   n3 = NOT Boolean(A="integer < integer"(A=@n1, B=GetStageCount!(self:Crew Progressive Set VM Object Reference=@n2) -> (ReturnValue:int=0)))
   n4 = "integer < integer"(A=@n1, B=1)
   n5 = Get Switcher_CheckDelay() -> (Switcher_CheckDelay:Common Visibility Switcher Object Reference)
   > SetVisibility(self=@n5, InVisibility=Select(Option 0=Collapsed, Option 1=HitTestInvisible, Index=OR Boolean(A=@n4, B=@n3)))
   > SetActiveWidgetIndex(self:Common Visibility Switcher Object Reference=@n5, Index=SelectInt(A=0, B=1, bPickA=@n3))
  ]}
 UpdateSubscribedState {nodes [
   n1 = UpdateSubscribedState!() -> (then, bInSubscribed:bool=~) {is=title K2Node_FunctionEntry}
   > n2 = GetDynamicMaterial(self=Get Image_Gradient() -> (Image_Gradient:Image Object Reference))
   > SetVectorParameterValue(self=@n2, ParameterName=Color2, Value=SelectColor(A=(R=0.3467,G=0.1946,B=0.0423,A=0), B=(R=0.0021,G=0.0437,B=0.3564,A=0), bPickA=@n1))
   > SetVectorParameterValue(self=@n2, ParameterName=Color1, Value=SelectColor(A=(R=0.3467,G=0.0953,B=0.0185,A=1), B=(R=0.0018,G=0.0356,B=0.2918,A=0.702), bPickA=@n1))
   > SetColorAndOpacity(
    self=Get Image_Dimmer() -> (Image_Dimmer:Image Object Reference)
    InColorAndOpacity=SelectColor(A=(R=0.147,G=0.0284,B=0.0044,A=0.9), B=(R=0.0014,G=0.0155,B=0.2031,A=0.9), bPickA=@n1)
   )
   > n3 = GetDynamicMaterial(self=Get Image_Item())
   > SetVectorParameterValue(self=@n3, ParameterName=Premium_A, Value=SelectColor(A=(R=0.3467,G=0.1441,B=0.0331,A=1), B=(R=0.0343,G=0.0343,B=0.3613,A=1), bPickA=@n1))
   > SetVectorParameterValue(self=@n3, ParameterName=Premium_B, Value=SelectColor(A=(R=1,G=0.7835,B=0.0976,A=1), B=(R=0,G=0.4678,B=1,A=1), bPickA=@n1))
   > SetVectorParameterValue(self=@n3, ParameterName=FallOff_Color, Value=SelectColor(A=(R=0.3467,G=0.0953,B=0.0185,A=1), B=(R=0.0612,G=0.2232,B=0.8963,A=1), bPickA=@n1))
  ]}
]
variables [
 LoadTileMaterialItemImage {}
 WasHovered {type=boolean, default=true}
 WasSelected {type=boolean, default=true}
]
widgets {
 info {inheritedWidgetCount=1, parentClass=&CommonButtonBase, rootWidgetClass=&SizeBox, widgetCount=21}
 widgets [
  SizeBox_0 {
   widgetClassPath=&SizeBox
   widget=SizeBox_0
   properties {bOverride_HeightOverride=true, bOverride_WidthOverride=true, heightOverride=450, widthOverride=320}
  }
  Overlay_67 {widgetClassPath=&Overlay, parent=SizeBox_0, widget=Overlay_67, slot {class=SizeBoxSlot, properties {padding=<ltrb 10 10 10 10>}}}
  Image_Frame {
   widgetClassPath=&Image
   parent=Overlay_67
   widget=Image_Frame
   properties {visibility=HitTestInvisible}
   slot {class=OverlaySlot, properties {horizontalAlignment=HAlign_Fill, verticalAlignment=VAlign_Fill}}
  }
  Overlay_Info {
   widgetClassPath=&Overlay
   parent=Overlay_67
   widget=Overlay_Info
   properties {clipping=ClipToBounds}
   slot {class=OverlaySlot, properties {horizontalAlignment=HAlign_Fill, verticalAlignment=VAlign_Fill}}
  }
  Image_Item {
   widgetClassPath=&FortLazyImage
   bIsVariable=true
   parent=Overlay_Info
   widget=Image_Item
   properties {brush {
     bIsDynamicallyLoaded=false
     drawAs=Image
     imageType=NoImage
     mirroring=NoMirror
     resourceName=None
     resourceObject=/Crew/UI/Subscriptions/Carabus/MI_ProgressiveItemTile_Pickaxe_CarabusTall
     tiling=NoTile
     imageSize=<512 512>
     margin=<ltrb 0 0 0 0>
     outlineSettings {
      bUseBrushTransparency=false
      roundingType=HalfHeightRadius
      width=0
      color {colorUseRule=UseColor_Specified, specifiedColor=<rgba 0 0 0 0>}
      cornerRadii=<0 0 0 0>
     }
     tintColor {colorUseRule=UseColor_Specified, specifiedColor=<rgba 1 1 1 1>}
     uVRegion {bIsValid=false, max=<0 0>, min=<0 0>}
    }}
   slot {class=OverlaySlot, properties {horizontalAlignment=HAlign_Fill, verticalAlignment=VAlign_Fill}}
  }
  Image_Gradient {
   widgetClassPath=&Image
   bIsVariable=true
   parent=Overlay_Info
   widget=Image_Gradient
   properties {brush {
     bIsDynamicallyLoaded=false
     drawAs=Image
     imageType=NoImage
     mirroring=NoMirror
     resourceName=None
     resourceObject=/Crew/UI/Materials/MI_UI_CrewLegacy_GradientOverlay
     tiling=NoTile
     imageSize=<32 32>
     margin=<ltrb 0 0 0 0>
     outlineSettings {
      bUseBrushTransparency=false
      roundingType=HalfHeightRadius
      width=0
      color {colorUseRule=UseColor_Specified, specifiedColor=<rgba 0 0 0 0>}
      cornerRadii=<0 0 0 0>
     }
     tintColor {colorUseRule=UseColor_Specified, specifiedColor=<rgba 1 1 1 1>}
     uVRegion {bIsValid=false, max=<0 0>, min=<0 0>}
    }}
   slot {class=OverlaySlot, properties {horizontalAlignment=HAlign_Fill, verticalAlignment=VAlign_Fill}}
  }
  VerticalBox_1 {
   widgetClassPath=&VerticalBox
   parent=Overlay_Info
   widget=VerticalBox_1
   slot {class=OverlaySlot, properties {horizontalAlignment=HAlign_Fill, verticalAlignment=VAlign_Bottom}}
  }
  TileText_ItemName {
   widgetClassPath=/Script/FortniteUI.AthenaItemShopReactiveTileText
   parent=VerticalBox_1
   widget=TileText_ItemName
   properties {
    autoWrapText=true
    bDebugScaling=true
    lineHeightPercentage=0.7
    maxCharacterCountBeforeScaling=40
    text=protocol backbling and xen pickaxe
    textTransformPolicy=ToUpper
    font {
     bForceMonospaced=false
     bMaterialIsStencil=false
     fontMaterial=null
     fontObject=&Burbank
     letterSpacing=0
     monospacedWidth=1
     size=26
     skewAmount=0
     typefaceFontName=BurbankRegular-Black
     outlineSettings {
      bApplyOutlineToDropShadows=false
      bMiteredCorners=false
      bSeparateFillAlpha=false
      outlineMaterial=null
      outlineSize=0
      outlineColor=<rgba 0 0 0 1>
     }
    }
    shadowOffset=<0 0>
   }
   slot {class=VerticalBoxSlot, properties {verticalAlignment=VAlign_Bottom, padding=<ltrb 20 0 20 0>}}
  }
  Overlay_Stage {
   widgetClassPath=&Overlay
   parent=VerticalBox_1
   widget=Overlay_Stage
   properties {visibility=Visible}
   slot {class=VerticalBoxSlot}
  }
  Image_PriceBg {
   widgetClassPath=&Image
   bIsVariable=true
   parent=Overlay_Stage
   widget=Image_PriceBg
   properties {
    brush {
     bIsDynamicallyLoaded=false
     drawAs=Box
     imageType=NoImage
     mirroring=NoMirror
     resourceName=None
     resourceObject=/Crew/UI/Materials/MI_UI_Skewed
     tiling=NoTile
     imageSize=<32 64>
     margin=<ltrb 0 0.25 0 0>
     outlineSettings {
      bUseBrushTransparency=false
      roundingType=HalfHeightRadius
      width=0
      color {colorUseRule=UseColor_Specified, specifiedColor=<rgba 0 0 0 0>}
      cornerRadii=<0 0 0 0>
     }
     tintColor {colorUseRule=UseColor_Specified, specifiedColor=<rgba 1 1 1 1>}
     uVRegion {bIsValid=false, max=<0 0>, min=<0 0>}
    }
    renderTransformPivot=<0.5 1>
   }
   slot {class=OverlaySlot, properties {horizontalAlignment=HAlign_Fill, verticalAlignment=VAlign_Fill}}
  }
  Text_Stages {
   widgetClassPath=&AthenaItemShopReactiveTileText
   bIsVariable=true
   parent=Overlay_Stage
   widget=Text_Stages
   properties {
    lineHeightPercentage=0.85
    maxCharacterCountBeforeScaling=17
    maxScaleDownPercentage=0.8
    mobileFontSizeMultiplier=1.2
    scrollStyle=&TextScrollStyle-Base_C
    text=STAGE 1
    textTransformPolicy=ToUpper
    colorAndOpacity {colorUseRule=UseColor_Specified, specifiedColor=<rgba 0 0 0 1>}
    font {
     bForceMonospaced=false
     bMaterialIsStencil=false
     fontMaterial=null
     fontObject=&BurbankSmall
     letterSpacing=0
     monospacedWidth=1
     size=18
     skewAmount=0
     typefaceFontName=Black
     outlineSettings {
      bApplyOutlineToDropShadows=false
      bMiteredCorners=false
      bSeparateFillAlpha=false
      outlineMaterial=null
      outlineSize=0
      outlineColor=<rgba 0 0 0 1>
     }
    }
    margin=<ltrb 0 0 0 10>
    renderTransform {angle=0, scale=<1 1>, shear=<-12 0>, translation=<0 0>}
    shadowOffset=<0 0>
   }
   slot {class=OverlaySlot, properties {horizontalAlignment=HAlign_Right, verticalAlignment=VAlign_Bottom, padding=<ltrb 0 0 10 0>}}
  }
  Overlay_InspectionText {
   widgetClassPath=&Overlay
   bIsVariable=true
   parent=VerticalBox_1
   widget=Overlay_InspectionText
   properties {visibility=Collapsed}
   slot {class=VerticalBoxSlot}
  }
  Image_86 {
   widgetClassPath=&Image
   parent=Overlay_InspectionText
   widget=Image_86
   properties {
    brush {
     bIsDynamicallyLoaded=false
     drawAs=Image
     imageType=NoImage
     mirroring=NoMirror
     resourceName=None
     resourceObject=null
     tiling=NoTile
     imageSize=<32 24>
     margin=<ltrb 0 0 0 0>
     outlineSettings {
      bUseBrushTransparency=false
      roundingType=HalfHeightRadius
      width=0
      color {colorUseRule=UseColor_Specified, specifiedColor=<rgba 0 0 0 0>}
      cornerRadii=<0 0 0 0>
     }
     tintColor {colorUseRule=UseColor_Specified, specifiedColor=<rgba 1 1 1 1>}
     uVRegion {bIsValid=false, max=<0 0>, min=<0 0>}
    }
    colorAndOpacity=<rgba 0.3564 0.7011 0.6172 1>
   }
   slot {class=OverlaySlot, properties {horizontalAlignment=HAlign_Fill, verticalAlignment=VAlign_Fill}}
  }
  Text_Subtitle {
   widgetClassPath=&CommonRichTextBlock
   parent=Overlay_InspectionText
   widget=Text_Subtitle
   properties {
    bOverrideDefaultStyle=true
    justification=Right
    lineHeightPercentage=0.5
    mobileTextBlockScale=1.2
    scrollStyle=&TextScrollStyle-Base_C
    text=NO TIME LIMIT TO COMPLETE SET
    textTransformPolicy=ToUpper
    defaultTextStyleOverride {
     overflowPolicy=Clip
     transformPolicy=None
     colorAndOpacity {colorUseRule=UseColor_Specified, specifiedColor=<rgba 0.0648 0.2542 0.227 1>}
     font {
      bForceMonospaced=false
      bMaterialIsStencil=false
      fontMaterial=null
      fontObject=&Burbank
      letterSpacing=0
      monospacedWidth=1
      size=11
      skewAmount=0
      typefaceFontName=BurbankRegular-Black
      outlineSettings {
       bApplyOutlineToDropShadows=false
       bMiteredCorners=false
       bSeparateFillAlpha=false
       outlineMaterial=null
       outlineSize=0
       outlineColor=<rgba 0 0 0 1>
      }
     }
     highlightColor {colorUseRule=UseColor_Specified, specifiedColor=<rgba 0 0 0 1>}
     highlightShape {
      bIsDynamicallyLoaded=false
      drawAs=Image
      imageType=NoImage
      mirroring=NoMirror
      resourceName=None
      resourceObject=null
      tiling=NoTile
      imageSize=<32 32>
      margin=<ltrb 0 0 0 0>
      outlineSettings {
       bUseBrushTransparency=false
       roundingType=HalfHeightRadius
       width=0
       color {colorUseRule=UseColor_Specified, specifiedColor=<rgba 0 0 0 0>}
       cornerRadii=<0 0 0 0>
      }
      tintColor {colorUseRule=UseColor_Specified, specifiedColor=<rgba 1 1 1 1>}
      uVRegion {bIsValid=false, max=<0 0>, min=<0 0>}
     }
     selectedBackgroundColor {colorUseRule=UseColor_Foreground, specifiedColor=<rgba 1 0 1 1>}
     shadowColorAndOpacity=<rgba 0 0 0 1>
     shadowOffset=<0 0>
     strikeBrush {
      bIsDynamicallyLoaded=false
      drawAs=Image
      imageType=NoImage
      mirroring=NoMirror
      resourceName=None
      resourceObject=null
      tiling=NoTile
      imageSize=<32 32>
      margin=<ltrb 0 0 0 0>
      outlineSettings {
       bUseBrushTransparency=false
       roundingType=HalfHeightRadius
       width=0
       color {colorUseRule=UseColor_Specified, specifiedColor=<rgba 0 0 0 0>}
       cornerRadii=<0 0 0 0>
      }
      tintColor {colorUseRule=UseColor_Specified, specifiedColor=<rgba 1 1 1 1>}
      uVRegion {bIsValid=false, max=<0 0>, min=<0 0>}
     }
     underlineBrush {
      bIsDynamicallyLoaded=false
      drawAs=Image
      imageType=NoImage
      mirroring=NoMirror
      resourceName=None
      resourceObject=null
      tiling=NoTile
      imageSize=<32 32>
      margin=<ltrb 0 0 0 0>
      outlineSettings {
       bUseBrushTransparency=false
       roundingType=HalfHeightRadius
       width=0
       color {colorUseRule=UseColor_Specified, specifiedColor=<rgba 0 0 0 0>}
       cornerRadii=<0 0 0 0>
      }
      tintColor {colorUseRule=UseColor_Specified, specifiedColor=<rgba 1 1 1 1>}
      uVRegion {bIsValid=false, max=<0 0>, min=<0 0>}
     }
    }
    margin=<ltrb 0 0 0 10>
    renderTransform {angle=0, scale=<1 1>, shear=<-7 0>, translation=<0 0>}
   }
   slot {class=OverlaySlot, properties {horizontalAlignment=HAlign_Right, verticalAlignment=VAlign_Center, padding=<ltrb 0 0 10 -0.5>}}
  }
  Switcher_CheckDelay {
   widgetClassPath=&CommonVisibilitySwitcher
   bIsVariable=true
   parent=Overlay_67
   widget=Switcher_CheckDelay
   properties {shownVisibility=HitTestInvisible, visibility=Collapsed}
   slot {class=OverlaySlot, properties {horizontalAlignment=HAlign_Right, padding=<ltrb 8 8 8 8>}}
  }
  Image_CompletedCheck {
   widgetClassPath=&Image
   parent=Switcher_CheckDelay
   widget=Image_CompletedCheck
   properties {
    visibility=SelfHitTestInvisible
    brush {
     bIsDynamicallyLoaded=false
     drawAs=Image
     imageType=NoImage
     mirroring=NoMirror
     resourceName=None
     resourceObject=/FortUILibrary/Foundation/Art_ToolKit/AngledBoxes/M_UI_AngledCheckBox_Crafting
     tiling=NoTile
     imageSize=<40 40>
     margin=<ltrb 0 0 0 0>
     outlineSettings {
      bUseBrushTransparency=false
      roundingType=HalfHeightRadius
      width=0
      color {colorUseRule=UseColor_Specified, specifiedColor=<rgba 0 0 0 0>}
      cornerRadii=<0 0 0 0>
     }
     tintColor {colorUseRule=UseColor_Specified, specifiedColor=<rgba 1 1 1 1>}
     uVRegion {bIsValid=false, max=<0 0>, min=<0 0>}
    }
   }
   slot {class=CommonVisibilitySwitcherSlot}
  }
  Overlay_Clock {
   widgetClassPath=&Overlay
   parent=Switcher_CheckDelay
   widget=Overlay_Clock
   slot {class=CommonVisibilitySwitcherSlot, properties {horizontalAlignment=HAlign_Left, verticalAlignment=VAlign_Top}}
  }
  Image {
   widgetClassPath=&Image
   parent=Overlay_Clock
   widget=Image
   properties {brush {
     bIsDynamicallyLoaded=false
     drawAs=Image
     imageType=NoImage
     mirroring=NoMirror
     resourceName=None
     resourceObject=/FortUILibrary/Foundation/Art_ToolKit/AngledBoxes/M_UI_AngledEmptyBlueBox
     tiling=NoTile
     imageSize=<40 40>
     margin=<ltrb 0 0 0 0>
     outlineSettings {
      bUseBrushTransparency=false
      roundingType=HalfHeightRadius
      width=0
      color {colorUseRule=UseColor_Specified, specifiedColor=<rgba 0 0 0 0>}
      cornerRadii=<0 0 0 0>
     }
     tintColor {colorUseRule=UseColor_Specified, specifiedColor=<rgba 1 1 1 1>}
     uVRegion {bIsValid=false, max=<0 0>, min=<0 0>}
    }}
   slot {class=OverlaySlot}
  }
  Image_32 {
   widgetClassPath=&Image
   parent=Overlay_Clock
   widget=Image_32
   properties {brush {
     bIsDynamicallyLoaded=false
     drawAs=Image
     imageType=NoImage
     mirroring=NoMirror
     resourceName=None
     resourceObject=/Game/Athena/UI/Frontend/ActivityBrowser/Art/M_UI_Timer_Icon_ShopOffer
     tiling=NoTile
     imageSize=<32 32>
     margin=<ltrb 0 0 0 0>
     outlineSettings {
      bUseBrushTransparency=false
      roundingType=HalfHeightRadius
      width=0
      color {colorUseRule=UseColor_Specified, specifiedColor=<rgba 0 0 0 0>}
      cornerRadii=<0 0 0 0>
     }
     tintColor {colorUseRule=UseColor_Specified, specifiedColor=<rgba 1 1 1 1>}
     uVRegion {bIsValid=false, max=<0 0>, min=<0 0>}
    }}
   slot {class=OverlaySlot, properties {horizontalAlignment=HAlign_Center, verticalAlignment=VAlign_Center}}
  }
  Image_Dimmer {
   widgetClassPath=&Image
   bIsVariable=true
   parent=Overlay_67
   widget=Image_Dimmer
   properties {renderOpacity=0, colorAndOpacity=<rgba 0.0157 0.0471 0.2667 0>}
   slot {class=OverlaySlot, properties {horizontalAlignment=HAlign_Fill, verticalAlignment=VAlign_Fill}}
  }
  InputActionWidget {widgetClassPath=&CommonActionWidget, bInherited=true, widget=null}
 ]
}
