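# Crew UI widget that shows a progressive cosmetic set's header, state tag, no-time-limit note and touch-only details button.
# Progress Changed hides the note on completion and, for unstarted sets with a view-model cutoff date, builds "Subscribe by" text.
# Update Tag restyles the tag as completed or expiring, collapsing it when neither applies.
# The button relays clicks through the MoreInfo Clicked dispatcher.
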WidgetBlueprint /Crew/UI/ProgressiveCosmetic/ProgressiveSetDetails
graphs [
 EventGraph {nodes [
   "On Clicked (Button_MoreInfo)"!() -> (OutputDelegate, then, Button:Common Button Base Object Reference) {is=title K2Node_ComponentBoundEvent}
   > Call More Info Clicked!(execute:Exec, self:Progressive Set Details Object Reference) -> (then:Exec) {is=title K2Node_CallDelegate}
  ]}
 MoreInfo Clicked {nodes [MoreInfo Clicked() {is=title K2Node_FunctionEntry}]}
 Progress Changed {nodes [
   n1 = Progress Changed!() -> (then, Progress:EProgressiveSetProgress Enum=NoProgress) {is=title K2Node_FunctionEntry}
   n2 = "Equal (Enum)"(A:EProgressiveSetProgress Enum=@n1, B:EProgressiveSetProgress Enum=Completed)
   > SetVisibility(
    self=Get Text_NoTimeLimit() -> (Text_NoTimeLimit:Common Text Object Reference)
    InVisibility=Select(Option 0=SelfHitTestInvisible, Option 1=Collapsed, Index=@n2)
   )
   n3 = "Not Equal (DateTime)"!(
    A:Date Time Structure=Get AcquisitionCutoffDate!(self:Crew Progressive Set VM Object Reference=Get CrewProgressiveSetVM() -> (CrewProgressiveSetVM:Crew Progressive Set VM Object Reference)) -> (AcquisitionCutoffDate:Date Time Structure)
    B:Date Time Structure="MinValue (DateTime)"!() -> (ReturnValue:Date Time Structure)
   ) -> (ReturnValue:bool)
   n4 = "Equal (Enum)"(A=@n1, B=NoProgress)
   n5 = AsShortDate!(InDateTime:"Date Time Structure (by ref)"=Get AcquisitionCutoffDate(self=Get CrewProgressiveSetVM())) -> (ReturnValue:Text)
   > Update Tag!(
    execute
    self
    Expiring Text:"Text (by ref)"=SelectText(A=Format Text!(Format="LOCTEXT(\"Subscribe by {ExpirationDate}\")", ExpirationDate:Text=@n5) -> (Result), bPickA=AND Boolean(A=@n4, B=@n3))
    Completed:bool=@n2
   ) -> (then)
  ]}
 Update Tag {nodes [
   Update Tag!() -> (then, Expiring Text:"Text (by ref)", Completed:bool=~) {is=title K2Node_FunctionEntry}
   n1 = TextIsEmpty(InText=Get Expiring Text() -> (Expiring Text:"Text (by ref)"))
   n2 = NOT Boolean(A=Get Completed() -> (Completed:bool))
   > SetVisibility(
    self=Get Tag_ItemSetState() -> (Tag_ItemSetState:Progressive Tags Base Object Reference)
    InVisibility=Select(Option 0=HitTestInvisible, Option 1=Collapsed, Index=AND Boolean(A=@n2, B=@n1))
   )
   > n3 = Branch(Condition=Get Completed())
   n4 = Get Tag_ItemSetState()
   > UpdateText!(execute, self:Progressive Tags Base Object Reference=@n4, Text:Text=Format Text!(Format="LOCTEXT(\"Completed!\")") -> (Result)) -> (then)
   n5 = Get Completed FG() -> (Completed FG:Slate Color Structure)
   > UpdateColor!(
    execute
    self:Progressive Tags Base Object Reference=@n4
    Background Color:Linear Color Structure=Get Completed BG() -> (Completed BG:Linear Color Structure)
    Text Color:Slate Color Structure=@n5
   ) -> (then)
   > Update Image!(
    execute
    self:Progressive Tags Base Object Reference=@n4
    Texture:Texture 2D Object Reference=/Game/UI/Foundation/Textures/Icons/Checks/T-Icon-Check-64
    Material:Material Interface Object Reference
    Color:Linear Color Structure=Break Slate Color!(SlateColor:"Slate Color Structure (by ref)"=@n5) -> (SpecifiedColor:Linear Color Structure, ColorUseRule:ESlateColorStylingMode Enum) {is=title K2Node_BreakStruct}
    Angle:double=0
   ) -> (then)
   n6 = Get Tag_ItemSetState()
   UpdateText(execute=@n3.else, self=@n6, Text=Get Expiring Text())
   n7 = Get Default FG() -> (Default FG:Slate Color Structure)
   > UpdateColor(self=@n6, Background Color=Get Default BG() -> (Default BG:Linear Color Structure), Text Color=@n7)
   > Update Image(self=@n6, Material=/Game/Athena/UI/Frontend/ActivityBrowser/Art/M_UI_Timer_Icon_ShopOffer, Color=Break Slate Color(SlateColor=@n7), Angle=0)
  ]}
]
variables [
 Completed BG {type=LinearColor, default=<rgba 0.0343 1 0.227 1>}
 Completed FG {type=SlateColor, default {colorUseRule=UseColor_Specified, specifiedColor=<rgba 0 0.147 0.063 1>}}
 Default BG {type=LinearColor, default=<rgba 0.7835 0.0086 0.1845 1>}
 Default FG {type=SlateColor, default {colorUseRule=UseColor_Specified, specifiedColor=<rgba 1 1 1 1>}}
 MoreInfo Clicked {}
 Text {type=string, default=""}
]
widgets {
 info {inheritedWidgetCount=0, parentClass=&CommonUserWidget, rootWidgetClass=&VerticalBox, widgetCount=6}
 widgets [
  VerticalBox_23 {widgetClassPath=&VerticalBox, widget=VerticalBox_23}
  Text_HeaderSetName {
   widgetClassPath=&CommonTextBlock
   bIsVariable=true
   parent=VerticalBox_23
   widget=Text_HeaderSetName
   properties {
    justification=Right
    style=/FortUILibrary/Styles/Typography/_TextStyle_SubH2._TextStyle_SubH2_C
    text="[PH] Protocol and Xen Set"
    textTransformPolicy=ToUpper
    wrapTextAt=512
    font {
     bForceMonospaced=false
     bMaterialIsStencil=false
     fontMaterial=null
     fontObject=&Burbank
     letterSpacing=0
     monospacedWidth=1
     size=33
     skewAmount=0
     typefaceFontName=BurbankRegular-Black
     outlineSettings {
      bApplyOutlineToDropShadows=false
      bMiteredCorners=false
      bSeparateFillAlpha=false
      outlineMaterial=null
      outlineSize=0
      outlineColor=<rgba 0 0 0 1>
     }
    }
    shadowOffset=<0 0>
    strikeBrush {
     bIsDynamicallyLoaded=false
     drawAs=NoDrawType
     imageType=NoImage
     mirroring=NoMirror
     resourceName=None
     resourceObject=null
     tiling=NoTile
     imageSize=<32 32>
     margin=<ltrb 0 0 0 0>
     outlineSettings {
      bUseBrushTransparency=false
      roundingType=HalfHeightRadius
      width=0
      color {colorUseRule=UseColor_Specified, specifiedColor=<rgba 0 0 0 0>}
      cornerRadii=<0 0 0 0>
     }
     tintColor {colorUseRule=UseColor_Specified, specifiedColor=<rgba 1 1 1 1>}
     uVRegion {bIsValid=false, max=<0 0>, min=<0 0>}
    }
   }
   slot {class=VerticalBoxSlot, properties {padding=<ltrb 0 0 0 16>}}
  }
  Tag_ItemSetState {
   widgetClassPath=&/ProgressiveTagsBase
   bIsVariable=true
   parent=VerticalBox_23
   widget=Tag_ItemSetState
   properties {
    image Texture=&T-Icon-Check-64
    text=Completed
    background Color=<rgba 0.0343 1 0.227 1>
    image Color=<rgba 0 0.147 0.063 1>
    text Color {colorUseRule=UseColor_Specified, specifiedColor=<rgba 0 0.147 0.063 1>}
   }
   slot {class=VerticalBoxSlot, properties {horizontalAlignment=HAlign_Right, verticalAlignment=VAlign_Center, padding=<ltrb 0 0 0 10>}}
  }
  Text_NoTimeLimit {
   widgetClassPath=&CommonTextBlock
   bIsVariable=true
   parent=VerticalBox_23
   widget=Text_NoTimeLimit
   properties {
    lineHeightPercentage=1.29
    style=/FortUILibrary/Styles/Typography/_TextStyle_Body1._TextStyle_Body1_C
    text=No time limit to complete set
    font {
     bForceMonospaced=false
     bMaterialIsStencil=false
     fontMaterial=null
     fontObject=&BurbankSmall
     letterSpacing=0
     monospacedWidth=1
     size=18
     skewAmount=0
     typefaceFontName=Bold
     outlineSettings {
      bApplyOutlineToDropShadows=false
      bMiteredCorners=false
      bSeparateFillAlpha=false
      outlineMaterial=null
      outlineSize=0
      outlineColor=<rgba 0 0 0 1>
     }
    }
    shadowOffset=<0 0>
    strikeBrush {
     bIsDynamicallyLoaded=false
     drawAs=NoDrawType
     imageType=NoImage
     mirroring=NoMirror
     resourceName=None
     resourceObject=null
     tiling=NoTile
     imageSize=<32 32>
     margin=<ltrb 0 0 0 0>
     outlineSettings {
      bUseBrushTransparency=false
      roundingType=HalfHeightRadius
      width=0
      color {colorUseRule=UseColor_Specified, specifiedColor=<rgba 0 0 0 0>}
      cornerRadii=<0 0 0 0>
     }
     tintColor {colorUseRule=UseColor_Specified, specifiedColor=<rgba 1 1 1 1>}
     uVRegion {bIsValid=false, max=<0 0>, min=<0 0>}
    }
   }
   slot {class=VerticalBoxSlot, properties {horizontalAlignment=HAlign_Right}}
  }
  CHVB_MobileButton {
   widgetClassPath=&CommonHardwareVisibilityBorder
   parent=VerticalBox_23
   widget=CHVB_MobileButton
   properties {
    horizontalAlignment=HAlign_Right
    verticalAlignment=VAlign_Top
    visibilityQuery {
     autoDescription=" ANY( Input.Touch )"
     tokenStreamVersion=0
     userDescription=""
     queryTokenStream [0,1,1,1,0]
     tagDictionary [{tagName=Input.Touch}]
    }
   }
   slot {class=VerticalBoxSlot, properties {verticalAlignment=VAlign_Bottom, padding=<ltrb 0 10 0 20>}}
  }
  Button_MoreInfo {
   widgetClassPath=&WBP_UIKit_Button_Regular
   bIsVariable=true
   parent=CHVB_MobileButton
   widget=Button_MoreInfo
   properties {text=Style Details}
   slot {class=BorderSlot, properties {horizontalAlignment=HAlign_Right, verticalAlignment=VAlign_Top, padding=<ltrb 0 0 0 0>}}
  }
 ]
}
