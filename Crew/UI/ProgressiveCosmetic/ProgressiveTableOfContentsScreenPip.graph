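# Pip button for the progressive cosmetic table-of-contents screen, built on the NoInputActionPipButton widget and representing a list of progressive sets.
# It holds an array of Crew progressive set view models in Sets.
# Its Selection Changed function marks the pip selected when the given set view model is in that array, and unselected otherwise.
# It adds no widgets of its own.
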WidgetBlueprint /Crew/UI/ProgressiveCosmetic/ProgressiveTableOfContentsScreenPip
graphs [
 EventGraph {links [], nodes []}
 Selection Changed {nodes [
   n1 = Selection Changed!() -> (then, Stage:Crew Progressive Set VM Object Reference) {is=title K2Node_FunctionEntry}
   > Set Selected!(
    execute
    self:No Input Action Pip Button Object Reference
    Selected:bool=Contains Item(
     TargetArray:Array of Crew Progressive Set VM Object References=Get Sets() -> (Sets:Array of Crew Progressive Set VM Object References)
     ItemToFind:"Crew Progressive Set VM Object Reference (by ref)"=@n1
    )
   ) -> (then) {is=title K2Node_CallFunction}
  ]}
]
variables [Sets {type="array<CrewProgressiveSetVM>", default []}]
widgets {
 info {
  inheritedWidgetCount=0
  parentClass=&/NoInputActionPipButton.NoInputActionPipButton_C
  rootWidgetClass=null
  widgetCount=0
 }
 widgets []
}
