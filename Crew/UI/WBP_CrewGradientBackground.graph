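# Crew UI background widget that shows a full-fill bottom-screen gradient overlay in one of two color schemes.
# It holds a single Image, using a legacy gradient material instance, inside an Overlay.
# SetCrewGradientColor branches on a boolean and sets Color1 and Color2 on the image's dynamic material: warm orange-brown when true, blue when false, each fading to transparent.
# PreConstruct is empty, so the color changes only when that function is called.
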WidgetBlueprint /Crew/UI/WBP_CrewGradientBackground
graphs [
 EventGraph {nodes [Event PreConstruct()]}
 SetCrewGradientColor {nodes [
   n1 = SetCrewGradientColor!() -> (then, Condition:bool=~) {is=title K2Node_FunctionEntry}
   > n2 = Branch(Condition=@n1)
   n3 = Get BottomScreenGradient() -> (BottomScreenGradient:Image Object Reference)
   > n4 = GetDynamicMaterial(self=@n3)
   > SetVectorParameterValue(self=@n4, ParameterName=Color1, Value=(R=0.2448,G=0.0693,B=0.0102,A=1))
   > SetVectorParameterValue(self=@n4, ParameterName=Color2, Value=(R=0.2462,G=0.0685,B=0.0103,A=0))
   n5 = GetDynamicMaterial(execute=@n2.else, self=@n3)
   > SetVectorParameterValue(self=@n5, ParameterName=Color1, Value=(R=0.0018,G=0.0356,B=0.2918,A=0.702))
   > SetVectorParameterValue(self=@n5, ParameterName=Color2, Value=(R=0.0021,G=0.0437,B=0.3564,A=0))
  ]}
]
variables []
widgets {
 info {inheritedWidgetCount=0, parentClass=&UserWidget, rootWidgetClass=&Overlay, widgetCount=2}
 widgets [
  Overlay_0 {widgetClassPath=&Overlay, widget=Overlay_0}
  BottomScreenGradient {
   widgetClassPath=&Image
   bIsVariable=true
   parent=Overlay_0
   widget=BottomScreenGradient
   properties {
    visibility=SelfHitTestInvisible
    brush {
     bIsDynamicallyLoaded=false
     drawAs=Image
     imageType=NoImage
     mirroring=NoMirror
     resourceName=None
     resourceObject=/Crew/UI/Materials/MI_UI_CrewLegacy_GradientOverlay
     tiling=NoTile
     imageSize=<1200 704>
     margin=<ltrb 0 0 0 0>
     outlineSettings {
      bUseBrushTransparency=false
      roundingType=HalfHeightRadius
      width=0
      color {colorUseRule=UseColor_Specified, specifiedColor=<rgba 0 0 0 0>}
      cornerRadii=<0 0 0 0>
     }
     tintColor {colorUseRule=UseColor_Specified, specifiedColor=<rgba 1 1 1 1>}
     uVRegion {bIsValid=false, max=<0 0>, min=<0 0>}
    }
   }
   slot {class=OverlaySlot, properties {horizontalAlignment=HAlign_Fill, verticalAlignment=VAlign_Fill}}
  }
 ]
}
