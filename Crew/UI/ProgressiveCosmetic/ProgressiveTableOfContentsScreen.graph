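# Fortnite Crew screen for browsing progressive cosmetic sets, with tile list, 3D preview, set details and subscribe button.
# C++ parent handles clicks; the Blueprint supplies view-model-based text, selection and visibility helpers, including empty and parental-blocked states.
# Activation plays the intro, starts the preview and applies subscription-based colors and preview material parameters.
# Pips each group two sets when over two show; mobile repads the layout, reparents the preview and drops hold-to-purchase.
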WidgetBlueprint /Crew/UI/ProgressiveCosmetic/ProgressiveTableOfContentsScreen
animations [Intro {
  binding_count=4
  length_seconds=0.25
  object=&self:Intro
  sequence {bindings [
    HB_MainContent {
     id=#1
     display_name=VB_LeftSideContent
     possessed_class=HorizontalBox
     tracks [
      {
       type=MovieScene2DTransformTrack
       sections [{
         type=MovieScene2DTransformSection
         channels [
          Translation.X {type=MovieSceneScriptingFloatChannel, const [0,1]}
          Translation.Y {type=MovieSceneScriptingFloatChannel, keys [[0,35],[5,0]]}
          Angle {type=MovieSceneScriptingFloatChannel, const [0,1]}
          Scale.X {type=MovieSceneScriptingFloatChannel, const [1,1]}
          Scale.Y {type=MovieSceneScriptingFloatChannel, const [1,1]}
          Shear.X {type=MovieSceneScriptingFloatChannel, const [0,1]}
          Shear.Y {type=MovieSceneScriptingFloatChannel, const [0,1]}
         ]
        }]
      }
      {
       type=MovieSceneFloatTrack
       display_name=Render Opacity
       sections [{type=MovieSceneFloatSection, channels [None {type=MovieSceneScriptingFloatChannel, keys [[0,0],[5,1]]}]}]
      }
     ]
    }
    VBox_RightSideContent {
     id=#2
     possessed_class=VerticalBox
     tracks [
      {
       type=MovieScene2DTransformTrack
       sections [{
         type=MovieScene2DTransformSection
         channels [
          Translation.X {type=MovieSceneScriptingFloatChannel, keys [[0,100],[5.0003,0]]}
          Translation.Y {type=MovieSceneScriptingFloatChannel, const [0,1]}
          Angle {type=MovieSceneScriptingFloatChannel, const [0,1]}
          Scale.X {type=MovieSceneScriptingFloatChannel, const [1,1]}
          Scale.Y {type=MovieSceneScriptingFloatChannel, const [1,1]}
          Shear.X {type=MovieSceneScriptingFloatChannel, const [0,1]}
          Shear.Y {type=MovieSceneScriptingFloatChannel, const [0,1]}
         ]
        }]
      }
      {
       type=MovieSceneFloatTrack
       display_name=Render Opacity
       sections [{type=MovieSceneFloatSection, channels [None {type=MovieSceneScriptingFloatChannel, keys [[0,0],[5.0003,1]]}]}]
      }
     ]
    }
    VerticalBox_Pips {
     id=#3
     possessed_class=VerticalBox
     tracks [{
       type=MovieSceneFloatTrack
       display_name=Render Opacity
       sections [{type=MovieSceneFloatSection, channels [None {type=MovieSceneScriptingFloatChannel, keys [[0,0],[5,1]]}]}]
      }]
    }
   ]}
 }]
graphs [
 BP_GetDesiredFocusTarget {nodes [
   Get Desired Focus Target()
   n1 = Get FortLocalPlayerVM() -> (FortLocalPlayerVM:Fort Local Player VM Object Reference)
   n2 = Get WBP_ContentPurchaseBlocker() -> (WBP_ContentPurchaseBlocker:WBP Content Purchase Blocker Object Reference)
   > Return Node!(
    execute
    ReturnValue:Widget Object Reference=Select(
     Option 0:Widget Object Reference=Get TileView_Sets() -> (TileView_Sets:Fort Item Tile View Object Reference)
     Option 1:Widget Object Reference=Get WBP_UIKit_Button_Loud!(self:WBP Content Purchase Blocker Object Reference=@n2) -> (WBP_UIKit_Button_Loud:WBP UIKit Button Loud Object Reference)
     Index=GetIsCabinModeEnabled!(self:Fort Local Player VM Object Reference=@n1) -> (ReturnValue:bool)
    ) -> (ReturnValue:Widget Object Reference)
   )
  ]}
 BP_GetDesiredInputConfig {nodes [
   Get Desired Input Config() {is=title K2Node_FunctionEntry}
   > Return Node(ReturnValue:UIInput Config Structure=Make UIInput Config!(
     bIgnoreMoveInput:bool=False
     bIgnoreLookInput:bool=False
     InputMode:ECommonInputMode Enum=All
     MouseCaptureMode:EMouseCaptureMode Enum=CaptureDuringMouseDown
     MouseLockMode:EMouseLockMode Enum=DoNotLock
     bHideCursorDuringViewportCapture:bool=True
    ) -> (UIInputConfig:UIInput Config Structure) {is=title K2Node_MakeStruct})
  ]}
 BackClicked {nodes [BackClicked() {is=title K2Node_FunctionEntry}]}
 ContentBlocking {nodes [
   n1 = ContentBlocking!() -> (then, IsContentBlocked:bool=~) {is=title K2Node_FunctionEntry}
   > SetVisibility(
    self=[Get WBP_ContentPurchaseBlocker(), Get BottomContentBlockerGradient() -> (BottomContentBlockerGradient:Image Object Reference)]
    InVisibility=Select(Option 0:ESlateVisibility Enum=Collapsed, Option 1:ESlateVisibility Enum=SelfHitTestInvisible, Index=@n1) -> (ReturnValue:ESlateVisibility Enum)
   )
   > SetVisibility(
    self=Get SizeBox_RightSideGatedContent() -> (SizeBox_RightSideGatedContent:Size Box Object Reference)
    InVisibility=Select(Option 0=SelfHitTestInvisible, Option 1=Collapsed, Index=@n1)
   )
  ]}
 EventGraph {nodes [
   n1 = "On Clicked (ConfirmPurchase_Button)"!() -> (OutputDelegate, then, Button:Common Button Base Object Reference) {is=title K2Node_ComponentBoundEvent}
   > SetIsEnabled(self=@n1.Button, bInIsEnabled=false)
   > HandleSubscribeEngaged(self:Fort Progressive Table Of Contents Screen Object Reference)
   n2 = "On Clicked (Button_MoreInfo)"!() -> (OutputDelegate, then, Button:Common Button Base Object Reference) {is=title K2Node_ComponentBoundEvent}
   > Fire Event Element Interacted(Element=@n2.Button)
   > HandleMoreInfoClicked(self:Fort Progressive Table Of Contents Screen Object Reference)
   "On Clicked (Button_Back)"!() -> (OutputDelegate, then, Button:Common Button Base Object Reference) {is=title K2Node_ComponentBoundEvent}
   > Fire Event Element Interacted(Element=Get Button_Back() -> (Button_Back:WBP UIKit Button Quiet Object Reference))
   > HandleBackClicked(self:Fort Progressive Table Of Contents Screen Object Reference)
   > Call Back Clicked!(execute:Exec, self:Progressive Table Of Contents Screen Object Reference) -> (then:Exec) {is=title K2Node_CallDelegate}
   n3 = "More Info Clicked (Widget_SetDetails)"() {is=title K2Node_ComponentBoundEvent}
   n4 = Get SelectedSet!(self:Crew Progressive Selection VM Object Reference=Get CrewProgressiveSelectionVM() -> (CrewProgressiveSelectionVM:Crew Progressive Selection VM Object Reference)) -> (SelectedSet:Crew Progressive Set VM Object Reference)
   FireEventElementInteractedByName!(
    execute=[@n3.then, @n6.else]
    Element Name:String=Get Source String from Text!(Text:Text=Get Name!(self:Crew Progressive Set VM Object Reference=@n4) -> (Name:Text)) -> (ReturnValue:String) {is=title K2Node_CallFunction}
    WidgetCallingFrom:Widget Object Reference=Self-Reference()
   ) -> (then, Success:bool)
   > HandleInspectSetTile(self:Fort Progressive Table Of Contents Screen Object Reference)
   "On Item Clicked (TileView_Sets)"!() -> (OutputDelegate, then, Item:Object Reference) {is=title K2Node_ComponentBoundEvent}
   n5 = Get CommonInputSubsystem()
   > n6 = Branch(Condition="Equal (Enum)"(A=GetCurrentInputType(self=@n5), B=Touch))
   Event OnSubscriptionPurchaseComplete!() -> (OutputDelegate, then, bSuccess:bool) {is=title K2Node_Event}
   > SetIsEnabled(self=@n1.Button, bInIsEnabled=true)
   Event On Activated()
   n7 = Get FortItemPreviewWidget() -> (FortItemPreviewWidget:WBP Item Preview Object Reference)
   > ActivateWidget(self=@n7)
   > PlayAnimation(InAnimation=Get Intro())
   > Update Subscribed Visuals!(
    execute
    self
    Subscribed:bool=IsSubscribed!(self:Crew VM Object Reference=Get CrewVM() -> (CrewVM:Crew VM Object Reference)) -> (ReturnValue:bool)
   ) -> (then)
   > Update Selected Set!(execute, self, Set:Crew Progressive Set VM Object Reference=Get SelectedSet(self=Get CrewProgressiveSelectionVM())) -> (then)
   > SetIsEnabled(self=Get TileView_Sets(), bInIsEnabled=true)
   > n8 = Cast To FortPlayerController(Object=GetOwningPlayer())
   > "Fire Event Sub View Opened (Non-Localized)"!(
    execute
    FortPC:Fort Player Controller Object Reference=@n8
    ViewName:String=CrewStylesScreen
    AdditionalInfo:Flow Analytics Additional Info Structure
    bIsOverlay:bool=false
   ) -> (then)
   Event PreConstruct()
   > n9 = Create Widget!(
    execute:Exec
    Class:User Widget Class Reference=/FortUILibrary/Systems/ItemShopTile/WBP_Price_PlatformStoreIcon.WBP_Price_PlatformStoreIcon_C
    OwningPlayer:Player Controller Object Reference=GetOwningPlayer()
    PreviewPlayStationPlatform:bool=Get PreviewPlatformIcon() -> (PreviewPlatformIcon:bool)
    ImageSize:Vector 2D Structure=Get PlatformIconSize() -> (PlatformIconSize:Vector 2D Structure)
   ) -> (then:Exec, ReturnValue:WBP Price Platform Store Icon Object Reference)
   > n10 = AddChildToGrid!(
    execute
    self:Grid Panel Object Reference=Get Root!(self:WBP UIKit Button CTA Base Object Reference=Get ConfirmPurchase_Button() -> (ConfirmPurchase_Button:WBP UIKit Button Loud Object Reference)) -> (Root:Grid Panel Object Reference)
    Content:Widget Object Reference=@n9
    InRow:int=1
    InColumn:int=4
   ) -> (then, ReturnValue:Grid Slot Object Reference)
   > Set VerticalAlignment!(execute, VerticalAlignment:EVerticalAlignment Enum=VAlign_Center, self:Grid Slot Object Reference=@n10) -> (then, Output_Get:EVerticalAlignment Enum=VAlign_Fill)
   > Set HorizontalAlignment!(execute, HorizontalAlignment:EHorizontalAlignment Enum=HAlign_Center, self:Grid Slot Object Reference=@n10) -> (then, Output_Get:EHorizontalAlignment Enum=HAlign_Fill)
   > SetGridSlotPadding!(execute, Widget:Widget Object Reference=@n9, Padding:Margin Structure=Make Margin(Left=4, Top=0, Right=8, Bottom=0)) -> (then)
   Event On Deactivated()
   > DeactivateWidget(self=@n7)
   > SetIsEnabled(self=Get TileView_Sets(), bInIsEnabled=false)
   Event Construct()
   > SetCompactUIStyle() {is=title K2Node_CallFunction}
  ]}
 Get Banner Text {nodes [
   n1 = Get Banner Text!() -> (then, All Complete:bool=~) {is=title K2Node_FunctionEntry}
   n2 = Get CrewVM()
   > Return Node!(
    execute
    Result:Text=SelectText(
     A="LOCTEXT(\"All available sets are completed!\")"
     B=Get ProgressiveInfoNewStagesUnlockFinePrint!(self:Crew VM Object Reference=@n2) -> (ProgressiveInfoNewStagesUnlockFinePrint:Text)
     bPickA=@n1
    )
   )
  ]}
 Get Initial Selection {nodes [
   n1 = Get Initial Selection!() -> (then, Visible Sets:Array of Crew Progressive Set VM Object References, Selected Set:Crew Progressive Set VM Object Reference) {is=title K2Node_FunctionEntry}
   > n2 = Branch(Condition=OR Boolean(A=Is Empty!(TargetArray:Array of Crew Progressive Set VM Object References=@n1) -> (ReturnValue:bool), B=IsValid(Object=@n1.Selected Set)))
   > Return Node(Result:Crew Progressive Set VM Object Reference=Get Selected Set() -> (Selected Set:Crew Progressive Set VM Object Reference))
   Return Node(
    execute=@n2.else
    Result="Get (a copy)"(Array:Array of Crew Progressive Set VM Object References=Get Visible Sets() -> (Visible Sets:Array of Crew Progressive Set VM Object References)) -> (Output:Crew Progressive Set VM Object Reference)
   )
  ]}
 Get Product Description {nodes [
   n1 = Get Product Description!() -> (then, Subscribed:bool=~, All Complete:bool=~) {is=title K2Node_FunctionEntry}
   > Return Node(Result:Text=SelectText(
     A="LOCTEXT(\"Crew exclusive cosmetics that unlock new stages each month you remain subscribed!\")"
     B="LOCTEXT(\"Subscribe to the Fortnite Crew and receive exclusive cosmetics that unlock new stages each month you remain subscribed!\")"
     bPickA=OR Boolean(A=@n1, B=@n1.All Complete)
    ))
  ]}
 Get Purchase Visibility {nodes [
   n1 = Get Purchase Visibility!() -> (then, Is Cabined:bool=~, Has CMS Data:bool=~, Sub State:EMcpSubscriptionState Enum=Inactive) {is=title K2Node_FunctionEntry}
   n2 = NOT Boolean(A=@n1)
   n3 = "Equal (Enum)"(A:EMcpSubscriptionState Enum=@n1.Sub State, B:EMcpSubscriptionState Enum=Inactive)
   > Return Node!(
    execute
    NewParam:ESlateVisibility Enum=Select(Option 0=Collapsed, Option 1=SelfHitTestInvisible, Index=AND Boolean!(A=@n2, B=@n1.Has CMS Data, C:bool=@n3) -> (ReturnValue))
   )
  ]}
 OpenUpsell {nodes [OpenUpsell() {is=title K2Node_FunctionEntry}]}
 SetCompactUIStyle {nodes [
   SetCompactUIStyle() {is=title K2Node_FunctionEntry}
   > Branch(Condition=IsMobileGame())
   > SetPadding!(
    execute
    self:Overlay Slot Object Reference=SlotAsOverlaySlot(Widget=Get SafeZone_Mobile() -> (SafeZone_Mobile:Safe Zone Object Reference))
    InPadding_Left:float=48
    InPadding_Top:float=24
    InPadding_Right:float=48
    InPadding_Bottom:float=24
   ) -> (then)
   > SetPadding(
    self:Horizontal Box Slot Object Reference=SlotAsHorizontalBoxSlot!(Widget:Widget Object Reference=Get VB_LeftSideContent() -> (VB_LeftSideContent:Vertical Box Object Reference)) -> (ReturnValue:Horizontal Box Slot Object Reference)
    InPadding_Left=48
    InPadding_Top=128
    InPadding_Right=0
    InPadding_Bottom=0
   )
   > SetPadding(
    self:Overlay Slot Object Reference=SlotAsOverlaySlot(Widget=Get RightSideContainer() -> (RightSideContainer:Overlay Object Reference))
    InPadding_Left=0
    InPadding_Top=0
    InPadding_Right=0
    InPadding_Bottom=128
   )
   > SetPadding(
    self=SlotAsOverlaySlot(Widget=Get WBP_ActionBarContainer() -> (WBP_ActionBarContainer:WBP Action Bar Container Object Reference))
    InPadding_Left=0
    InPadding_Top=0
    InPadding_Right=0
    InPadding_Bottom=0
   )
   > SetPadding(
    self=SlotAsOverlaySlot(Widget=Get WBP_PlayerInfo() -> (WBP_PlayerInfo:WBP Player Info Object Reference))
    InPadding_Left=0
    InPadding_Top=0
    InPadding_Right=0
    InPadding_Bottom=0
   )
   > SetPadding(
    self=SlotAsOverlaySlot(Widget=Get NoItemsMessage() -> (NoItemsMessage:Vertical Box Object Reference))
    InPadding_Left=0
    InPadding_Top=0
    InPadding_Right=0
    InPadding_Bottom=0
   )
   > SetPadding(self=SlotAsOverlaySlot(Widget=Get SizeBox_RightSideGatedContent()), InPadding_Left=0, InPadding_Top=0, InPadding_Right=0, InPadding_Bottom=0)
   > SetPadding(
    self:Vertical Box Slot Object Reference=SlotAsVerticalBoxSlot(Widget=Get NoItemsToggleBox() -> (NoItemsToggleBox:Vertical Box Object Reference))
    InPadding_Left=0
    InPadding_Top=24
    InPadding_Right=0
    InPadding_Bottom=0
   )
   n1 = Get FortItemPreviewWidget()
   > AddChild(self=Get Widget_CameraFrame() -> (Widget_CameraFrame:Overlay Object Reference), Content=@n1)
   n2 = SlotAsOverlaySlot(Widget=@n1)
   > SetHorizontalAlignment(self=@n2, InHorizontalAlignment=HAlign_Fill)
   > SetVerticalAlignment(self=@n2, InVerticalAlignment=VAlign_Fill)
   > SetPadding(self:Overlay Slot Object Reference=@n2, InPadding_Left=0, InPadding_Top=128, InPadding_Right=400, InPadding_Bottom=0)
   > Set WrapTextAt!(
    execute
    WrapTextAt:float=536
    self:Text Layout Widget Object Reference=Get Banner_Text!(self:Progressive Header Banner Object Reference=Get ProgressiveHeaderBanner() -> (ProgressiveHeaderBanner:Progressive Header Banner Object Reference)) -> (Banner_Text:Common Rich Text Block Object Reference)
   ) -> (then, Output_Get:float)
   > SetRequiresHold!(execute, self:Common Button Base Object Reference=Get ConfirmPurchase_Button(), bInRequiresHold:bool=false) -> (then)
  ]}
 SetPrice {nodes [
   n1 = SetPrice!() -> (then, PRICE:Text) {is=title K2Node_FunctionEntry}
   n2 = Get SubscriptionState!(self:Crew VM Object Reference=Get CrewVM()) -> (SubscriptionState:EMcpSubscriptionState Enum=Inactive)
   n3 = TextToUpper!(InText:"Text (by ref)"=Get PriceTemplate() -> (PriceTemplate:Text)) -> (ReturnValue:Text)
   > SetSecondaryText!(
    execute
    self:Fort CTAButton Object Reference=Get ConfirmPurchase_Button()
    ButtonSecondaryText:"Text (by ref)"=SelectText(B=Format Text!(Format=@n3, PRICE:Text=@n1) -> (Result), bPickA="Equal (Enum)"(A=@n2, B=Canceled))
   ) -> (then)
  ]}
 Update Selected Set {nodes [
   n1 = Update Selected Set!() -> (then, Set:Crew Progressive Set VM Object Reference) {is=title K2Node_FunctionEntry}
   n2 = Get CrewVM()
   > Branch(Condition=AND Boolean!(A=IsValid(Object=@n2), B=IsValid(Object=@n1)) -> (ReturnValue))
   > Setup Header Info!(
    execute
    self:Progressive Item State Title Object Reference=Get ProgressiveItem_Header() -> (ProgressiveItem_Header:Progressive Item State Title Object Reference)
    InputPin:"Text (by ref)"=Get Name(self=@n1)
    Subscribed:bool=IsSubscribed(self=@n2)
    Current Stage:int=0
    Max Stage:int=0
   ) -> (then)
   n3 = Get TileView_Sets()
   n4 = Get Set() -> (Set:Crew Progressive Set VM Object Reference)
   > Set Selected Item!(execute, self:List View Object Reference=@n3, Item:Object Reference=@n4) -> (then) {is=title K2Node_CallFunction}
   > Navigate To Item!(execute, self:List View Object Reference=@n3, Item:Object Reference=@n4) -> (then)
  ]}
 Update Subscribed Visuals {nodes [
   n1 = Update Subscribed Visuals!() -> (then, Subscribed:bool=~) {is=title K2Node_FunctionEntry}
   > n2 = Sequence()
   > SetColorAndOpacity(
    self=Get NoItems_HeaderText() -> (NoItems_HeaderText:Common Text Object Reference)
    InColorAndOpacity=Select(
     Option 0:Slate Color Structure=Get UnsubscribedTextColor() -> (UnsubscribedTextColor:Slate Color Structure)
     Option 1:Slate Color Structure=Get SubscribedTextColor() -> (SubscribedTextColor:Slate Color Structure)
     Index=@n1
    ) -> (ReturnValue:Slate Color Structure)
   )
   n3 = SelectFloat(A=1, B=0, bPickA=@n1)
   n4 = Get FortItemPreviewWidget()
   RequestChangeBackgroundMaterialScalarParameter!(execute=@n2.then_1, self:Fort Item Preview Screen Object Reference=@n4, ParameterName:Name=CrewLegacyUnsubscribed, Value:float=@n3) -> (then)
   > RequestChangeFloorMaterialScalarParameter!(execute, self:Fort Item Preview Screen Object Reference=@n4, ParameterName:Name=CrewLegacyUnsubscribed, Value:float=@n3) -> (then)
  ]}
 Update Visible Items {nodes [
   n1 = Update Visible Items!() -> (then, ItemsShown:Array of Crew Progressive Set VM Object References) {is=title K2Node_FunctionEntry}
   n2 = Length(TargetArray:Array of Crew Progressive Set VM Object References=@n1)
   n3 = "integer <= integer"(A=@n2)
   > SetVisibility(
    self=Get NoItemsMessage()
    InVisibility=Select(Option 0:ESlateVisibility Enum=Collapsed, Option 1:ESlateVisibility Enum=HitTestInvisible, Index=@n3) -> (ReturnValue:ESlateVisibility Enum)
   )
   > SetVisibility(
    self=[Get VBox_RightSideContent() -> (VBox_RightSideContent:Vertical Box Object Reference), Get NoItemsToggleBox()]
    InVisibility=Select(Option 0=SelfHitTestInvisible, Option 1=Collapsed, Index=@n3)
   )
   > Reset!(
    execute
    self:Dynamic Entry Box Object Reference=Get DynamicEntryBox_Pips() -> (DynamicEntryBox_Pips:Dynamic Entry Box Object Reference)
    bDeleteWidgets:bool=false
   ) -> (then)
   > Branch(Condition="integer > integer"(A=@n2, B=2))
   > n4 = For Each Loop(Array:Array of Crew Progressive Set VM Object References=Get ItemsShown() -> (ItemsShown:Array of Crew Progressive Set VM Object References)) -> (Array Element:Crew Progressive Set VM Object Reference)
   > n5 = Branch(Condition="Equal (Integer)"(A="% (Integer)"(A=@n4.Array Index, B=2) -> (ReturnValue=0)))
   > n6 = Create Entry of Class!(
    execute
    self:Dynamic Entry Box Object Reference=Get DynamicEntryBox_Pips()
    EntryClass:User Widget Class Reference=&/ProgressiveTableOfContentsScreenPip.ProgressiveTableOfContentsScreenPip_C
   ) -> (then, ReturnValue:Progressive Table Of Contents Screen Pip Object Reference)
   > n7 = Set LastPipAdded(LastPipAdded:Progressive Table Of Contents Screen Pip Object Reference=@n6) -> (Output_Get:Progressive Table Of Contents Screen Pip Object Reference)
   Add(
    execute=[@n5.else, @n7.then]
    TargetArray:Array of Crew Progressive Set VM Object References=Get Sets!(self:Progressive Table Of Contents Screen Pip Object Reference=Get LastPipAdded() -> (LastPipAdded:Progressive Table Of Contents Screen Pip Object Reference)) -> (Sets:Array of Crew Progressive Set VM Object References)
    NewItem:"Crew Progressive Set VM Object Reference (by ref)"=@n4
   ) {is=title K2Node_CallArrayFunction}
  ]}
]
variables [
 BackClicked {}
 OpenUpsell {}
 PlatformIconSize {type=Vector2D, default=<40 40>}
 PreviewPlatformIcon {type=boolean}
 SubscribedTextColor {type=SlateColor, default {colorUseRule=UseColor_Specified, specifiedColor=<rgba 0.147 0.0284 0.0044 1>}}
 UnsubscribedTextColor {type=SlateColor, default {colorUseRule=UseColor_Specified, specifiedColor=<rgba 0.0319 0.855 1 1>}}
]
widgets {
 info {
  inheritedWidgetCount=0
  namedSlotCount=1
  parentClass=/Script/CrewUI.FortProgressiveTableOfContentsScreen
  rootWidgetClass=&Overlay
  widgetCount=37
 }
 widgets [
  Overlay_29 {widgetClassPath=&Overlay, widget=Overlay_29}
  BottomContentBlockerGradient {
   widgetClassPath=&Image
   bIsVariable=true
   parent=Overlay_29
   widget=BottomContentBlockerGradient
   properties {
    visibility=Collapsed
    brush {
     bIsDynamicallyLoaded=false
     drawAs=Image
     imageType=NoImage
     mirroring=NoMirror
     resourceName=None
     resourceObject=/Game/UI/ContentControls/Materials/M_UI_ContentControls_LinearGradient
     tiling=NoTile
     imageSize=<32 32>
     margin=<ltrb 0 0 0 0>
     outlineSettings {
      bUseBrushTransparency=false
      roundingType=HalfHeightRadius
      width=0
      color {colorUseRule=UseColor_Specified, specifiedColor=<rgba 0 0 0 0>}
      cornerRadii=<0 0 0 0>
     }
     tintColor {colorUseRule=UseColor_Specified, specifiedColor=<rgba 1 1 1 1>}
     uVRegion {bIsValid=false, max=<0 0>, min=<0 0>}
    }
    colorAndOpacity=<rgba 1 1 1 0.6>
   }
   slot {class=OverlaySlot, properties {horizontalAlignment=HAlign_Fill, verticalAlignment=VAlign_Fill}}
  }
  WBP_CrewGradientBackground {
   widgetClassPath=/Crew/UI/WBP_CrewGradientBackground
   bIsVariable=true
   parent=Overlay_29
   widget=WBP_CrewGradientBackground
   slot {class=OverlaySlot, properties {horizontalAlignment=HAlign_Fill, verticalAlignment=VAlign_Bottom}}
  }
  SafeZone_Mobile {
   widgetClassPath=&SafeZone
   bIsVariable=true
   parent=Overlay_29
   widget=SafeZone_Mobile
   slot {class=OverlaySlot, properties {horizontalAlignment=HAlign_Fill, verticalAlignment=VAlign_Fill}}
  }
  Overlay_Content {widgetClassPath=&Overlay, parent=SafeZone_Mobile, widget=Overlay_Content, slot {class=SafeZoneSlot}}
  HB_MainContent {
   widgetClassPath=&HorizontalBox
   parent=Overlay_Content
   widget=HB_MainContent
   slot {class=OverlaySlot, properties {horizontalAlignment=HAlign_Fill, verticalAlignment=VAlign_Fill}}
  }
  VB_LeftSideContent {
   widgetClassPath=&VerticalBox
   bIsVariable=true
   parent=HB_MainContent
   widget=VB_LeftSideContent
   slot {class=HorizontalBoxSlot, properties {padding=<ltrb 110 92 0 64>, size {sizeRule=Fill, value=1}}}
  }
  ProgressiveItem_Header {
   widgetClassPath=&/ProgressiveItemStateTitle
   bIsVariable=true
   parent=VB_LeftSideContent
   widget=ProgressiveItem_Header
   properties {show Multi-Benefits tag=true}
   slot {class=VerticalBoxSlot, properties {horizontalAlignment=HAlign_Left, verticalAlignment=VAlign_Top}}
  }
  NoItemsToggleBox {
   widgetClassPath=&VerticalBox
   bIsVariable=true
   parent=VB_LeftSideContent
   widget=NoItemsToggleBox
   slot {class=VerticalBoxSlot, properties {padding=<ltrb 10 0 0 0>, size {sizeRule=Fill, value=1}}}
  }
  ProgressiveHeaderBanner {
   widgetClassPath=&/ProgressiveHeaderBanner
   bIsVariable=true
   parent=NoItemsToggleBox
   widget=ProgressiveHeaderBanner
   slot {class=VerticalBoxSlot, properties {horizontalAlignment=HAlign_Left}}
  }
  TileView_Sets {
   widgetClassPath=/Script/FortniteUI.FortItemTileView
   bIsVariable=true
   parent=NoItemsToggleBox
   widget=TileView_Sets
   properties {
    entryHeight=450
    entryWidgetClass=&/ProgressiveItem_Tall.ProgressiveItem_Tall_C
    entryWidth=320
    horizontalEntrySpacing=30
    numDesignerPreviewEntries=4
    tileAlignment=LeftAligned
    verticalEntrySpacing=30
    scrollBarStyle {
     thickness=0
     draggedThumbImage {
      bIsDynamicallyLoaded=false
      drawAs=Box
      imageType=FullColor
      mirroring=NoMirror
      resourceName=../../../Engine/Content/Slate/Common/Scrollbar_Thumb.png
      resourceObject=null
      tiling=NoTile
      imageSize=<0 0>
      margin=<ltrb 0.25 0.25 0.25 0.25>
      outlineSettings {
       bUseBrushTransparency=false
       roundingType=HalfHeightRadius
       width=0
       color {colorUseRule=UseColor_Specified, specifiedColor=<rgba 0 0 0 0>}
       cornerRadii=<0 0 0 0>
      }
      tintColor {colorUseRule=UseColor_Specified, specifiedColor=<rgba 1 1 1 1>}
      uVRegion {bIsValid=false, max=<0 0>, min=<0 0>}
     }
     horizontalBackgroundImage {
      bIsDynamicallyLoaded=false
      drawAs=NoDrawType
      imageType=NoImage
      mirroring=NoMirror
      resourceName=None
      resourceObject=null
      tiling=NoTile
      imageSize=<0 0>
      margin=<ltrb 0 0 0 0>
      outlineSettings {
       bUseBrushTransparency=false
       roundingType=HalfHeightRadius
       width=0
       color {colorUseRule=UseColor_Specified, specifiedColor=<rgba 0 0 0 0>}
       cornerRadii=<0 0 0 0>
      }
      tintColor {colorUseRule=UseColor_Specified, specifiedColor=<rgba 1 1 1 1>}
      uVRegion {bIsValid=false, max=<0 0>, min=<0 0>}
     }
     horizontalBottomSlotImage {
      bIsDynamicallyLoaded=false
      drawAs=Image
      imageType=FullColor
      mirroring=NoMirror
      resourceName=../../../Engine/Content/Slate/Common/Scrollbar_Background_Horizontal.png
      resourceObject=null
      tiling=NoTile
      imageSize=<8 8>
      margin=<ltrb 0 0 0 0>
      outlineSettings {
       bUseBrushTransparency=false
       roundingType=HalfHeightRadius
       width=0
       color {colorUseRule=UseColor_Specified, specifiedColor=<rgba 0 0 0 0>}
       cornerRadii=<0 0 0 0>
      }
      tintColor {colorUseRule=UseColor_Specified, specifiedColor=<rgba 1 1 1 1>}
      uVRegion {bIsValid=false, max=<0 0>, min=<0 0>}
     }
     horizontalTopSlotImage {
      bIsDynamicallyLoaded=false
      drawAs=Image
      imageType=FullColor
      mirroring=NoMirror
      resourceName=../../../Engine/Content/Slate/Common/Scrollbar_Background_Horizontal.png
      resourceObject=null
      tiling=NoTile
      imageSize=<8 8>
      margin=<ltrb 0 0 0 0>
      outlineSettings {
       bUseBrushTransparency=false
       roundingType=HalfHeightRadius
       width=0
       color {colorUseRule=UseColor_Specified, specifiedColor=<rgba 0 0 0 0>}
       cornerRadii=<0 0 0 0>
      }
      tintColor {colorUseRule=UseColor_Specified, specifiedColor=<rgba 1 1 1 1>}
      uVRegion {bIsValid=false, max=<0 0>, min=<0 0>}
     }
     hoveredThumbImage {
      bIsDynamicallyLoaded=false
      drawAs=Box
      imageType=FullColor
      mirroring=NoMirror
      resourceName=../../../Engine/Content/Slate/Common/Scrollbar_Thumb.png
      resourceObject=null
      tiling=NoTile
      imageSize=<0 0>
      margin=<ltrb 0.25 0.25 0.25 0.25>
      outlineSettings {
       bUseBrushTransparency=false
       roundingType=HalfHeightRadius
       width=0
       color {colorUseRule=UseColor_Specified, specifiedColor=<rgba 0 0 0 0>}
       cornerRadii=<0 0 0 0>
      }
      tintColor {colorUseRule=UseColor_Specified, specifiedColor=<rgba 1 1 1 1>}
      uVRegion {bIsValid=false, max=<0 0>, min=<0 0>}
     }
     normalThumbImage {
      bIsDynamicallyLoaded=false
      drawAs=Box
      imageType=FullColor
      mirroring=NoMirror
      resourceName=../../../Engine/Content/Slate/Common/Scrollbar_Thumb.png
      resourceObject=null
      tiling=NoTile
      imageSize=<0 0>
      margin=<ltrb 0.25 0.25 0.25 0.25>
      outlineSettings {
       bUseBrushTransparency=false
       roundingType=HalfHeightRadius
       width=0
       color {colorUseRule=UseColor_Specified, specifiedColor=<rgba 0 0 0 0>}
       cornerRadii=<0 0 0 0>
      }
      tintColor {colorUseRule=UseColor_Specified, specifiedColor=<rgba 1 1 1 1>}
      uVRegion {bIsValid=false, max=<0 0>, min=<0 0>}
     }
     verticalBackgroundImage {
      bIsDynamicallyLoaded=false
      drawAs=NoDrawType
      imageType=NoImage
      mirroring=NoMirror
      resourceName=None
      resourceObject=null
      tiling=NoTile
      imageSize=<0 0>
      margin=<ltrb 0 0 0 0>
      outlineSettings {
       bUseBrushTransparency=false
       roundingType=HalfHeightRadius
       width=0
       color {colorUseRule=UseColor_Specified, specifiedColor=<rgba 0 0 0 0>}
       cornerRadii=<0 0 0 0>
      }
      tintColor {colorUseRule=UseColor_Specified, specifiedColor=<rgba 1 1 1 1>}
      uVRegion {bIsValid=false, max=<0 0>, min=<0 0>}
     }
     verticalBottomSlotImage {
      bIsDynamicallyLoaded=false
      drawAs=Image
      imageType=FullColor
      mirroring=NoMirror
      resourceName=../../../Engine/Content/Slate/Common/Scrollbar_Background_Vertical.png
      resourceObject=null
      tiling=NoTile
      imageSize=<8 8>
      margin=<ltrb 0 0 0 0>
      outlineSettings {
       bUseBrushTransparency=false
       roundingType=HalfHeightRadius
       width=0
       color {colorUseRule=UseColor_Specified, specifiedColor=<rgba 0 0 0 0>}
       cornerRadii=<0 0 0 0>
      }
      tintColor {colorUseRule=UseColor_Specified, specifiedColor=<rgba 1 1 1 1>}
      uVRegion {bIsValid=false, max=<0 0>, min=<0 0>}
     }
     verticalTopSlotImage {
      bIsDynamicallyLoaded=false
      drawAs=Image
      imageType=FullColor
      mirroring=NoMirror
      resourceName=../../../Engine/Content/Slate/Common/Scrollbar_Background_Vertical.png
      resourceObject=null
      tiling=NoTile
      imageSize=<8 8>
      margin=<ltrb 0 0 0 0>
      outlineSettings {
       bUseBrushTransparency=false
       roundingType=HalfHeightRadius
       width=0
       color {colorUseRule=UseColor_Specified, specifiedColor=<rgba 0 0 0 0>}
       cornerRadii=<0 0 0 0>
      }
      tintColor {colorUseRule=UseColor_Specified, specifiedColor=<rgba 1 1 1 1>}
      uVRegion {bIsValid=false, max=<0 0>, min=<0 0>}
     }
    }
   }
   slot {class=VerticalBoxSlot, properties {padding=<ltrb 0 30 0 0>, size {sizeRule=Fill, value=1}}}
  }
  FortItemPreviewWidget {
   widgetClassPath=/FortniteCoreUI/Frontend/ItemPreview/WBP_ItemPreview
   bIsVariable=true
   parent=HB_MainContent
   widget=FortItemPreviewWidget
   properties {
    bPreviewActionsEnabled=false
    bSupportsActivationFocus=false
    itemViewObject=&self:WidgetTree.FortItemPreviewWidget.ItemViewObject
    previewSceneChangerTag {tagName=UI.ItemPreview.SceneChanger.FortniteCrew}
   }
   slot {class=HorizontalBoxSlot, properties {size {sizeRule=Fill, value=1}}}
  }
  VerticalBox_Pips {
   widgetClassPath=&VerticalBox
   parent=Overlay_Content
   widget=VerticalBox_Pips
   slot {class=OverlaySlot, properties {horizontalAlignment=HAlign_Fill, verticalAlignment=VAlign_Fill}}
  }
  Spacer_66 {
   widgetClassPath=&Spacer
   parent=VerticalBox_Pips
   widget=Spacer_66
   slot {class=VerticalBoxSlot, properties {horizontalAlignment=HAlign_Left, size {sizeRule=Fill, value=0.19}}}
  }
  ScaleBox_0 {
   widgetClassPath=&ScaleBox
   parent=VerticalBox_Pips
   widget=ScaleBox_0
   properties {stretch=ScaleToFitY, stretchDirection=DownOnly}
   slot {class=VerticalBoxSlot, properties {size {sizeRule=Fill, value=0.6}}}
  }
  SizeBox_EntryBox_Pips {
   widgetClassPath=&SizeBox
   parent=ScaleBox_0
   widget=SizeBox_EntryBox_Pips
   properties {bOverride_WidthOverride=true, widthOverride=64}
   slot {class=ScaleBoxSlot, properties {horizontalAlignment=HAlign_Left}}
  }
  DynamicEntryBox_Pips {
   widgetClassPath=&DynamicEntryBox
   bIsVariable=true
   parent=SizeBox_EntryBox_Pips
   widget=DynamicEntryBox_Pips
   properties {
    entryBoxType=Vertical
    entryWidgetClass=&ProgressiveTableOfContentsScreenPip_C
    numDesignerPreviewEntries=5
   }
   slot {class=SizeBoxSlot}
  }
  Spacer {
   widgetClassPath=&Spacer
   parent=VerticalBox_Pips
   widget=Spacer
   slot {class=VerticalBoxSlot, properties {horizontalAlignment=HAlign_Left, size {sizeRule=Fill, value=0.19}}}
  }
  RightSideContainer {
   widgetClassPath=&Overlay
   bIsVariable=true
   parent=Overlay_Content
   widget=RightSideContainer
   slot {class=OverlaySlot, properties {horizontalAlignment=HAlign_Right, verticalAlignment=VAlign_Fill, padding=<ltrb 0 0 40 110>}}
  }
  SizeBox_RightSideGatedContent {
   widgetClassPath=&SizeBox
   bIsVariable=true
   parent=RightSideContainer
   widget=SizeBox_RightSideGatedContent
   slot {class=OverlaySlot, properties {horizontalAlignment=HAlign_Right, verticalAlignment=VAlign_Bottom, padding=<ltrb 0 0 40 0>}}
  }
  VBox_RightSideContent {
   widgetClassPath=&VerticalBox
   bIsVariable=true
   parent=SizeBox_RightSideGatedContent
   widget=VBox_RightSideContent
   slot {class=SizeBoxSlot}
  }
  ErrorInfoFlag {
   widgetClassPath=/Crew/UI/ErrorInfoFlag
   bIsVariable=true
   parent=VBox_RightSideContent
   widget=ErrorInfoFlag
   properties {accentColor=<rgba 0.7922 0.0078 0.1843 1>}
   slot {class=VerticalBoxSlot, properties {horizontalAlignment=HAlign_Right, padding=<ltrb 0 0 0 20>}}
  }
  Widget_SetDetails {
   widgetClassPath=&/ProgressiveSetDetails
   bIsVariable=true
   parent=VBox_RightSideContent
   widget=Widget_SetDetails
   slot {class=VerticalBoxSlot, properties {horizontalAlignment=HAlign_Right, padding=<ltrb 0 0 0 20>}}
  }
  ConfirmPurchase_Button {
   widgetClassPath=/FortUILibrary/UIKit/Buttons/WBP_UIKit_Button_Loud
   bIsVariable=true
   parent=VBox_RightSideContent
   widget=ConfirmPurchase_Button
   properties {
    bIsFocusable=false
    bRequiresHold=true
    buttonSize=Large
    minHeight=56
    minWidth=632
    overrideSecondaryTextSizeMultiplier=0.5
    secondaryText=$11.99
    text=Confirm Purchase
    triggeringInputAction {dataTable=/FortUILibrary/Input/ItemShopInputActionDataTable, rowName=PurchaseCrewSetHold}
   }
   slot {class=VerticalBoxSlot, properties {horizontalAlignment=HAlign_Right, verticalAlignment=VAlign_Center}}
  }
  WBP_ContentPurchaseBlocker {
   widgetClassPath=/FortniteCoreUI/UI/ContentControls/WBP_ContentPurchaseBlocker
   bIsVariable=true
   parent=RightSideContainer
   widget=WBP_ContentPurchaseBlocker
   properties {
    text_Details=To purchase this item or bundle ask your parent or guardian to set up your account from Parental Controls.
    visibility=Collapsed
   }
   slot {class=OverlaySlot, properties {horizontalAlignment=HAlign_Fill, verticalAlignment=VAlign_Center, padding=<ltrb 50 50 50 50>}}
  }
  Text_DebugPreviewDetails {
   widgetClassPath=&CommonTextBlock
   parent=RightSideContainer
   widget=Text_DebugPreviewDetails
   properties {
    wrapTextAt=650
    font {
     bForceMonospaced=false
     bMaterialIsStencil=false
     fontMaterial=null
     fontObject=&NotoSans
     letterSpacing=0
     monospacedWidth=1
     size=12
     skewAmount=0
     typefaceFontName=NotoSans-Regular
     outlineSettings {
      bApplyOutlineToDropShadows=false
      bMiteredCorners=false
      bSeparateFillAlpha=false
      outlineMaterial=null
      outlineSize=1
      outlineColor=<rgba 0 0 0 1>
     }
    }
    shadowOffset=<4 0>
   }
   slot {class=OverlaySlot, properties {horizontalAlignment=HAlign_Center, verticalAlignment=VAlign_Center}}
  }
  WBP_ActionBarContainer {
   widgetClassPath=/Game/UI/Frontend/BottomBar/WBP_ActionBarContainer
   bIsVariable=true
   parent=Overlay_Content
   widget=WBP_ActionBarContainer
   slot {class=OverlaySlot, properties {horizontalAlignment=HAlign_Right, verticalAlignment=VAlign_Bottom, padding=<ltrb 0 0 80 24>}}
  }
  HBox_BottomButtons {widgetClassPath=&HorizontalBox, namedSlotHost=WBP_ActionBarContainer, widget=HBox_BottomButtons}
  Button_MoreInfo {
   widgetClassPath=/FortUILibrary/UIKit/Buttons/WBP_UIKit_Button_Quiet
   bIsVariable=true
   parent=HBox_BottomButtons
   widget=Button_MoreInfo
   properties {minWidth=184, text=More Info, triggeringInputAction {dataTable=&InputActionDataTable, rowName=GenericFaceButtonTop}}
   slot {class=HorizontalBoxSlot, properties {horizontalAlignment=HAlign_Right, verticalAlignment=VAlign_Top}}
  }
  WBP_AthenaBottomBar_WithNotifications {
   widgetClassPath=/Game/UI/Frontend/BottomBar/WBP_AthenaBottomBar_WithNotifications
   bIsVariable=true
   parent=HBox_BottomButtons
   widget=WBP_AthenaBottomBar_WithNotifications
   slot {class=HorizontalBoxSlot, properties {padding=<ltrb 24 0 0 0>}}
  }
  Button_Back {
   widgetClassPath=&WBP_UIKit_Button_Quiet
   bIsVariable=true
   parent=HBox_BottomButtons
   widget=Button_Back
   properties {text=Back, triggeringInputAction {dataTable=&InputActionDataTable, rowName=GenericBack}}
   slot {class=HorizontalBoxSlot, properties {horizontalAlignment=HAlign_Right, verticalAlignment=VAlign_Top, padding=<ltrb 24 0 0 0>}}
  }
  WBP_PlayerInfo {
   widgetClassPath=/Game/UI/Frontend/ItemShop/MultiProduct/WBP_PlayerInfo
   bIsVariable=true
   parent=Overlay_Content
   widget=WBP_PlayerInfo
   slot {class=OverlaySlot, properties {horizontalAlignment=HAlign_Right, padding=<ltrb 0 70 48 0>}}
  }
  NoItemsMessage {
   widgetClassPath=&VerticalBox
   bIsVariable=true
   parent=Overlay_Content
   widget=NoItemsMessage
   properties {visibility=Collapsed}
   slot {class=OverlaySlot, properties {horizontalAlignment=HAlign_Center, verticalAlignment=VAlign_Center, padding=<ltrb 128 0 128 0>}}
  }
  NoItems_HeaderText {
   widgetClassPath=&CommonTextBlock
   bIsVariable=true
   parent=NoItemsMessage
   widget=NoItems_HeaderText
   properties {
    justification=Center
    mobileFontSizeMultiplier=1.2
    text=There are no offers that you are eligible for at this time
    textTransformPolicy=ToUpper
    colorAndOpacity {colorUseRule=UseColor_Specified, specifiedColor=<rgba 0.0314 0.8549 1 1>}
    font {
     bForceMonospaced=false
     bMaterialIsStencil=false
     fontMaterial=null
     fontObject=&Burbank
     letterSpacing=0
     monospacedWidth=1
     size=24
     skewAmount=0
     typefaceFontName=Burbank-PlayerPower
     outlineSettings {
      bApplyOutlineToDropShadows=false
      bMiteredCorners=false
      bSeparateFillAlpha=false
      outlineMaterial=null
      outlineSize=0
      outlineColor=<rgba 0 0 0 1>
     }
    }
    renderTransform {angle=0, scale=<1 1>, shear=<-8 0>, translation=<0 0>}
    shadowOffset=<0 0>
   }
   slot {class=VerticalBoxSlot, properties {horizontalAlignment=HAlign_Center, size {sizeRule=Fill, value=1}}}
  }
  CommonTextBlock {
   widgetClassPath=&CommonTextBlock
   parent=NoItemsMessage
   widget=CommonTextBlock
   properties {
    autoWrapText=true
    justification=Center
    lineHeightPercentage=0.8
    mobileFontSizeMultiplier=1.2
    text="Check back soon to see the latest Crew Legacy Cosmetic!"
    font {
     bForceMonospaced=false
     bMaterialIsStencil=false
     fontMaterial=null
     fontObject=&BurbankSmall
     letterSpacing=0
     monospacedWidth=1
     size=16
     skewAmount=0
     typefaceFontName=Black
     outlineSettings {
      bApplyOutlineToDropShadows=false
      bMiteredCorners=false
      bSeparateFillAlpha=false
      outlineMaterial=null
      outlineSize=0
      outlineColor=<rgba 0 0 0 1>
     }
    }
    shadowOffset=<0 0>
   }
   slot {class=VerticalBoxSlot, properties {verticalAlignment=VAlign_Center}}
  }
  Widget_CameraFrame {
   widgetClassPath=&Overlay
   bIsVariable=true
   parent=Overlay_Content
   widget=Widget_CameraFrame
   slot {class=OverlaySlot, properties {horizontalAlignment=HAlign_Right, verticalAlignment=VAlign_Fill}}
  }
  Spacer_FrameSize {widgetClassPath=&Spacer, parent=Widget_CameraFrame, widget=Spacer_FrameSize, properties {size=<1200 0>}, slot {class=OverlaySlot}}
 ]
}
